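# Physics asset for the cape of the M_MED_Realm male medium player body, built entirely from capsule bodies.
# Four six-segment chains (outer and middle, left and right) hang from MSTR bodies, with limited swing and zero twist limits.
# Free-angular constraints cross-link the chain tips, joining the two middle tips in both directions.
# Torso, limb, coat and two large root side capsules have no constraints, apparently serving as colliders.
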PhysicsAsset /BRCosmetics/Characters/Player/Male/Medium/Bodies/M_MED_Realm/Meshes/M_MED_Realm_Cape_Physics
bodies {
 calf_l [{shapeType=Capsule, length=43.179, radius=7.1561, shapeName=calf_l_capsule, center=<-18.4993 0 0>, rotation=<rot 90 0 0>}]
 calf_r [{shapeType=Capsule, length=43.179, radius=7.1561, shapeName=calf_r_capsule, center=<18.4993 0 0>, rotation=<rot 90 0 0>}]
 cape_01_l [{shapeType=Capsule, length=13.0431, radius=6.9275, shapeName=cape_01_l_capsule, center=<11.9983 0 0>, rotation=<rot 90 153.435 -116.565>}]
 cape_01_r [{shapeType=Capsule, length=9.6914, radius=5.0845, shapeName=cape_01_r_capsule, center=<8.2303 0 0>, rotation=<rot 90 0 0>}]
 cape_02_l [
  {shapeType=Capsule, length=9.564, radius=7.6988, shapeName=cape_02_l_capsule, center=<18.223 0 0>, rotation=<rot 0 0 90>}
  {shapeType=Capsule, length=14.1165, radius=6.0281, shapeName=cape_02_l_capsule, center=<6.364 0 0>, rotation=<rot 0 0 90>}
 ]
 cape_02_r [
  {shapeType=Capsule, length=18, radius=6, shapeName=cape_02_r_capsule, center=<5 0 0>, rotation=<rot 90 0 90>}
  {shapeType=Capsule, length=16.75, radius=7.25, shapeName=cape_02_r_capsule, center=<20 0 0>, rotation=<rot 90 0 90>}
 ]
 cape_03_l [{shapeType=Capsule, length=12, radius=9.25, shapeName=cape_03_l_capsule, center=<11 0 0>, rotation=<rot 90 0 90>}]
 cape_03_r [{shapeType=Capsule, length=12, radius=9.25, shapeName=cape_03_r_capsule, center=<11 0 0>, rotation=<rot 0 0 90>}]
 cape_04_l [{shapeType=Capsule, length=25, radius=10, shapeName=cape_04_l_capsule, center=<8 -6 3>, rotation=<rot 0 0 90>}]
 cape_04_r [{shapeType=Capsule, length=25, radius=10, shapeName=cape_04_r_capsule, center=<8 6 3>, rotation=<rot 0 0 90>}]
 cape_05_l [{shapeType=Capsule, length=25, radius=10, shapeName=cape_05_l_capsule, center=<8 -6 3>, rotation=<rot 0 0 90>}]
 cape_05_r [{shapeType=Capsule, length=25, radius=10, shapeName=cape_05_r_capsule, center=<8 6 3>, rotation=<rot 0 0 90>}]
 cape_06_l [{shapeType=Capsule, length=25, radius=10, shapeName=cape_06_l_capsule, center=<8 -6 3>, rotation=<rot -0.0001 0 100>}]
 cape_06_r [{shapeType=Capsule, length=25, radius=10, shapeName=cape_06_r_capsule, center=<8 6 3>, rotation=<rot -0.0001 0 80>}]
 cape_l_MSTR [{shapeType=Capsule, length=1, radius=1, shapeName=cape_l_MSTR_capsule, rotation=<rot 0 -90 0>}]
 cape_mid_01_l [{shapeType=Capsule, length=15.772, radius=4.5132, shapeName=cape_mid_01_l_capsule, center=<10.9091 0 5.8>, rotation=<rot 82.4316 -180 -180>}]
 cape_mid_01_r [{shapeType=Capsule, length=15.772, radius=4.5132, shapeName=cape_mid_01_r_capsule, center=<10.9091 0 5.8>, rotation=<rot 82.4316 180 180>}]
 cape_mid_02_l [{shapeType=Capsule, length=15.2976, radius=5.525, shapeName=cape_mid_02_l_capsule, center=<15.4392 0 0>, rotation=<rot 90 0 0>}]
 cape_mid_02_r [{shapeType=Capsule, length=15.2976, radius=5.525, shapeName=cape_mid_02_r_capsule, center=<15.4392 0 0>, rotation=<rot 90 0 0>}]
 cape_mid_03_l [{shapeType=Capsule, length=12, radius=9.25, shapeName=cape_mid_03_l_capsule, center=<11 0 0>, rotation=<rot 0 0 90>}]
 cape_mid_03_r [{shapeType=Capsule, length=12, radius=9.25, shapeName=cape_mid_03_r_capsule, center=<11 0 0>, rotation=<rot 0 0 90>}]
 cape_mid_04_l [{shapeType=Capsule, length=25, radius=10, shapeName=cape_mid_04_l_capsule, center=<8 0 3>, rotation=<rot 0 0 90>}]
 cape_mid_04_r [{shapeType=Capsule, length=25, radius=10, shapeName=cape_mid_04_r_capsule, center=<8 0 3>, rotation=<rot 0 0 90>}]
 cape_mid_05_l [{shapeType=Capsule, length=25, radius=10, shapeName=cape_mid_05_l_capsule, center=<8 0 3>, rotation=<rot 0 0 90>}]
 cape_mid_05_r [{shapeType=Capsule, length=25, radius=10, shapeName=cape_mid_05_r_capsule, center=<8 0 3>, rotation=<rot 0 0 90>}]
 cape_mid_06_l [{shapeType=Capsule, length=25, radius=10, shapeName=cape_mid_06_l_capsule, center=<8 0 3>, rotation=<rot 0 0 90>}]
 cape_mid_06_r [{shapeType=Capsule, length=25, radius=10, shapeName=cape_mid_06_r_capsule, center=<8 0 3>, rotation=<rot 0 0 90>}]
 cape_mid_l_MSTR [{shapeType=Capsule, length=1, radius=1, shapeName=cape_mid_l_MSTR_capsule, rotation=<rot 0 -90 0>}]
 cape_mid_r_MSTR [{shapeType=Capsule, length=1, radius=1, shapeName=cape_mid_r_MSTR_capsule, rotation=<rot 0 -90 0>}]
 cape_r_MSTR [{shapeType=Capsule, length=1, radius=1, shapeName=cape_r_MSTR_capsule, rotation=<rot 0 -90 0>}]
 clavicle_l [{shapeType=Capsule, length=12.0737, radius=7.5263, shapeName=clavicle_l_capsule, center=<9.3 0 0>, rotation=<rot 81 0 0>}]
 clavicle_r [{shapeType=Capsule, length=12.0737, radius=7.5263, shapeName=clavicle_r_capsule, center=<-9.3 0 0>, rotation=<rot 81 0 0>}]
 dyn_coat_bk_c_02 [{shapeType=Capsule, length=21.7863, radius=12.9186, shapeName=dyn_coat_bk_c_02_capsule, center=<0 0 -10>, rotation=<rot 90 0 0>}]
 dyn_coat_bk_c_04 [{shapeType=Capsule, length=20.3449, radius=11.9475, shapeName=dyn_coat_bk_c_04_capsule, center=<0 0 -10>, rotation=<rot 90 0 0>}]
 dyn_coat_bk_c_06 [{shapeType=Capsule, length=26.9918, radius=11.7774, shapeName=dyn_coat_bk_c_06_capsule, center=<0 0 -10>, rotation=<rot 90 0 0>}]
 dyn_coat_bk_l_02 [{shapeType=Capsule, length=19.963, radius=12.3294, shapeName=dyn_coat_bk_l_02_capsule, center=<0 0 -12>, rotation=<rot 90 0 0>}]
 dyn_coat_bk_l_04 [{shapeType=Capsule, length=20.3449, radius=11.9475, shapeName=dyn_coat_bk_l_04_capsule, center=<0 0 -12>, rotation=<rot 90 0 0>}]
 dyn_coat_bk_l_06 [{shapeType=Capsule, length=20.515, radius=11.7774, shapeName=dyn_coat_bk_l_06_capsule, center=<0 0 -12>, rotation=<rot 90 0 0>}]
 dyn_coat_bk_r_02 [{shapeType=Capsule, length=15.998, radius=12.0961, shapeName=dyn_coat_bk_r_02_capsule, center=<0 0 12>, rotation=<rot 90 -5.4103 -8.731>}]
 dyn_coat_bk_r_04 [{shapeType=Capsule, length=12.6236, radius=11.9475, shapeName=dyn_coat_bk_r_04_capsule, center=<0 0 12>, rotation=<rot 90 0 0>}]
 dyn_coat_bk_r_06 [{shapeType=Capsule, length=12.7938, radius=11.7774, shapeName=dyn_coat_bk_r_06_capsule, center=<0 0 12>, rotation=<rot 90 0 0>}]
 dyn_coat_front_l_04 [{shapeType=Capsule, length=22.2896, radius=10.0028, shapeName=dyn_coat_l_06_capsule, center=<0 0 -8.817>, rotation=<rot 90 0 0>}]
 dyn_coat_front_l_06 [{shapeType=Capsule, length=22.2896, radius=10.0028, shapeName=dyn_coat_l_06_capsule, center=<0 0 -8.817>, rotation=<rot 90 0 0>}]
 dyn_coat_l_04 [{shapeType=Capsule, length=21.0396, radius=11.2528, shapeName=dyn_coat_l_04_capsule, center=<0 0 -12>, rotation=<rot 90 0 0>}]
 dyn_coat_l_06 [{shapeType=Capsule, length=21.0396, radius=11.2528, shapeName=dyn_coat_l_06_capsule, center=<0 0 -12>, rotation=<rot 90 0 0>}]
 dyn_coat_r_04 [{shapeType=Capsule, length=22.3371, radius=13.8187, shapeName=dyn_coat_r_04_capsule, center=<0 0 12>, rotation=<rot 90 0 0>}]
 dyn_coat_r_06 [{shapeType=Capsule, length=22.8478, radius=13.3079, shapeName=dyn_coat_r_06_capsule, center=<0 0 12>, rotation=<rot 90 0 0>}]
 foot_l [
  {shapeType=Capsule, length=17.8017, radius=7, shapeName=foot_l_capsule, center=<-4.0851 7.2082 0>, rotation=<rot 90 180 -90>}
  {shapeType=Capsule, length=12.1748, radius=9.0432, shapeName=foot_l_capsule, center=<0 -1.6375 0>}
 ]
 foot_r [
  {shapeType=Capsule, length=16.3613, radius=7, shapeName=foot_r_capsule, center=<5.1029 -7.0735 0>, rotation=<rot 90 180 -90>}
  {shapeType=Capsule, length=12.1748, radius=9.0432, shapeName=foot_r_capsule, center=<0 2.2446 0>}
 ]
 lowerarm_l [{shapeType=Capsule, length=26.3904, radius=6.1093, shapeName=lowerarm_l_capsule, center=<10.6887 0 0>, rotation=<rot 90 0 0>}]
 lowerarm_r [{shapeType=Capsule, length=26, radius=6, shapeName=lowerarm_r_capsule, center=<-10 0 0>, rotation=<rot 90 0 0>}]
 pelvis [{shapeType=Capsule, length=11.546, radius=12.8138, shapeName=pelvis_capsule}]
 root [
  {
   shapeType=Capsule
   length=89.2237
   radius=17.4796
   shapeName=root_capsule
   center=<-36.9579 1.2052 66.2044>
   rotation=<rot 90 -178.2101 171.1006>
  }
  {
   shapeType=Capsule
   length=80.8476
   radius=15.9492
   shapeName=root_capsule
   center=<36.9579 1.2052 66.2044>
   rotation=<rot 90 178.2101 -171.1006>
  }
 ]
 spine_02 [{shapeType=Capsule, length=14.1596, radius=10.2002, shapeName=spine_02_capsule}]
 spine_03 [{shapeType=Capsule, length=13.5763, radius=11.3042, shapeName=spine_03_capsule}]
 spine_04 [{shapeType=Capsule, length=13.7137, radius=12.1257, shapeName=spine_04_capsule}]
 spine_05 [{shapeType=Capsule, length=21.5542, radius=9.794, shapeName=spine_05_capsule}]
 thigh_l [{shapeType=Capsule, length=33.1221, radius=8.1367, shapeName=thigh_l_capsule, center=<-18.4993 0 0>, rotation=<rot 90 0 0>}]
 thigh_r [{shapeType=Capsule, length=33.1221, radius=8.1367, shapeName=thigh_r_capsule, center=<18.4993 0 0>, rotation=<rot 90 0 0>}]
 upperarm_l [{shapeType=Capsule, length=20.0045, radius=7.0339, shapeName=upperarm_l_capsule, center=<14.4803 0 0>, rotation=<rot 90 0 0>}]
 upperarm_r [{shapeType=Capsule, length=20, radius=7, shapeName=upperarm_r_capsule, center=<-14.5 0 0>, rotation=<rot 90 0 0>}]
}
constraints [
 {bone1Name=cape_01_l, bone2Name=cape_l_MSTR, swing1LimitDegrees=30, swing2LimitDegrees=20, twistLimitDegrees=0}
 {bone1Name=cape_02_l, bone2Name=cape_01_l, swing1LimitDegrees=20, swing2LimitDegrees=25, twistLimitDegrees=0}
 {bone1Name=cape_03_l, bone2Name=cape_02_l, swing1LimitDegrees=20, swing2LimitDegrees=20, twistLimitDegrees=0}
 {bone1Name=cape_04_l, bone2Name=cape_03_l, swing1LimitDegrees=15, swing2LimitDegrees=20, twistLimitDegrees=0}
 {bone1Name=cape_05_l, bone2Name=cape_04_l, swing1LimitDegrees=25, swing2LimitDegrees=25, twistLimitDegrees=0}
 {bone1Name=cape_06_l, bone2Name=cape_05_l, swing1LimitDegrees=25, swing2LimitDegrees=25, twistLimitDegrees=0}
 {bone1Name=cape_mid_01_l, bone2Name=cape_mid_l_MSTR, swing1LimitDegrees=5, swing2LimitDegrees=30, twistLimitDegrees=0}
 {bone1Name=cape_mid_02_l, bone2Name=cape_mid_01_l, swing1LimitDegrees=10, swing2LimitDegrees=22, twistLimitDegrees=0}
 {bone1Name=cape_mid_03_l, bone2Name=cape_mid_02_l, swing1LimitDegrees=10, swing2LimitDegrees=24, twistLimitDegrees=0}
 {bone1Name=cape_mid_04_l, bone2Name=cape_mid_03_l, swing1LimitDegrees=10, swing2LimitDegrees=30, twistLimitDegrees=0}
 {bone1Name=cape_mid_05_l, bone2Name=cape_mid_04_l, swing1LimitDegrees=25, swing2LimitDegrees=25, twistLimitDegrees=0}
 {bone1Name=cape_mid_06_l, bone2Name=cape_mid_05_l, swing1LimitDegrees=25, swing2LimitDegrees=25, twistLimitDegrees=0}
 {bone1Name=cape_mid_01_r, bone2Name=cape_mid_r_MSTR, swing1LimitDegrees=5, swing2LimitDegrees=30, twistLimitDegrees=0}
 {bone1Name=cape_mid_02_r, bone2Name=cape_mid_01_r, swing1LimitDegrees=10, swing2LimitDegrees=22, twistLimitDegrees=0}
 {bone1Name=cape_mid_03_r, bone2Name=cape_mid_02_r, swing1LimitDegrees=10, swing2LimitDegrees=24, twistLimitDegrees=0}
 {bone1Name=cape_mid_04_r, bone2Name=cape_mid_03_r, swing1LimitDegrees=10, swing2LimitDegrees=30, twistLimitDegrees=0}
 {bone1Name=cape_mid_05_r, bone2Name=cape_mid_04_r, swing1LimitDegrees=25, swing2LimitDegrees=25, twistLimitDegrees=0}
 {bone1Name=cape_mid_06_r, bone2Name=cape_mid_05_r, swing1LimitDegrees=25, swing2LimitDegrees=25, twistLimitDegrees=0}
 {bone1Name=cape_01_r, bone2Name=cape_r_MSTR, swing1LimitDegrees=10, swing2LimitDegrees=20, twistLimitDegrees=0}
 {bone1Name=cape_02_r, bone2Name=cape_01_r, swing1LimitDegrees=20, swing2LimitDegrees=20, twistLimitDegrees=0}
 {bone1Name=cape_03_r, bone2Name=cape_02_r, swing1LimitDegrees=20, swing2LimitDegrees=30, twistLimitDegrees=0}
 {bone1Name=cape_04_r, bone2Name=cape_03_r, swing1LimitDegrees=25, swing2LimitDegrees=30, twistLimitDegrees=0}
 {bone1Name=cape_05_r, bone2Name=cape_04_r, swing1LimitDegrees=25, swing2LimitDegrees=25, twistLimitDegrees=0}
 {bone1Name=cape_06_r, bone2Name=cape_05_r, swing1LimitDegrees=25, swing2LimitDegrees=25, twistLimitDegrees=0}
 {
  bone1Name=cape_mid_06_r
  bone2Name=cape_06_r
  swing1LimitDegrees=45
  swing1Motion=Free
  swing2LimitDegrees=45
  swing2Motion=Free
  twistLimitDegrees=45
  twistMotion=Free
 }
 {
  bone1Name=cape_mid_06_l
  bone2Name=cape_06_l
  swing1LimitDegrees=45
  swing1Motion=Free
  swing2LimitDegrees=45
  swing2Motion=Free
  twistLimitDegrees=45
  twistMotion=Free
 }
 {
  bone1Name=cape_mid_06_l
  bone2Name=cape_mid_06_r
  swing1LimitDegrees=45
  swing1Motion=Free
  swing2LimitDegrees=45
  swing2Motion=Free
  twistLimitDegrees=45
  twistMotion=Free
 }
 {
  bone1Name=cape_mid_06_r
  bone2Name=cape_mid_06_l
  swing1LimitDegrees=45
  swing1Motion=Free
  swing2LimitDegrees=45
  swing2Motion=Free
  twistLimitDegrees=45
  twistMotion=Free
 }
]
