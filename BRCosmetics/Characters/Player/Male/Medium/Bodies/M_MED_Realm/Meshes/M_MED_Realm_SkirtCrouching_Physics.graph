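# Physics asset for the M_MED_Realm male character's skirt, a crouching variant by name, covering the coat bones and lower body.
# Seven six-bone dyn_coat capsule chains hang from the pelvis, with narrow swing limits and zero twist limits, tighter at some tips.
# Pelvis, legs, feet and two left cape bones have free angular constraints, and each thigh adds an angled second capsule.
# A huge root sphere-like capsule sits below the character, apparently as a floor collider.
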PhysicsAsset /BRCosmetics/Characters/Player/Male/Medium/Bodies/M_MED_Realm/Meshes/M_MED_Realm_SkirtCrouching_Physics
bodies {
 calf_l [{shapeType=Capsule, length=47.0028, radius=9.9759, shapeName=calf_l_capsule, center=<-19.7926 0 0>, rotation=<rot 90 0 0>}]
 calf_r [{shapeType=Capsule, length=46.9971, radius=9.9816, shapeName=calf_r_capsule, center=<19.7926 0 0>, rotation=<rot -90 180 180>}]
 cape_02_l [{shapeType=Capsule, length=13.6992, radius=6.8039, shapeName=cape_02_l_capsule, center=<13.9703 0 0>, rotation=<rot 90 0 0>}]
 cape_03_l [{shapeType=Capsule, length=8.5899, radius=6.7153, shapeName=cape_03_l_capsule, center=<13.9703 0 0>, rotation=<rot 90 0 0>}]
 dyn_coat_bk_c_01 [{shapeType=Capsule, length=4.6194, radius=10.1684, shapeName=dyn_coat_bk_c_01_capsule, center=<7.4311 0 12>, rotation=<rot 90 0 0>}]
 dyn_coat_bk_c_02 [{shapeType=Capsule, length=4.2577, radius=10.53, shapeName=dyn_coat_bk_c_02_capsule, center=<7.4311 0 12>, rotation=<rot 90 0 0>}]
 dyn_coat_bk_c_03 [{shapeType=Capsule, length=8.2, radius=11.1173, shapeName=dyn_coat_bk_c_03_capsule, center=<9.2473 0 12>, rotation=<rot 90 0 0>}]
 dyn_coat_bk_c_04 [{
   shapeType=Capsule
   length=23.9667
   radius=12.9482
   shapeName=dyn_coat_bk_c_04_capsule
   center=<7.4311 0 12>
   rotation=<rot 90 153.435 -116.565>
  }]
 dyn_coat_bk_c_05 [{
   shapeType=Capsule
   length=23.9667
   radius=12.9482
   shapeName=dyn_coat_bk_c_05_capsule
   center=<7.4311 0 12>
   rotation=<rot 90 153.435 -116.565>
  }]
 dyn_coat_bk_c_06 [{
   shapeType=Capsule
   length=23.9667
   radius=12.9482
   shapeName=dyn_coat_bk_c_05_capsule
   center=<7.4311 0 12>
   rotation=<rot 90 153.435 -116.565>
  }]
 dyn_coat_bk_l_01 [{shapeType=Capsule, length=4.6194, radius=10.1684, shapeName=dyn_coat_bk_l_01_capsule, center=<7.4311 0 12>, rotation=<rot 90 0 0>}]
 dyn_coat_bk_l_02 [{shapeType=Capsule, length=4.2577, radius=10.53, shapeName=dyn_coat_bk_l_02_capsule, center=<7.4311 0 12>, rotation=<rot 90 0 0>}]
 dyn_coat_bk_l_03 [{shapeType=Capsule, length=8.2, radius=11.1173, shapeName=dyn_coat_bk_l_03_capsule, center=<9.2473 0 12>, rotation=<rot 90 0 0>}]
 dyn_coat_bk_l_04 [{
   shapeType=Capsule
   length=23.9667
   radius=12.9482
   shapeName=dyn_coat_bk_l_04_capsule
   center=<7.4311 0 12>
   rotation=<rot 90 153.435 -116.565>
  }]
 dyn_coat_bk_l_05 [{
   shapeType=Capsule
   length=23.9667
   radius=12.9482
   shapeName=dyn_coat_bk_l_05_capsule
   center=<7.4311 0 12>
   rotation=<rot 90 153.435 -116.565>
  }]
 dyn_coat_bk_l_06 [{
   shapeType=Capsule
   length=23.9667
   radius=12.9482
   shapeName=dyn_coat_bk_l_05_capsule
   center=<7.4311 0 12>
   rotation=<rot 90 153.435 -116.565>
  }]
 dyn_coat_bk_r_01 [{shapeType=Capsule, length=4.6194, radius=10.1684, shapeName=dyn_coat_bk_r_01_capsule, center=<-7.4311 0 -12>, rotation=<rot 90 0 0>}]
 dyn_coat_bk_r_02 [{shapeType=Capsule, length=4.2577, radius=10.53, shapeName=dyn_coat_bk_r_02_capsule, center=<-7.4311 0 -12>, rotation=<rot 90 0 0>}]
 dyn_coat_bk_r_03 [{shapeType=Capsule, length=8.2, radius=11.1173, shapeName=dyn_coat_bk_r_03_capsule, center=<-9.2473 0 -12>, rotation=<rot 90 0 0>}]
 dyn_coat_bk_r_04 [{shapeType=Capsule, length=23.9667, radius=12.9482, shapeName=dyn_coat_bk_r_04_capsule, center=<-7.4311 0 -12>, rotation=<rot 90 90 0>}]
 dyn_coat_bk_r_05 [{shapeType=Capsule, length=23.9667, radius=12.9482, shapeName=dyn_coat_bk_r_05_capsule, center=<-7.4311 0 -12>, rotation=<rot 90 90 0>}]
 dyn_coat_bk_r_06 [{shapeType=Capsule, length=26.6205, radius=13.3803, shapeName=dyn_coat_bk_r_06_capsule, center=<-9.507 0 -12>, rotation=<rot 90 90 0>}]
 dyn_coat_front_l_01 [{shapeType=Capsule, length=0, radius=11.6594, shapeName=dyn_coat_front_l_01_capsule, center=<5.5557 0 12>, rotation=<rot 90 0 0>}]
 dyn_coat_front_l_02 [{shapeType=Capsule, length=5.7139, radius=11.5884, shapeName=dyn_coat_front_l_02_capsule, center=<7.6896 0 11.5>, rotation=<rot 90 0 0>}]
 dyn_coat_front_l_03 [{shapeType=Capsule, length=10.563, radius=11.322, shapeName=dyn_coat_front_l_03_capsule, center=<9.2473 0 11.5>, rotation=<rot 90 0 0>}]
 dyn_coat_front_l_04 [{shapeType=Capsule, length=15.3805, radius=10.286, shapeName=dyn_coat_front_l_04_capsule, center=<7.4311 0 9.6>, rotation=<rot 90 0 0>}]
 dyn_coat_front_l_05 [{shapeType=Capsule, length=9.052, radius=13.3803, shapeName=dyn_coat_front_l_05_capsule, center=<9.507 0 12>, rotation=<rot 90 0 0>}]
 dyn_coat_front_l_06 [{shapeType=Capsule, length=11.302, radius=13.3803, shapeName=dyn_coat_front_l_05_capsule, center=<9.507 0 12>, rotation=<rot 90 0 0>}]
 dyn_coat_front_r_01 [{shapeType=Capsule, length=9.4492, radius=9.2733, shapeName=dyn_coat_front_r_01_capsule, center=<-5.5123 0 -10>, rotation=<rot 90 0 0>}]
 dyn_coat_front_r_02 [{shapeType=Capsule, length=8.7787, radius=9.8873, shapeName=dyn_coat_front_r_02_capsule, center=<-7.916 0 -10>, rotation=<rot 90 0 0>}]
 dyn_coat_front_r_03 [{shapeType=Capsule, length=11.0159, radius=10.8691, shapeName=dyn_coat_front_r_03_capsule, center=<-9.2473 0 -10>, rotation=<rot 90 0 0>}]
 dyn_coat_front_r_04 [{shapeType=Capsule, length=15.1143, radius=10.5522, shapeName=dyn_coat_front_r_04_capsule, center=<-7.4311 0 -10>, rotation=<rot 90 0 0>}]
 dyn_coat_front_r_05 [{shapeType=Capsule, length=11.3231, radius=11.8329, shapeName=dyn_coat_front_r_05_capsule, center=<-9.507 0 -10>, rotation=<rot 90 0 0>}]
 dyn_coat_front_r_06 [{shapeType=Capsule, length=9.052, radius=13.3803, shapeName=dyn_coat_front_r_05_capsule, center=<-9.507 0 -10>, rotation=<rot 90 0 0>}]
 dyn_coat_l_01 [{
   shapeType=Capsule
   length=10.1006
   radius=4.6872
   shapeName=dyn_coat_l_01_capsule
   center=<6.7136 0 4.6>
   rotation=<rot 82.3252 -0.0004 -0.0007>
  }]
 dyn_coat_l_02 [{
   shapeType=Capsule
   length=14.8277
   radius=4.1854
   shapeName=dyn_coat_l_02_capsule
   center=<12.8193 6 2.5>
   rotation=<rot 88.5458 -0.0011 175.7232>
  }]
 dyn_coat_l_03 [{shapeType=Capsule, length=16.2489, radius=5.6361, shapeName=dyn_coat_l_03_capsule, center=<9.2473 0 3.8>, rotation=<rot 90 0 0>}]
 dyn_coat_l_04 [{shapeType=Capsule, length=12.7182, radius=12.9482, shapeName=dyn_coat_l_04_capsule, center=<7.4311 0 12>, rotation=<rot 90 153.435 -116.565>}]
 dyn_coat_l_05 [{shapeType=Capsule, length=12.0641, radius=13.3803, shapeName=dyn_coat_l_05_capsule, center=<9.507 0 12>, rotation=<rot 90 180 -90>}]
 dyn_coat_l_06 [{shapeType=Capsule, length=12.0641, radius=13.3803, shapeName=dyn_coat_l_05_capsule, center=<9.507 0 12>, rotation=<rot 90 180 -90>}]
 dyn_coat_r_01 [{shapeType=Capsule, length=4.6194, radius=10.1684, shapeName=dyn_coat_r_01_capsule, center=<-7.4311 0 -11>, rotation=<rot 90 0 0>}]
 dyn_coat_r_02 [{shapeType=Capsule, length=16.7638, radius=11.1173, shapeName=dyn_coat_r_02_capsule, center=<-12.8193 0 -11>, rotation=<rot 90 0 0>}]
 dyn_coat_r_03 [{shapeType=Capsule, length=10.7677, radius=11.1173, shapeName=dyn_coat_r_03_capsule, center=<-9.2473 0 -11>, rotation=<rot 90 0 0>}]
 dyn_coat_r_04 [{shapeType=Capsule, length=12.7182, radius=12.9482, shapeName=dyn_coat_r_04_capsule, center=<-7.4311 0 -11>, rotation=<rot 90 0 0>}]
 dyn_coat_r_05 [{shapeType=Capsule, length=12.0641, radius=13.3803, shapeName=dyn_coat_r_05_capsule, center=<-9.507 0 -11>, rotation=<rot 90 0 0>}]
 dyn_coat_r_06 [{shapeType=Capsule, length=12.0641, radius=13.3803, shapeName=dyn_coat_r_05_capsule, center=<-9.507 0 -11>, rotation=<rot 90 0 0>}]
 foot_l [{shapeType=Capsule, length=21.0099, radius=6.7017, shapeName=foot_l_capsule, center=<-4.1587 6.9986 0>, rotation=<rot 0 0 90>}]
 foot_r [{shapeType=Capsule, length=21.0099, radius=6.7017, shapeName=foot_r_capsule, center=<5.6683 -6.9255 0>, rotation=<rot 0 0 90>}]
 pelvis [{shapeType=Capsule, length=11.6259, radius=11.3469, shapeName=pelvis_capsule, center=<0 -0.5135 0>}]
 root [{shapeType=Capsule, length=0, radius=400, shapeName=root_capsule, center=<14.6126 -37.398 -387.9122>, rotation=<rot 0 -90 0>}]
 thigh_l [
  {shapeType=Capsule, length=40.6784, radius=9.4565, shapeName=thigh_l_capsule, center=<-13.9616 0 0>, rotation=<rot 90 0 0>}
  {
   shapeType=Capsule
   length=34.8442
   radius=8.1663
   shapeName=thigh_l_capsule
   center=<-16.2074 13.5712 8.5012>
   rotation=<rot 0 -0.0001 78.3189>
  }
 ]
 thigh_r [
  {shapeType=Capsule, length=40.7825, radius=9.3525, shapeName=thigh_l_capsule, center=<13.9616 0 0>, rotation=<rot 90 0 0>}
  {
   shapeType=Capsule
   length=21.8337
   radius=8.1663
   shapeName=thigh_l_capsule
   center=<16.2074 -7.4999 -7.246>
   rotation=<rot 0 -0.0001 78.3189>
  }
 ]
}
constraints [
 {bone1Name=dyn_coat_front_l_02, bone2Name=dyn_coat_front_l_01, swing1LimitDegrees=30, swing2LimitDegrees=30, twistLimitDegrees=0}
 {bone1Name=dyn_coat_front_l_03, bone2Name=dyn_coat_front_l_02, swing1LimitDegrees=30, swing2LimitDegrees=30, twistLimitDegrees=0}
 {bone1Name=dyn_coat_front_l_04, bone2Name=dyn_coat_front_l_03, swing1LimitDegrees=30, swing2LimitDegrees=30, twistLimitDegrees=0}
 {bone1Name=dyn_coat_front_l_05, bone2Name=dyn_coat_front_l_04, swing1LimitDegrees=30, swing2LimitDegrees=30, twistLimitDegrees=0}
 {bone1Name=dyn_coat_l_02, bone2Name=dyn_coat_l_01, swing1LimitDegrees=30, swing2LimitDegrees=30, twistLimitDegrees=0}
 {bone1Name=dyn_coat_l_03, bone2Name=dyn_coat_l_02, swing1LimitDegrees=30, swing2LimitDegrees=30, twistLimitDegrees=0}
 {bone1Name=dyn_coat_l_04, bone2Name=dyn_coat_l_03, swing1LimitDegrees=30, swing2LimitDegrees=30, twistLimitDegrees=0}
 {bone1Name=dyn_coat_l_05, bone2Name=dyn_coat_l_04, swing1LimitDegrees=30, swing2LimitDegrees=30, twistLimitDegrees=0}
 {bone1Name=dyn_coat_bk_l_02, bone2Name=dyn_coat_bk_l_01, swing1LimitDegrees=30, swing2LimitDegrees=30, twistLimitDegrees=0}
 {bone1Name=dyn_coat_bk_l_03, bone2Name=dyn_coat_bk_l_02, swing1LimitDegrees=30, swing2LimitDegrees=30, twistLimitDegrees=0}
 {bone1Name=dyn_coat_bk_l_04, bone2Name=dyn_coat_bk_l_03, swing1LimitDegrees=30, swing2LimitDegrees=30, twistLimitDegrees=0}
 {bone1Name=dyn_coat_bk_l_05, bone2Name=dyn_coat_bk_l_04, swing1LimitDegrees=30, swing2LimitDegrees=30, twistLimitDegrees=0}
 {bone1Name=dyn_coat_bk_l_06, bone2Name=dyn_coat_bk_l_05, swing1LimitDegrees=30, swing2LimitDegrees=30, twistLimitDegrees=0}
 {bone1Name=dyn_coat_bk_c_02, bone2Name=dyn_coat_bk_c_01, swing1LimitDegrees=30, swing2LimitDegrees=30, twistLimitDegrees=0}
 {bone1Name=dyn_coat_bk_c_03, bone2Name=dyn_coat_bk_c_02, swing1LimitDegrees=30, swing2LimitDegrees=30, twistLimitDegrees=0}
 {bone1Name=dyn_coat_bk_c_04, bone2Name=dyn_coat_bk_c_03, swing1LimitDegrees=30, swing2LimitDegrees=30, twistLimitDegrees=0}
 {bone1Name=dyn_coat_bk_c_05, bone2Name=dyn_coat_bk_c_04, swing1LimitDegrees=30, swing2LimitDegrees=30, twistLimitDegrees=0}
 {bone1Name=dyn_coat_bk_c_06, bone2Name=dyn_coat_bk_c_05, swing1LimitDegrees=30, swing2LimitDegrees=30, twistLimitDegrees=0}
 {bone1Name=dyn_coat_front_r_02, bone2Name=dyn_coat_front_r_01, swing1LimitDegrees=30, swing2LimitDegrees=30, twistLimitDegrees=0}
 {bone1Name=dyn_coat_front_r_03, bone2Name=dyn_coat_front_r_02, swing1LimitDegrees=30, swing2LimitDegrees=30, twistLimitDegrees=0}
 {bone1Name=dyn_coat_front_r_04, bone2Name=dyn_coat_front_r_03, swing1LimitDegrees=30, swing2LimitDegrees=30, twistLimitDegrees=0}
 {bone1Name=dyn_coat_front_r_05, bone2Name=dyn_coat_front_r_04, swing1LimitDegrees=30, swing2LimitDegrees=30, twistLimitDegrees=0}
 {bone1Name=dyn_coat_r_02, bone2Name=dyn_coat_r_01, swing1LimitDegrees=30, swing2LimitDegrees=30, twistLimitDegrees=0}
 {bone1Name=dyn_coat_r_03, bone2Name=dyn_coat_r_02, swing1LimitDegrees=30, swing2LimitDegrees=30, twistLimitDegrees=0}
 {bone1Name=dyn_coat_r_04, bone2Name=dyn_coat_r_03, swing1LimitDegrees=30, swing2LimitDegrees=30, twistLimitDegrees=0}
 {bone1Name=dyn_coat_r_05, bone2Name=dyn_coat_r_04, swing1LimitDegrees=30, swing2LimitDegrees=30, twistLimitDegrees=0}
 {bone1Name=dyn_coat_bk_r_02, bone2Name=dyn_coat_bk_r_01, swing1LimitDegrees=30, swing2LimitDegrees=30, twistLimitDegrees=0}
 {bone1Name=dyn_coat_bk_r_03, bone2Name=dyn_coat_bk_r_02, swing1LimitDegrees=30, swing2LimitDegrees=30, twistLimitDegrees=0}
 {bone1Name=dyn_coat_bk_r_04, bone2Name=dyn_coat_bk_r_03, swing1LimitDegrees=30, swing2LimitDegrees=30, twistLimitDegrees=0}
 {bone1Name=dyn_coat_bk_r_05, bone2Name=dyn_coat_bk_r_04, swing1LimitDegrees=30, swing2LimitDegrees=30, twistLimitDegrees=0}
 {bone1Name=dyn_coat_bk_r_06, bone2Name=dyn_coat_bk_r_05, swing1LimitDegrees=30, swing2LimitDegrees=30, twistLimitDegrees=0}
 {
  bone1Name=thigh_l
  bone2Name=pelvis
  swing1LimitDegrees=45
  swing1Motion=Free
  swing2LimitDegrees=45
  swing2Motion=Free
  twistLimitDegrees=45
  twistMotion=Free
 }
 {
  bone1Name=thigh_r
  bone2Name=pelvis
  swing1LimitDegrees=45
  swing1Motion=Free
  swing2LimitDegrees=45
  swing2Motion=Free
  twistLimitDegrees=45
  twistMotion=Free
 }
 {bone1Name=dyn_coat_front_l_01, bone2Name=pelvis, swing1LimitDegrees=30, swing2LimitDegrees=30, twistLimitDegrees=0}
 {bone1Name=dyn_coat_l_01, bone2Name=pelvis, swing1LimitDegrees=30, swing2LimitDegrees=30, twistLimitDegrees=0}
 {bone1Name=dyn_coat_bk_l_01, bone2Name=pelvis, swing1LimitDegrees=30, swing2LimitDegrees=30, twistLimitDegrees=0}
 {bone1Name=dyn_coat_bk_c_01, bone2Name=pelvis, swing1LimitDegrees=30, swing2LimitDegrees=30, twistLimitDegrees=0}
 {bone1Name=dyn_coat_front_r_01, bone2Name=pelvis, swing1LimitDegrees=30, swing2LimitDegrees=30, twistLimitDegrees=0}
 {bone1Name=dyn_coat_r_01, bone2Name=pelvis, swing1LimitDegrees=30, swing2LimitDegrees=30, twistLimitDegrees=0}
 {bone1Name=dyn_coat_bk_r_01, bone2Name=pelvis, swing1LimitDegrees=30, swing2LimitDegrees=30, twistLimitDegrees=0}
 {
  bone1Name=calf_l
  bone2Name=thigh_l
  swing1LimitDegrees=45
  swing1Motion=Free
  swing2LimitDegrees=45
  swing2Motion=Free
  twistLimitDegrees=45
  twistMotion=Free
 }
 {
  bone1Name=calf_r
  bone2Name=thigh_r
  swing1LimitDegrees=45
  swing1Motion=Free
  swing2LimitDegrees=45
  swing2Motion=Free
  twistLimitDegrees=45
  twistMotion=Free
 }
 {
  bone1Name=pelvis
  bone2Name=root
  swing1LimitDegrees=45
  swing1Motion=Free
  swing2LimitDegrees=45
  swing2Motion=Free
  twistLimitDegrees=45
  twistMotion=Free
 }
 {bone1Name=dyn_coat_r_06, bone2Name=dyn_coat_r_05, swing1LimitDegrees=15, swing2LimitDegrees=15, twistLimitDegrees=0}
 {bone1Name=dyn_coat_front_r_06, bone2Name=dyn_coat_front_r_05, swing1LimitDegrees=15, swing2LimitDegrees=15, twistLimitDegrees=0}
 {bone1Name=dyn_coat_l_06, bone2Name=dyn_coat_l_05, swing1LimitDegrees=15, swing2LimitDegrees=15, twistLimitDegrees=0}
 {bone1Name=dyn_coat_front_l_06, bone2Name=dyn_coat_front_l_05, swing1LimitDegrees=15, swing2LimitDegrees=15, twistLimitDegrees=0}
 {
  bone1Name=foot_l
  bone2Name=calf_l
  swing1LimitDegrees=45
  swing1Motion=Free
  swing2LimitDegrees=45
  swing2Motion=Free
  twistLimitDegrees=45
  twistMotion=Free
 }
 {
  bone1Name=foot_r
  bone2Name=calf_r
  swing1LimitDegrees=45
  swing1Motion=Free
  swing2LimitDegrees=45
  swing2Motion=Free
  twistLimitDegrees=45
  twistMotion=Free
 }
 {
  bone1Name=cape_03_l
  bone2Name=cape_02_l
  swing1LimitDegrees=45
  swing1Motion=Free
  swing2LimitDegrees=45
  swing2Motion=Free
  twistLimitDegrees=45
  twistMotion=Free
 }
 {
  bone1Name=cape_02_l
  bone2Name=pelvis
  swing1LimitDegrees=45
  swing1Motion=Free
  swing2LimitDegrees=45
  swing2Motion=Free
  twistLimitDegrees=45
  twistMotion=Free
 }
]
