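# Physics asset for the Realm medium male body's skirt, apparently for the standing front-end (menu) character, simulating its coat bones against the lower body.
# Seven six-bone coat chains hang from the pelvis, mostly swing-limited with zero twist; the right side is tighter and the front-left tip is free.
# Capsules on pelvis, legs, feet and two left cape bones use free angular constraints.
# Some lower right-side coat bones carry second, crossed capsules for extra coverage.
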PhysicsAsset /BRCosmetics/Characters/Player/Male/Medium/Bodies/M_MED_Realm/Meshes/M_MED_Realm_SkirtStandingFrontEnd_Physics
bodies {
 calf_l [{shapeType=Capsule, length=41.7528, radius=6.9759, shapeName=calf_l_capsule, center=<-19.7926 -1.7806 0>, rotation=<rot 90 0 0>}]
 calf_r [{shapeType=Capsule, length=39.0028, radius=9.7259, shapeName=calf_l_capsule, center=<19.7926 1.7806 0>, rotation=<rot 90 0 180>}]
 cape_02_l [{shapeType=Capsule, length=13.6992, radius=6.8039, shapeName=cape_02_l_capsule, center=<13.9703 0 0>, rotation=<rot 90 0 0>}]
 cape_03_l [{shapeType=Capsule, length=8.5899, radius=6.7153, shapeName=cape_03_l_capsule, center=<13.9703 0 0>, rotation=<rot 90 0 0>}]
 dyn_coat_bk_c_01 [{shapeType=Capsule, length=4.6194, radius=10.1684, shapeName=dyn_coat_bk_c_01_capsule, center=<7.4311 0 12>, rotation=<rot 90 0 0>}]
 dyn_coat_bk_c_02 [{shapeType=Capsule, length=4.2577, radius=10.53, shapeName=dyn_coat_bk_c_02_capsule, center=<7.4311 0 12>, rotation=<rot 90 0 0>}]
 dyn_coat_bk_c_03 [{shapeType=Capsule, length=8.2, radius=11.1173, shapeName=dyn_coat_bk_c_03_capsule, center=<9.2473 0 12>, rotation=<rot 90 0 0>}]
 dyn_coat_bk_c_04 [{
   shapeType=Capsule
   length=23.9667
   radius=12.9482
   shapeName=dyn_coat_bk_c_04_capsule
   center=<7.4311 0 12>
   rotation=<rot 90 153.435 -116.565>
  }]
 dyn_coat_bk_c_05 [{
   shapeType=Capsule
   length=27.9667
   radius=15.9482
   shapeName=dyn_coat_bk_c_05_capsule
   center=<7.4311 0 12>
   rotation=<rot 90 153.435 -116.565>
  }]
 dyn_coat_bk_c_06 [{
   shapeType=Capsule
   length=27.9667
   radius=15.9482
   shapeName=dyn_coat_bk_c_05_capsule
   center=<7.4311 0 12>
   rotation=<rot 90 153.435 -116.565>
  }]
 dyn_coat_bk_l_01 [{shapeType=Capsule, length=4.6194, radius=10.1684, shapeName=dyn_coat_bk_l_01_capsule, center=<7.4311 0 12>, rotation=<rot 90 0 0>}]
 dyn_coat_bk_l_02 [{shapeType=Capsule, length=4.2577, radius=10.53, shapeName=dyn_coat_bk_l_02_capsule, center=<7.4311 0 12>, rotation=<rot 90 0 0>}]
 dyn_coat_bk_l_03 [{shapeType=Capsule, length=8.2, radius=11.1173, shapeName=dyn_coat_bk_l_03_capsule, center=<9.2473 0 12>, rotation=<rot 90 0 0>}]
 dyn_coat_bk_l_04 [{
   shapeType=Capsule
   length=23.9667
   radius=12.9482
   shapeName=dyn_coat_bk_l_04_capsule
   center=<7.4311 0 12>
   rotation=<rot 90 153.435 -116.565>
  }]
 dyn_coat_bk_l_05 [{
   shapeType=Capsule
   length=22.7167
   radius=14.1982
   shapeName=dyn_coat_bk_l_05_capsule
   center=<7.4311 0 12>
   rotation=<rot 90 153.435 -116.565>
  }]
 dyn_coat_bk_l_06 [{
   shapeType=Capsule
   length=20.7167
   radius=16.1982
   shapeName=dyn_coat_bk_l_05_capsule
   center=<7.4311 0 12>
   rotation=<rot 90 153.435 -116.565>
  }]
 dyn_coat_bk_r_01 [{shapeType=Capsule, length=4.6194, radius=10.1684, shapeName=dyn_coat_bk_r_01_capsule, center=<-7.4311 0 -12>, rotation=<rot 90 0 0>}]
 dyn_coat_bk_r_02 [{shapeType=Capsule, length=4.2577, radius=10.53, shapeName=dyn_coat_bk_r_02_capsule, center=<-7.4311 0 -12>, rotation=<rot 90 0 0>}]
 dyn_coat_bk_r_03 [{shapeType=Capsule, length=14.95, radius=11.1173, shapeName=dyn_coat_bk_r_03_capsule, center=<-9.2473 0 -12>, rotation=<rot 90 0 0>}]
 dyn_coat_bk_r_04 [{shapeType=Capsule, length=30.2167, radius=13.4482, shapeName=dyn_coat_bk_r_04_capsule, center=<-7.4311 0 -12>, rotation=<rot 90 90 0>}]
 dyn_coat_bk_r_05 [
  {
   shapeType=Capsule
   length=33.4667
   radius=15.1982
   shapeName=dyn_coat_bk_r_05_capsule
   center=<-7.4311 0 -12>
   rotation=<rot 88.1428 89.9983 -0.0018>
  }
  {shapeType=Capsule, length=39, radius=15, shapeName=dyn_coat_bk_r_05_capsule, center=<0 0 -12>, rotation=<rot -0.0001 0 108.309>}
 ]
 dyn_coat_bk_r_06 [
  {
   shapeType=Capsule
   length=34.3705
   radius=17.3803
   shapeName=dyn_coat_bk_r_06_capsule
   center=<-9.507 0 -12>
   rotation=<rot 88.1428 89.9983 -0.0017>
  }
  {shapeType=Capsule, length=39, radius=15, shapeName=dyn_coat_bk_r_06_capsule, center=<0 0 -12>, rotation=<rot -0.0001 0 108.309>}
 ]
 dyn_coat_front_l_01 [{shapeType=Capsule, length=1.6952, radius=9.9642, shapeName=dyn_coat_front_l_01_capsule, center=<5.5557 0 10>, rotation=<rot 90 0 0>}]
 dyn_coat_front_l_02 [{shapeType=Capsule, length=8.3024, radius=9, shapeName=dyn_coat_front_l_02_capsule, center=<7.6896 0 9>, rotation=<rot 0 0 90>}]
 dyn_coat_front_l_03 [{shapeType=Capsule, length=12.0977, radius=9.7873, shapeName=dyn_coat_front_l_03_capsule, center=<9.2473 0 9>, rotation=<rot 0 0 90>}]
 dyn_coat_front_l_04 [{shapeType=Capsule, length=15.3805, radius=10.286, shapeName=dyn_coat_front_l_04_capsule, center=<7.4311 0 9.6>, rotation=<rot 90 0 0>}]
 dyn_coat_front_l_05 [{shapeType=Capsule, length=9.052, radius=13.3803, shapeName=dyn_coat_front_l_05_capsule, center=<9.507 0 12>, rotation=<rot 90 0 0>}]
 dyn_coat_front_l_06 [{shapeType=Capsule, length=6.25, radius=8, shapeName=dyn_coat_front_l_05_capsule, center=<7.5 0 5.8>, rotation=<rot 90 0 0>}]
 dyn_coat_front_r_01 [{
   shapeType=Capsule
   length=9.4492
   radius=9.2733
   shapeName=dyn_coat_front_r_01_capsule
   center=<-5.5123 0 -10>
   rotation=<rot 77.1151 -90.0008 -0.001>
  }]
 dyn_coat_front_r_02 [{
   shapeType=Capsule
   length=21.2787
   radius=8.3873
   shapeName=dyn_coat_front_r_02_capsule
   center=<-7.916 0 -7>
   rotation=<rot 77.1151 -90.0008 -0.001>
  }]
 dyn_coat_front_r_03 [{
   shapeType=Capsule
   length=23.0159
   radius=9.8691
   shapeName=dyn_coat_front_r_03_capsule
   center=<-9.2473 0 -7>
   rotation=<rot 77.1151 -90.0008 -0.001>
  }]
 dyn_coat_front_r_04 [{
   shapeType=Capsule
   length=26.0194
   radius=9.5522
   shapeName=dyn_coat_front_r_04_capsule
   center=<-7.4311 0 -7>
   rotation=<rot 77.1151 -90.0008 -0.001>
  }]
 dyn_coat_front_r_05 [{
   shapeType=Capsule
   length=22.0748
   radius=11.8329
   shapeName=dyn_coat_front_r_05_capsule
   center=<-9.507 0 -7>
   rotation=<rot 77.1151 -90.0008 -0.001>
  }]
 dyn_coat_front_r_06 [{
   shapeType=Capsule
   length=26.1176
   radius=8.8803
   shapeName=dyn_coat_front_r_05_capsule
   center=<-3.0155 0 -7>
   rotation=<rot 77.1151 -90.0008 -0.001>
  }]
 dyn_coat_l_01 [{
   shapeType=Capsule
   length=10.1006
   radius=4.6872
   shapeName=dyn_coat_l_01_capsule
   center=<6.7136 0 4.6>
   rotation=<rot 82.3252 -0.0004 -0.0007>
  }]
 dyn_coat_l_02 [{
   shapeType=Capsule
   length=13.3277
   radius=3.4354
   shapeName=dyn_coat_l_02_capsule
   center=<7.8952 -0.3683 2.625>
   rotation=<rot 88.5458 -0.0011 175.7232>
  }]
 dyn_coat_l_03 [{shapeType=Capsule, length=9.2489, radius=12.6361, shapeName=dyn_coat_l_03_capsule, center=<9.2473 0 11.3>, rotation=<rot 0 0 90>}]
 dyn_coat_l_04 [{shapeType=Capsule, length=12.4682, radius=13.1982, shapeName=dyn_coat_l_04_capsule, center=<7.4311 0 12>, rotation=<rot 90 153.435 -116.565>}]
 dyn_coat_l_05 [{
   shapeType=Capsule
   length=22.3141
   radius=14.1303
   shapeName=dyn_coat_l_05_capsule
   center=<9.507 0 12>
   rotation=<rot 80.2707 -89.9999 0.0001>
  }]
 dyn_coat_l_06 [{
   shapeType=Capsule
   length=22.3141
   radius=14.1303
   shapeName=dyn_coat_l_05_capsule
   center=<9.507 0 12>
   rotation=<rot 80.2707 -89.9999 0.0001>
  }]
 dyn_coat_r_01 [{shapeType=Capsule, length=4.6194, radius=10.1684, shapeName=dyn_coat_r_01_capsule, center=<-7.4311 0 -11>, rotation=<rot 90 0 0>}]
 dyn_coat_r_02 [{
   shapeType=Capsule
   length=25.0138
   radius=13.8673
   shapeName=dyn_coat_r_02_capsule
   center=<-12.8193 0 -11>
   rotation=<rot 79.3851 90 179.9996>
  }]
 dyn_coat_r_03 [{
   shapeType=Capsule
   length=27.7677
   radius=12.3673
   shapeName=dyn_coat_r_03_capsule
   center=<-9.2473 0 -8>
   rotation=<rot 79.3851 90 179.9996>
  }]
 dyn_coat_r_04 [
  {
   shapeType=Capsule
   length=25.6233
   radius=11.4482
   shapeName=dyn_coat_r_04_capsule
   center=<-7.4311 0.513 -8>
   rotation=<rot 83.4874 -89.9996 0.0002>
  }
  {shapeType=Capsule, length=25.5, radius=11, shapeName=dyn_coat_r_04_capsule, center=<-11.7399 0 -8>, rotation=<rot 0 0.0001 101>}
 ]
 dyn_coat_r_05 [
  {
   shapeType=Capsule
   length=30.3224
   radius=12.1303
   shapeName=dyn_coat_r_05_capsule
   center=<-9.507 0.4927 -8>
   rotation=<rot 76.0501 -90.0164 0>
  }
  {shapeType=Capsule, length=24.25, radius=12.25, shapeName=dyn_coat_r_05_capsule, center=<-11.7399 0 -8>, rotation=<rot 0 0.0001 101.9966>}
 ]
 dyn_coat_r_06 [
  {
   shapeType=Capsule
   length=32.1297
   radius=12.3803
   shapeName=dyn_coat_r_05_capsule
   center=<-9.507 0.5091 -8>
   rotation=<rot 76.4895 -90.0054 0>
  }
  {shapeType=Capsule, length=24.25, radius=12.25, shapeName=dyn_coat_r_06_capsule, center=<-11.7399 0 -8>, rotation=<rot 0 0.0001 101.9966>}
 ]
 foot_l [{shapeType=Capsule, length=22.5099, radius=5.2017, shapeName=foot_l_capsule, center=<-5.1364 6.7816 0>, rotation=<rot 0 12.5135 90>}]
 foot_r [{shapeType=Capsule, length=21.0099, radius=6.7017, shapeName=foot_r_capsule, center=<5.6683 -6.9255 0>, rotation=<rot 0 0 90>}]
 pelvis [{shapeType=Capsule, length=11.6259, radius=11.3469, shapeName=pelvis_capsule, center=<0 -0.5135 0>}]
 thigh_l [{shapeType=Capsule, length=41.9284, radius=8.2065, shapeName=thigh_l_capsule, center=<-13.9616 0 0>, rotation=<rot 90 0 0>}]
 thigh_r [{shapeType=Capsule, length=41.5325, radius=8.6025, shapeName=thigh_l_capsule, center=<13.9616 0 0>, rotation=<rot 90 0 0>}]
}
constraints [
 {bone1Name=dyn_coat_front_l_02, bone2Name=dyn_coat_front_l_01, swing1LimitDegrees=30, swing2LimitDegrees=30, twistLimitDegrees=0}
 {bone1Name=dyn_coat_front_l_03, bone2Name=dyn_coat_front_l_02, swing1LimitDegrees=30, swing2LimitDegrees=30, twistLimitDegrees=0}
 {bone1Name=dyn_coat_front_l_04, bone2Name=dyn_coat_front_l_03, swing1LimitDegrees=30, swing2LimitDegrees=30, twistLimitDegrees=0}
 {bone1Name=dyn_coat_front_l_05, bone2Name=dyn_coat_front_l_04, swing1LimitDegrees=30, swing2LimitDegrees=30, twistLimitDegrees=0}
 {bone1Name=dyn_coat_l_02, bone2Name=dyn_coat_l_01, swing1LimitDegrees=30, swing2LimitDegrees=30, twistLimitDegrees=0}
 {bone1Name=dyn_coat_l_03, bone2Name=dyn_coat_l_02, swing1LimitDegrees=30, swing2LimitDegrees=30, twistLimitDegrees=0}
 {bone1Name=dyn_coat_l_04, bone2Name=dyn_coat_l_03, swing1LimitDegrees=30, swing2LimitDegrees=30, twistLimitDegrees=0}
 {bone1Name=dyn_coat_l_05, bone2Name=dyn_coat_l_04, swing1LimitDegrees=30, swing2LimitDegrees=30, twistLimitDegrees=0}
 {bone1Name=dyn_coat_bk_l_02, bone2Name=dyn_coat_bk_l_01, swing1LimitDegrees=30, swing2LimitDegrees=30, twistLimitDegrees=0}
 {bone1Name=dyn_coat_bk_l_03, bone2Name=dyn_coat_bk_l_02, swing1LimitDegrees=30, swing2LimitDegrees=30, twistLimitDegrees=0}
 {bone1Name=dyn_coat_bk_l_04, bone2Name=dyn_coat_bk_l_03, swing1LimitDegrees=30, swing2LimitDegrees=30, twistLimitDegrees=0}
 {bone1Name=dyn_coat_bk_l_05, bone2Name=dyn_coat_bk_l_04, swing1LimitDegrees=30, swing2LimitDegrees=30, twistLimitDegrees=0}
 {bone1Name=dyn_coat_bk_l_06, bone2Name=dyn_coat_bk_l_05, swing1LimitDegrees=30, swing2LimitDegrees=30, twistLimitDegrees=0}
 {bone1Name=dyn_coat_bk_c_02, bone2Name=dyn_coat_bk_c_01, swing1LimitDegrees=30, swing2LimitDegrees=30, twistLimitDegrees=0}
 {bone1Name=dyn_coat_bk_c_03, bone2Name=dyn_coat_bk_c_02, swing1LimitDegrees=30, swing2LimitDegrees=30, twistLimitDegrees=0}
 {bone1Name=dyn_coat_bk_c_04, bone2Name=dyn_coat_bk_c_03, swing1LimitDegrees=30, swing2LimitDegrees=30, twistLimitDegrees=0}
 {bone1Name=dyn_coat_bk_c_05, bone2Name=dyn_coat_bk_c_04, swing1LimitDegrees=30, swing2LimitDegrees=30, twistLimitDegrees=0}
 {bone1Name=dyn_coat_bk_c_06, bone2Name=dyn_coat_bk_c_05, swing1LimitDegrees=30, swing2LimitDegrees=30, twistLimitDegrees=0}
 {bone1Name=dyn_coat_front_r_02, bone2Name=dyn_coat_front_r_01, swing1LimitDegrees=30, swing2LimitDegrees=30, twistLimitDegrees=0}
 {bone1Name=dyn_coat_front_r_03, bone2Name=dyn_coat_front_r_02, swing1LimitDegrees=30, swing2LimitDegrees=30, twistLimitDegrees=0}
 {bone1Name=dyn_coat_front_r_04, bone2Name=dyn_coat_front_r_03, swing1LimitDegrees=30, swing2LimitDegrees=30, twistLimitDegrees=0}
 {bone1Name=dyn_coat_front_r_05, bone2Name=dyn_coat_front_r_04, swing1LimitDegrees=30, swing2LimitDegrees=30, twistLimitDegrees=0}
 {bone1Name=dyn_coat_r_02, bone2Name=dyn_coat_r_01, swing1LimitDegrees=30, swing2LimitDegrees=30, twistLimitDegrees=0}
 {bone1Name=dyn_coat_r_03, bone2Name=dyn_coat_r_02, swing1LimitDegrees=12, swing2LimitDegrees=30, twistLimitDegrees=0}
 {bone1Name=dyn_coat_r_04, bone2Name=dyn_coat_r_03, swing1LimitDegrees=15, swing2LimitDegrees=30, twistLimitDegrees=0}
 {bone1Name=dyn_coat_r_05, bone2Name=dyn_coat_r_04, swing1LimitDegrees=15, swing2LimitDegrees=30, twistLimitDegrees=0}
 {bone1Name=dyn_coat_bk_r_02, bone2Name=dyn_coat_bk_r_01, swing1LimitDegrees=30, swing2LimitDegrees=30, twistLimitDegrees=0}
 {bone1Name=dyn_coat_bk_r_03, bone2Name=dyn_coat_bk_r_02, swing1LimitDegrees=30, swing2LimitDegrees=30, twistLimitDegrees=0}
 {bone1Name=dyn_coat_bk_r_04, bone2Name=dyn_coat_bk_r_03, swing1LimitDegrees=30, swing2LimitDegrees=30, twistLimitDegrees=0}
 {bone1Name=dyn_coat_bk_r_05, bone2Name=dyn_coat_bk_r_04, swing1LimitDegrees=30, swing2LimitDegrees=30, twistLimitDegrees=0}
 {bone1Name=dyn_coat_bk_r_06, bone2Name=dyn_coat_bk_r_05, swing1LimitDegrees=30, swing2LimitDegrees=30, twistLimitDegrees=0}
 {
  bone1Name=thigh_l
  bone2Name=pelvis
  swing1LimitDegrees=45
  swing1Motion=Free
  swing2LimitDegrees=45
  swing2Motion=Free
  twistLimitDegrees=45
  twistMotion=Free
 }
 {
  bone1Name=thigh_r
  bone2Name=pelvis
  swing1LimitDegrees=45
  swing1Motion=Free
  swing2LimitDegrees=45
  swing2Motion=Free
  twistLimitDegrees=45
  twistMotion=Free
 }
 {bone1Name=dyn_coat_front_l_01, bone2Name=pelvis, swing1LimitDegrees=30, swing2LimitDegrees=30, twistLimitDegrees=0}
 {bone1Name=dyn_coat_l_01, bone2Name=pelvis, swing1LimitDegrees=30, swing2LimitDegrees=30, twistLimitDegrees=0}
 {bone1Name=dyn_coat_bk_l_01, bone2Name=pelvis, swing1LimitDegrees=30, swing2LimitDegrees=30, twistLimitDegrees=0}
 {bone1Name=dyn_coat_bk_c_01, bone2Name=pelvis, swing1LimitDegrees=30, swing2LimitDegrees=30, twistLimitDegrees=0}
 {bone1Name=dyn_coat_front_r_01, bone2Name=pelvis, swing1LimitDegrees=30, swing2LimitDegrees=30, twistLimitDegrees=0}
 {bone1Name=dyn_coat_r_01, bone2Name=pelvis, swing1LimitDegrees=15, swing2LimitDegrees=30, twistLimitDegrees=0}
 {bone1Name=dyn_coat_bk_r_01, bone2Name=pelvis, swing1LimitDegrees=30, swing2LimitDegrees=30, twistLimitDegrees=0}
 {
  bone1Name=calf_l
  bone2Name=thigh_l
  swing1LimitDegrees=45
  swing1Motion=Free
  swing2LimitDegrees=45
  swing2Motion=Free
  twistLimitDegrees=45
  twistMotion=Free
 }
 {bone1Name=dyn_coat_r_06, bone2Name=dyn_coat_r_05, swing1LimitDegrees=15, swing2LimitDegrees=30, twistLimitDegrees=0}
 {bone1Name=dyn_coat_front_r_06, bone2Name=dyn_coat_front_r_05, swing1LimitDegrees=30, swing2LimitDegrees=30, twistLimitDegrees=0}
 {bone1Name=dyn_coat_l_06, bone2Name=dyn_coat_l_05, swing1LimitDegrees=20, swing2LimitDegrees=20, twistLimitDegrees=0}
 {
  bone1Name=dyn_coat_front_l_06
  bone2Name=dyn_coat_front_l_05
  swing1LimitDegrees=45
  swing1Motion=Free
  swing2LimitDegrees=45
  swing2Motion=Free
  twistLimitDegrees=45
  twistMotion=Free
 }
 {
  bone1Name=foot_l
  bone2Name=calf_l
  swing1LimitDegrees=45
  swing1Motion=Free
  swing2LimitDegrees=45
  swing2Motion=Free
  twistLimitDegrees=45
  twistMotion=Free
 }
 {
  bone1Name=foot_r
  bone2Name=calf_r
  swing1LimitDegrees=45
  swing1Motion=Free
  swing2LimitDegrees=45
  swing2Motion=Free
  twistLimitDegrees=45
  twistMotion=Free
 }
 {
  bone1Name=cape_03_l
  bone2Name=cape_02_l
  swing1LimitDegrees=45
  swing1Motion=Free
  swing2LimitDegrees=45
  swing2Motion=Free
  twistLimitDegrees=45
  twistMotion=Free
 }
 {
  bone1Name=cape_02_l
  bone2Name=pelvis
  swing1LimitDegrees=45
  swing1Motion=Free
  swing2LimitDegrees=45
  swing2Motion=Free
  twistLimitDegrees=45
  twistMotion=Free
 }
 {
  bone1Name=calf_r
  bone2Name=thigh_r
  swing1LimitDegrees=45
  swing1Motion=Free
  swing2LimitDegrees=45
  swing2Motion=Free
  twistLimitDegrees=45
  twistMotion=Free
 }
]
