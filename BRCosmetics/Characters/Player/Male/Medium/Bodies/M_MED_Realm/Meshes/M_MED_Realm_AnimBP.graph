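# Anim Blueprint for the Realm male body cosmetic, correcting its cape, coat, shoulder pads, lightsaber and belt boxes per gameplay state.
# The cached 'postSim' pose feeds a long component-space chain of additive bone transforms and constraints.
# Sprint, skydive, parachute, crouch, ADS, melee, emote, front-end and vehicle variables gate the corrections, many with timed blends.
# The lightsaber adds weighted rotation constraints and spring Anim Dynamics; vehicle alpha drives cape-tip scaling.
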AnimBlueprint /BRCosmetics/Characters/Player/Male/Medium/Bodies/M_MED_Realm/Meshes/M_MED_Realm_AnimBP
target_skeleton=&/M_MED_Realm_Skeleton
graphs [
 AnimGraph {nodes [
   Component To Local()
   n1 = Get bIsTacticalSprinting()
   "Constraint - cape_02_r (PT)"!(ComponentPose, bAlphaBoolEnabled:bool=@n1, ConstraintWeights_0=1, ConstraintWeights_1:float=1) -> (Pose) {
    alphaInputType=Bool
    alphaBoolBlend {blendOption=Linear}
    boneToModify {boneName=cape_02_r}
    constraintSetup [{targetBone {boneName=upperarm_r}}, {transformType=Translation, targetBone {boneName=cape_l_MSTR}}]
    constraintWeights [1,1]
   }
   "Constraint - cape_01_r (PT)"!(ComponentPose, bAlphaBoolEnabled:bool=@n1, ConstraintWeights_0=1, ConstraintWeights_1:float=1) -> (Pose) {
    alphaInputType=Bool
    alphaBoolBlend {blendOption=Linear}
    boneToModify {boneName=cape_01_r}
    constraintSetup [{targetBone {boneName=upperarm_r}}, {transformType=Translation, targetBone {boneName=cape_l_MSTR}}]
    constraintWeights [1,1]
   }
   n2 = "Transform (Modify) Bone - Bone: cape_01_r"!(ComponentPose, bAlphaBoolEnabled:bool=True, Translation="0,0,0", Rotation="20.623,6.428,-0.001", Scale="1,1,1") -> (Pose) {
    alphaInputType=Bool
    rotationMode=BMM_Additive
    rotationSpace=BCS_BoneSpace
    alphaBoolBlend {blendOption=Linear}
    boneToModify {boneName=cape_01_r}
    rotation=<rot 20.623 6.428 -0.001>
   }
   n3 = "Transform (Modify) Bone - Bone: cape_02_r"!(ComponentPose=@n2, bAlphaBoolEnabled:bool=True, Translation="0,0,0", Rotation="-0.750,6.803,-0.095", Scale="1,1,1") -> (Pose) {
    alphaInputType=Bool
    rotationMode=BMM_Additive
    rotationSpace=BCS_BoneSpace
    alphaBoolBlend {blendOption=Linear}
    boneToModify {boneName=cape_02_r}
    rotation=<rot -0.75 6.803 -0.095>
   }
   n4 = "Transform (Modify) Bone - Bone: cape_03_r"!(ComponentPose=@n3, bAlphaBoolEnabled:bool=True, Translation="0,0,0", Rotation="-0.002,12.516,-0.000", Scale="1,1,1") -> (Pose) {
    alphaInputType=Bool
    rotationMode=BMM_Additive
    rotationSpace=BCS_BoneSpace
    alphaBoolBlend {blendOption=Linear}
    boneToModify {boneName=cape_03_r}
    rotation=<rot -0.002 12.516 0>
   }
   "Transform (Modify) Bone - Bone: cape_04_r"!(ComponentPose=@n4, bAlphaBoolEnabled:bool=True, Translation="0,0,0", Rotation="-0.006,9.940,-0.012", Scale="1,1,1") -> (Pose) {
    alphaInputType=Bool
    rotationMode=BMM_Additive
    rotationSpace=BCS_BoneSpace
    alphaBoolBlend {blendOption=Linear}
    boneToModify {boneName=cape_04_r}
    rotation=<rot -0.006 9.94 -0.012>
   }
   n5 = "Transform (Modify) Bone - Bone: dyn_capeShoulderMSTR_l_01"!(
    ComponentPose=Local To Component(LocalPose=Use cached pose 'postSim'())
    bAlphaBoolEnabled:bool=false
    Translation="3,3,-3"
    Rotation="0,0,0"
    Scale="1,1,1"
   ) -> (Pose) {
    alphaInputType=Bool
    rotationMode=BMM_Additive
    rotationSpace=BCS_BoneSpace
    translationMode=BMM_Additive
    translationSpace=BCS_BoneSpace
    alphaBoolBlend {blendOption=Linear}
    boneToModify {boneName=dyn_capeShoulderMSTR_l_01}
    translation=<3 3 -3>
   }
   n6 = "Transform (Modify) Bone - Bone: dyn_capeShoulderMSTR_r_01"!(ComponentPose=@n5, bAlphaBoolEnabled:bool=false, Translation="-2,-2,-3", Rotation="-10,0,-10", Scale="1,1,1") -> (Pose) {
    alphaInputType=Bool
    rotationMode=BMM_Additive
    rotationSpace=BCS_BoneSpace
    translationMode=BMM_Additive
    translationSpace=BCS_BoneSpace
    alphaBoolBlend {blendOption=Linear}
    boneToModify {boneName=dyn_capeShoulderMSTR_r_01}
    rotation {pitch=-10, roll=-10}
    translation=<-2 -2 -3>
   }
   n7 = Get skydiveNoParachute() -> (skydiveNoParachute:bool)
   n8 = "Transform (Modify) Bone - Bone: cape_01_r"(ComponentPose=@n6, bAlphaBoolEnabled=@n7, Translation="0,2,15", Rotation="0,-10,0", Scale="1,1,1") {
    alphaInputType=Bool
    rotationMode=BMM_Additive
    rotationSpace=BCS_BoneSpace
    translationMode=BMM_Additive
    translationSpace=BCS_BoneSpace
    alphaBoolBlend {blendOption=Linear}
    boneToModify {boneName=cape_01_r}
    rotation {yaw=-10}
    translation {y=2, z=15}
   }
   n9 = "Transform (Modify) Bone - Bone: dyn_shoulderPad_l_01"!(ComponentPose=@n8, bAlphaBoolEnabled:bool=@n7, Translation="0,0,0", Rotation="0,0,0", Scale="1,1,1") -> (Pose) {
    alphaInputType=Bool
    rotationMode=BMM_Additive
    rotationSpace=BCS_BoneSpace
    translationMode=BMM_Additive
    translationSpace=BCS_BoneSpace
    alphaBoolBlend {blendOption=Linear}
    boneToModify {boneName=dyn_shoulderPad_l_01}
   }
   n10 = "Transform (Modify) Bone - Bone: cape_mid_01_r"!(ComponentPose=@n9, bAlphaBoolEnabled:bool=@n7, Translation="0,5,10", Rotation="0,0,0", Scale="1,1,1") -> (Pose) {
    alphaInputType=Bool
    rotationMode=BMM_Additive
    rotationSpace=BCS_BoneSpace
    translationMode=BMM_Additive
    translationSpace=BCS_BoneSpace
    alphaBoolBlend {blendOption=Linear}
    boneToModify {boneName=cape_mid_01_r}
    translation {y=5, z=10}
   }
   n11 = "Transform (Modify) Bone - Bone: cape_mid_01_l"!(ComponentPose=@n10, bAlphaBoolEnabled:bool=@n7, Translation="0,5,8", Rotation="0,0,0", Scale="1,1,1") -> (Pose) {
    alphaInputType=Bool
    rotationMode=BMM_Additive
    rotationSpace=BCS_BoneSpace
    translationMode=BMM_Additive
    translationSpace=BCS_BoneSpace
    alphaBoolBlend {blendOption=Linear}
    boneToModify {boneName=cape_mid_01_l}
    translation {y=5, z=8}
   }
   n12 = "Transform (Modify) Bone - Bone: cape_01_l"!(ComponentPose=@n11, bAlphaBoolEnabled:bool=@n7, Translation="0,-4,10", Rotation="0,0,0", Scale="1,1,1") -> (Pose) {
    alphaInputType=Bool
    rotationMode=BMM_Additive
    rotationSpace=BCS_BoneSpace
    translationMode=BMM_Additive
    translationSpace=BCS_BoneSpace
    alphaBoolBlend {blendOption=Linear}
    boneToModify {boneName=cape_01_l}
    translation {y=-4, z=10}
   }
   n13 = "Constraint - dyn_shoulderPad_l_01 (P)"!(ComponentPose=@n12, bAlphaBoolEnabled:bool=@n7, ConstraintWeights_0=1) -> (Pose) {
    alphaInputType=Bool
    alphaBoolBlend {blendOption=Linear}
    boneToModify {boneName=dyn_shoulderPad_l_01}
    constraintSetup [{targetBone {boneName=upperarm_twist_02_l}}]
    constraintWeights [1]
   }
   n14 = "Constraint - dyn_shoulderPad_r_01 (P)"!(ComponentPose=@n13, bAlphaBoolEnabled:bool=@n7, ConstraintWeights_0=1) -> (Pose) {
    alphaInputType=Bool
    alphaBoolBlend {blendOption=Linear}
    boneToModify {boneName=dyn_shoulderPad_r_01}
    constraintSetup [{targetBone {boneName=upperarm_twist_02_r}}]
    constraintWeights [1]
   }
   n15 = Get bIsParachuteOpened()
   n16 = "Transform (Modify) Bone - Bone: cape_01_r"(ComponentPose=@n14, bAlphaBoolEnabled=@n15, Translation="0,10,0", Rotation="0,0,-10", Scale="1,1,1") {
    alphaInputType=Bool
    rotationMode=BMM_Additive
    rotationSpace=BCS_BoneSpace
    translationMode=BMM_Additive
    translationSpace=BCS_BoneSpace
    alphaBoolBlend {blendOption=Linear}
    boneToModify {boneName=cape_01_r}
    rotation {roll=-10}
    translation {y=10}
   }
   n17 = "Transform (Modify) Bone - Bone: cape_01_l"(ComponentPose=@n16, bAlphaBoolEnabled=@n15, Translation="0,-5,0", Rotation="0,0,30", Scale="1,1,1") {
    alphaInputType=Bool
    rotationMode=BMM_Additive
    rotationSpace=BCS_BoneSpace
    translationMode=BMM_Additive
    translationSpace=BCS_BoneSpace
    alphaBoolBlend {blendOption=Linear}
    boneToModify {boneName=cape_01_l}
    rotation {roll=30}
    translation {y=-5}
   }
   n18 = "Constraint - dyn_capeShoulder_l_01 (PP)"!(ComponentPose=@n17, bAlphaBoolEnabled:bool=@n15, ConstraintWeights_0=1, ConstraintWeights_1:float=0) -> (Pose) {
    alphaInputType=Bool
    alphaBoolBlend {blendOption=Linear}
    boneToModify {boneName=dyn_capeShoulder_l_01}
    constraintSetup [{targetBone {boneName=upperarm_twist_02_l}}, {targetBone {boneName=spine_05}}]
    constraintWeights [1,1]
   }
   n19 = "Transform (Modify) Bone - Bone: dyn_capeShoulder_l_01"!(ComponentPose=@n18, bAlphaBoolEnabled:bool=@n15, Translation="-12,-2,5", Rotation="-10,0,0", Scale="1,1.5,1.5") -> (Pose) {
    alphaInputType=Bool
    rotationMode=BMM_Additive
    rotationSpace=BCS_BoneSpace
    scaleMode=BMM_Additive
    scaleSpace=BCS_BoneSpace
    translationMode=BMM_Additive
    translationSpace=BCS_BoneSpace
    alphaBoolBlend {blendOption=Linear}
    boneToModify {boneName=dyn_capeShoulder_l_01}
    rotation {pitch=-10}
    scale {y=1.5, z=1.5}
    translation=<-12 -2 5>
   }
   n20 = "Constraint - dyn_shoulderPad_l_01 (P)"(ComponentPose=@n19, bAlphaBoolEnabled=@n15, ConstraintWeights_0=1) {
    alphaInputType=Bool
    alphaBoolBlend {blendOption=Linear}
    boneToModify {boneName=dyn_shoulderPad_l_01}
    constraintSetup [{targetBone {boneName=upperarm_twist_02_l}}]
    constraintWeights [1]
   }
   n21 = "Transform (Modify) Bone - Bone: dyn_shoulderPad_l_01"(ComponentPose=@n20, bAlphaBoolEnabled=@n15, Translation="-3,0,0", Rotation="10,-35,0", Scale="1,1,0.85") {
    alphaInputType=Bool
    rotationMode=BMM_Additive
    rotationSpace=BCS_BoneSpace
    scaleMode=BMM_Additive
    translationMode=BMM_Additive
    translationSpace=BCS_BoneSpace
    alphaBoolBlend {blendOption=Linear}
    boneToModify {boneName=dyn_shoulderPad_l_01}
    rotation {pitch=10, yaw=-35}
    scale {z=0.85}
    translation {x=-3}
   }
   n22 = "Constraint - dyn_shoulderPad_r_01 (P)"(ComponentPose=@n21, bAlphaBoolEnabled=@n15, ConstraintWeights_0=1) {
    alphaInputType=Bool
    alphaBoolBlend {blendOption=Linear}
    boneToModify {boneName=dyn_shoulderPad_r_01}
    constraintSetup [{targetBone {boneName=upperarm_twist_02_r}}]
    constraintWeights [1]
   }
   n23 = "Constraint - dyn_capeShoulder_r_01 (PP)"!(ComponentPose=@n22, bAlphaBoolEnabled:bool=@n15, ConstraintWeights_0=1, ConstraintWeights_1:float=0) -> (Pose) {
    alphaInputType=Bool
    alphaBoolBlend {blendOption=Linear}
    boneToModify {boneName=dyn_capeShoulder_r_01}
    constraintSetup [{targetBone {boneName=upperarm_twist_02_r}}, {targetBone {boneName=spine_05}}]
    constraintWeights [1,1]
   }
   n24 = "Transform (Modify) Bone - Bone: dyn_capeShoulder_r_01"!(ComponentPose=@n23, bAlphaBoolEnabled:bool=@n15, Translation="10,0,-5", Rotation="-30,0,0", Scale="1,1.5,1.5") -> (Pose) {
    alphaInputType=Bool
    rotationMode=BMM_Additive
    rotationSpace=BCS_BoneSpace
    scaleMode=BMM_Additive
    scaleSpace=BCS_BoneSpace
    translationMode=BMM_Additive
    translationSpace=BCS_BoneSpace
    alphaBoolBlend {blendOption=Linear}
    boneToModify {boneName=dyn_capeShoulder_r_01}
    rotation {pitch=-30}
    scale {y=1.5, z=1.5}
    translation {x=10, z=-5}
   }
   n25 = "Transform (Modify) Bone - Bone: dyn_shoulderPad_r_01"!(ComponentPose=@n24, bAlphaBoolEnabled:bool=@n15, Translation="3,0,0", Rotation="0,-30,-6", Scale="1,1,0.85") -> (Pose) {
    alphaInputType=Bool
    rotationMode=BMM_Additive
    rotationSpace=BCS_BoneSpace
    scaleMode=BMM_Additive
    translationMode=BMM_Additive
    translationSpace=BCS_BoneSpace
    alphaBoolBlend {blendOption=Linear}
    boneToModify {boneName=dyn_shoulderPad_r_01}
    rotation {roll=-6, yaw=-30}
    scale {z=0.85}
    translation {x=3}
   }
   n26 = Get bIsSprinting()
   n27 = "Transform (Modify) Bone - Bone: cape_mid_02_l"!(ComponentPose=@n25, bAlphaBoolEnabled:bool=@n26, Translation="0,0,0", Rotation="2,0,0", Scale="1,1,1") -> (Pose) {
    alphaInputType=Bool
    rotationMode=BMM_Additive
    rotationSpace=BCS_BoneSpace
    scaleMode=BMM_Additive
    scaleSpace=BCS_BoneSpace
    alphaBoolBlend {blendInTime=0.2, blendOption=Linear, blendOutTime=1}
    boneToModify {boneName=cape_mid_02_l}
    rotation {pitch=2}
   }
   n28 = "Transform (Modify) Bone - Bone: cape_mid_04_l"!(ComponentPose=@n27, bAlphaBoolEnabled:bool=@n26, Translation="0,0,0", Rotation="4,0,0", Scale="1,1,1") -> (Pose) {
    alphaInputType=Bool
    rotationMode=BMM_Additive
    rotationSpace=BCS_BoneSpace
    scaleMode=BMM_Additive
    scaleSpace=BCS_BoneSpace
    alphaBoolBlend {blendInTime=0.2, blendOption=Linear, blendOutTime=1}
    boneToModify {boneName=cape_mid_04_l}
    rotation {pitch=4}
   }
   n29 = "Transform (Modify) Bone - Bone: cape_02_l"!(ComponentPose=@n28, bAlphaBoolEnabled:bool=@n26, Translation="-3,-3,10", Rotation="0,-5,0", Scale="1,1,1") -> (Pose) {
    alphaInputType=Bool
    rotationMode=BMM_Additive
    rotationSpace=BCS_BoneSpace
    scaleMode=BMM_Additive
    scaleSpace=BCS_BoneSpace
    translationMode=BMM_Additive
    translationSpace=BCS_BoneSpace
    alphaBoolBlend {blendInTime=0.2, blendOption=Linear, blendOutTime=1}
    boneToModify {boneName=cape_02_l}
    rotation {yaw=-5}
    translation=<-3 -3 10>
   }
   n30 = "Transform (Modify) Bone - Bone: cape_01_l"(ComponentPose=@n29, bAlphaBoolEnabled=@n26, Translation="0,-3,0", Rotation="0,0,0", Scale="1,1,1") {
    alphaInputType=Bool
    rotationMode=BMM_Additive
    rotationSpace=BCS_BoneSpace
    scaleMode=BMM_Additive
    scaleSpace=BCS_BoneSpace
    translationMode=BMM_Additive
    translationSpace=BCS_BoneSpace
    alphaBoolBlend {blendInTime=1, blendOption=Linear, blendOutTime=1}
    boneToModify {boneName=cape_01_l}
    translation {y=-3}
   }
   n31 = Get bIsSprinting()
   n32 = "Transform (Modify) Bone - Bone: cape_mid_02_r"!(ComponentPose=@n30, bAlphaBoolEnabled:bool=@n31, Translation="0,0,0", Rotation="2,0,0", Scale="1,1,1") -> (Pose) {
    alphaInputType=Bool
    rotationMode=BMM_Additive
    rotationSpace=BCS_BoneSpace
    scaleMode=BMM_Additive
    scaleSpace=BCS_BoneSpace
    alphaBoolBlend {blendInTime=0.2, blendOption=Linear, blendOutTime=1}
    boneToModify {boneName=cape_mid_02_r}
    rotation {pitch=2}
   }
   n33 = "Transform (Modify) Bone - Bone: cape_mid_04_r"!(ComponentPose=@n32, bAlphaBoolEnabled:bool=@n31, Translation="0,0,0", Rotation="4,0,0", Scale="1,1,1") -> (Pose) {
    alphaInputType=Bool
    rotationMode=BMM_Additive
    rotationSpace=BCS_BoneSpace
    scaleMode=BMM_Additive
    scaleSpace=BCS_BoneSpace
    alphaBoolBlend {blendInTime=0.2, blendOption=Linear, blendOutTime=1}
    boneToModify {boneName=cape_mid_04_r}
    rotation {pitch=4}
   }
   n34 = "Transform (Modify) Bone - Bone: cape_01_r"(ComponentPose=@n33, bAlphaBoolEnabled=@n31, Translation="0,0,0", Rotation="0,5,0", Scale="1,1,1") {
    alphaInputType=Bool
    rotationMode=BMM_Additive
    rotationSpace=BCS_BoneSpace
    scaleMode=BMM_Additive
    scaleSpace=BCS_BoneSpace
    translationMode=BMM_Additive
    translationSpace=BCS_BoneSpace
    alphaBoolBlend {blendInTime=1, blendOption=Linear, blendOutTime=1}
    boneToModify {boneName=cape_01_r}
    rotation {yaw=5}
   }
   n35 = Get bIsSprinting()
   n36 = "Transform (Modify) Bone - Bone: cape_mid_02_r"(ComponentPose=@n34, bAlphaBoolEnabled=@n35, Translation="0,0,0", Rotation="2,0,0", Scale="1,1,1") {
    alphaInputType=Bool
    rotationMode=BMM_Additive
    rotationSpace=BCS_BoneSpace
    scaleMode=BMM_Additive
    scaleSpace=BCS_BoneSpace
    alphaBoolBlend {blendInTime=0.2, blendOption=Linear, blendOutTime=1}
    boneToModify {boneName=cape_mid_02_r}
    rotation {pitch=2}
   }
   n37 = "Transform (Modify) Bone - Bone: cape_mid_04_r"(ComponentPose=@n36, bAlphaBoolEnabled=@n35, Translation="0,0,0", Rotation="4,0,0", Scale="1,1,1") {
    alphaInputType=Bool
    rotationMode=BMM_Additive
    rotationSpace=BCS_BoneSpace
    scaleMode=BMM_Additive
    scaleSpace=BCS_BoneSpace
    alphaBoolBlend {blendInTime=0.2, blendOption=Linear, blendOutTime=1}
    boneToModify {boneName=cape_mid_04_r}
    rotation {pitch=4}
   }
   n38 = Get bIsSprinting()
   n39 = "Transform (Modify) Bone - Bone: cape_01_r"(ComponentPose=@n37, bAlphaBoolEnabled=@n38, Translation="0,0,0", Rotation="2,0,0", Scale="1,1,1") {
    alphaInputType=Bool
    rotationMode=BMM_Additive
    rotationSpace=BCS_BoneSpace
    scaleMode=BMM_Additive
    scaleSpace=BCS_BoneSpace
    alphaBoolBlend {blendInTime=0.6, blendOption=Linear, blendOutTime=0.1}
    boneToModify {boneName=cape_01_r}
    rotation {pitch=2}
   }
   n40 = "Transform (Modify) Bone - Bone: cape_04_r"(ComponentPose=@n39, bAlphaBoolEnabled=@n38, Translation="0,0,0", Rotation="3,7,0", Scale="0.7,1,1") {
    alphaInputType=Bool
    rotationMode=BMM_Additive
    rotationSpace=BCS_BoneSpace
    scaleMode=BMM_Additive
    scaleSpace=BCS_BoneSpace
    alphaBoolBlend {blendInTime=0.2, blendOption=Linear, blendOutTime=1}
    boneToModify {boneName=cape_04_r}
    rotation {pitch=3, yaw=7}
    scale {x=0.7}
   }
   n41 = Get bIsSprinting()
   n42 = "Transform (Modify) Bone - Bone: dyn_coat_bk_c_03"!(ComponentPose=@n40, bAlphaBoolEnabled:bool=@n41, Translation="0,0,0", Rotation="15,0,0", Scale="1,1,1") -> (Pose) {
    alphaInputType=Bool
    rotationMode=BMM_Additive
    rotationSpace=BCS_BoneSpace
    scaleMode=BMM_Additive
    scaleSpace=BCS_BoneSpace
    alphaBoolBlend {blendInTime=0.2, blendOption=Linear, blendOutTime=1}
    boneToModify {boneName=dyn_coat_bk_c_03}
    rotation {pitch=15}
   }
   n43 = "Transform (Modify) Bone - Bone: dyn_coat_bk_c_04"!(ComponentPose=@n42, bAlphaBoolEnabled:bool=@n41, Translation="0,0,0", Rotation="15,0,0", Scale="1,1,1") -> (Pose) {
    alphaInputType=Bool
    rotationMode=BMM_Additive
    rotationSpace=BCS_BoneSpace
    scaleMode=BMM_Additive
    scaleSpace=BCS_BoneSpace
    alphaBoolBlend {blendInTime=0.2, blendOption=Linear, blendOutTime=1}
    boneToModify {boneName=dyn_coat_bk_c_04}
    rotation {pitch=15}
   }
   n44 = "Transform (Modify) Bone - Bone: dyn_coat_bk_l_03"!(ComponentPose=@n43, bAlphaBoolEnabled:bool=@n41, Translation="0,0,0", Rotation="5,0,0", Scale="1,1,1") -> (Pose) {
    alphaInputType=Bool
    rotationMode=BMM_Additive
    rotationSpace=BCS_BoneSpace
    scaleMode=BMM_Additive
    scaleSpace=BCS_BoneSpace
    alphaBoolBlend {blendInTime=0.2, blendOption=Linear, blendOutTime=1}
    boneToModify {boneName=dyn_coat_bk_l_03}
    rotation {pitch=5}
   }
   n45 = "Transform (Modify) Bone - Bone: dyn_coat_bk_l_04"!(ComponentPose=@n44, bAlphaBoolEnabled:bool=@n41, Translation="0,0,0", Rotation="2,0,0", Scale="1,1,1") -> (Pose) {
    alphaInputType=Bool
    rotationMode=BMM_Additive
    rotationSpace=BCS_BoneSpace
    scaleMode=BMM_Additive
    scaleSpace=BCS_BoneSpace
    alphaBoolBlend {blendInTime=0.2, blendOption=Linear, blendOutTime=1}
    boneToModify {boneName=dyn_coat_bk_l_04}
    rotation {pitch=2}
   }
   n46 = "Transform (Modify) Bone - Bone: dyn_coat_bk_r_03"!(ComponentPose=@n45, bAlphaBoolEnabled:bool=@n41, Translation="0,0,0", Rotation="10,0,0", Scale="1,1,1") -> (Pose) {
    alphaInputType=Bool
    rotationMode=BMM_Additive
    rotationSpace=BCS_BoneSpace
    scaleMode=BMM_Additive
    scaleSpace=BCS_BoneSpace
    alphaBoolBlend {blendInTime=0.2, blendOption=Linear, blendOutTime=1}
    boneToModify {boneName=dyn_coat_bk_r_03}
    rotation {pitch=10}
   }
   n47 = "Transform (Modify) Bone - Bone: dyn_coat_bk_r_04"!(ComponentPose=@n46, bAlphaBoolEnabled:bool=@n41, Translation="0,0,0", Rotation="4,0,0", Scale="1,1,1") -> (Pose) {
    alphaInputType=Bool
    rotationMode=BMM_Additive
    rotationSpace=BCS_BoneSpace
    scaleMode=BMM_Additive
    scaleSpace=BCS_BoneSpace
    alphaBoolBlend {blendInTime=0.2, blendOption=Linear, blendOutTime=1}
    boneToModify {boneName=dyn_coat_bk_r_04}
    rotation {pitch=4}
   }
   n48 = "Transform (Modify) Bone - Bone: cape_01_l"!(ComponentPose=@n47, Alpha:float=0, Translation="0,-5,4", Rotation="0,0,15", Scale="1,1,1") -> (Pose) {
    rotationMode=BMM_Additive
    rotationSpace=BCS_BoneSpace
    scaleMode=BMM_Additive
    scaleSpace=BCS_BoneSpace
    translationMode=BMM_Additive
    translationSpace=BCS_BoneSpace
    alphaBoolBlend {blendInTime=0.6, blendOption=Linear, blendOutTime=0.1}
    alphaScaleBiasClamp {bInterpResult=true, bMapRange=true}
    boneToModify {boneName=cape_01_l}
    rotation {roll=15}
    translation {y=-5, z=4}
   }
   n49 = "Transform (Modify) Bone - Bone: cape_03_l"(ComponentPose=@n48, Alpha=0, Translation="0,0,0", Rotation="0,0,0", Scale="1,1,1") {
    rotationMode=BMM_Additive
    rotationSpace=BCS_BoneSpace
    scaleMode=BMM_Additive
    scaleSpace=BCS_BoneSpace
    translationMode=BMM_Additive
    translationSpace=BCS_BoneSpace
    alphaBoolBlend {blendInTime=0.6, blendOption=Linear, blendOutTime=0.1}
    alphaScaleBiasClamp {bMapRange=true}
    boneToModify {boneName=cape_03_l}
   }
   n50 = "Transform (Modify) Bone - Bone: cape_04_l"(ComponentPose=@n49, Alpha=0, Translation="0,0,0", Rotation="-20,0,0", Scale="1,1,1") {
    rotationMode=BMM_Additive
    rotationSpace=BCS_BoneSpace
    scaleMode=BMM_Additive
    scaleSpace=BCS_BoneSpace
    translationMode=BMM_Additive
    translationSpace=BCS_BoneSpace
    alphaBoolBlend {blendInTime=0.6, blendOption=Linear, blendOutTime=0.1}
    alphaScaleBiasClamp {bMapRange=true}
    boneToModify {boneName=cape_04_l}
    rotation {pitch=-20}
   }
   n51 = "Constraint - dyn_lightSaber_01 (RRRR)"!(
    ComponentPose=@n50
    Alpha=1
    ConstraintWeights_0=0
    ConstraintWeights_1:float=0.1
    ConstraintWeights_2:float=0.6
    ConstraintWeights_3:float=0.3
   ) -> (Pose) {
    alphaBoolBlend {blendOption=Linear}
    boneToModify {boneName=dyn_lightSaber_01}
    constraintSetup [
     {transformType=Rotation, targetBone {boneName=dyn_coat_front_l_01}}
     {transformType=Rotation, targetBone {boneName=dyn_coat_l_01}}
     {transformType=Rotation, targetBone {boneName=dyn_coat_front_l_03}}
     {transformType=Rotation, targetBone {boneName=spine_01}}
    ]
    constraintWeights [1,1,1,1]
   }
   n52 = "Constraint - dyn_lightSaber_01 (PPP)"!(ComponentPose=@n51, Alpha=0, ConstraintWeights_0=0.5, ConstraintWeights_1:float=0, ConstraintWeights_2:float=0.5) -> (Pose) {
    alphaBoolBlend {blendOption=Linear}
    boneToModify {boneName=dyn_lightSaber_01}
    constraintSetup [{targetBone {boneName=dyn_coat_front_l_02}}, {targetBone {boneName=dyn_coat_l_02}}, {targetBone {boneName=dyn_coat_front_l_03}}]
    constraintWeights [1,1,1]
   }
   n53 = "Transform (Modify) Bone - Bone: dyn_lightSaber_01"(ComponentPose=@n52, Alpha=1, Translation="-3,-2,0", Rotation="5,0,0", Scale="1,1,1") {
    rotationMode=BMM_Additive
    rotationSpace=BCS_BoneSpace
    translationMode=BMM_Additive
    translationSpace=BCS_BoneSpace
    alphaBoolBlend {blendOption=Linear}
    boneToModify {boneName=dyn_lightSaber_01}
    rotation {pitch=5}
    translation=<-3 -2>
   }
   n54 = "Transform (Modify) Bone - Bone: dyn_lightSaber_02"(ComponentPose=@n53, Alpha=1, Translation="0,0,0", Rotation="0,0,0", Scale="1,1,1") {
    rotationMode=BMM_Additive
    rotationSpace=BCS_BoneSpace
    translationMode=BMM_Additive
    translationSpace=BCS_BoneSpace
    alphaBoolBlend {blendOption=Linear}
    boneToModify {boneName=dyn_lightSaber_02}
   }
   n55 = Anim Dynamics - Bone: dyn_lightSaber_01(ComponentPose=@n54, Alpha=0, ExternalForce="0,0,0") {
    angularDampingOverride=0.9
    angularSpringConstant=60
    bAngularSpring=true
    bOverrideAngularDamping=true
    bUseGravityOverride=true
    simulationSpace=RootRelative
    alphaBoolBlend {blendOption=Linear}
    boundBone {boneName=dyn_lightSaber_01}
    componentAppliedLinearAccClamp=<3000 3000 3000>
    componentLinearAccScale=<0.1 0.1 0.1>
    componentLinearVelScale=<0.1 0.1 0.1>
    gravityOverride {z=-980}
    physicsBodyDefinitions [{
      boundBone {boneName=dyn_lightSaber_01}
      boxExtents=<15 30 30>
      constraintSetup {angularLimitsMax=<0 0 5>, angularLimitsMin=<0 0 -5>, angularTarget=<0 0 0>}
      localJointOffset=<8 0 0>
     }]
    relativeSpaceBone {boneName=pelvis}
   }
   n56 = Anim Dynamics - Bone: dyn_lightSaber_02(ComponentPose=@n55, Alpha=1, ExternalForce="0,0,0") {
    angularDampingOverride=0.7
    angularSpringConstant=10
    bAngularSpring=true
    bOverrideAngularDamping=true
    bUseGravityOverride=true
    numSolverIterationsPreUpdate=3
    simulationSpace=RootRelative
    alphaBoolBlend {blendOption=Linear}
    boundBone {boneName=dyn_lightSaber_02}
    componentAppliedLinearAccClamp=<3000 3000 3000>
    componentLinearAccScale=<0.05 0.05 0.05>
    componentLinearVelScale=<0.05 0.05 0.05>
    gravityOverride {z=-980}
    physicsBodyDefinitions [{
      boundBone {boneName=dyn_lightSaber_02}
      boxExtents=<10 30 30>
      constraintSetup {angularLimitsMax=<0 0 8>, angularLimitsMin=<0 0 -8>, angularTarget=<0 0 0>}
      localJointOffset=<15 0 0>
     }]
    relativeSpaceBone {boneName=pelvis}
   }
   n57 = Get isCrouchStopped() -> (isCrouchStopped:bool)
   n58 = "Transform (Modify) Bone - Bone: dyn_lightSaber_01"!(ComponentPose=@n56, bAlphaBoolEnabled:bool=@n57, Translation="0,0,0", Rotation="20,0,0", Scale="1,1,1") -> (Pose) {
    alphaInputType=Bool
    rotationMode=BMM_Additive
    rotationSpace=BCS_BoneSpace
    translationMode=BMM_Additive
    translationSpace=BCS_BoneSpace
    alphaBoolBlend {blendOption=Linear, blendOutTime=1}
    boneToModify {boneName=dyn_lightSaber_01}
    rotation {pitch=20}
   }
   n59 = "Transform (Modify) Bone - Bone: dyn_lightSaber_02"!(ComponentPose=@n58, bAlphaBoolEnabled:bool=@n57, Translation="0,0,0", Rotation="7,0,0", Scale="1,1,1") -> (Pose) {
    alphaInputType=Bool
    rotationMode=BMM_Additive
    rotationSpace=BCS_BoneSpace
    translationMode=BMM_Additive
    translationSpace=BCS_BoneSpace
    alphaBoolBlend {blendOption=Linear, blendOutTime=1}
    boneToModify {boneName=dyn_lightSaber_02}
    rotation {pitch=7}
   }
   n60 = Get bIsTacticalSprinting()
   n61 = "Transform (Modify) Bone - Bone: cape_mid_02_l"(ComponentPose=@n59, bAlphaBoolEnabled=@n60, Translation="0,0,5", Rotation="0,0,0", Scale="1,1,1") {
    alphaInputType=Bool
    rotationMode=BMM_Additive
    rotationSpace=BCS_BoneSpace
    scaleMode=BMM_Additive
    scaleSpace=BCS_BoneSpace
    translationMode=BMM_Additive
    translationSpace=BCS_BoneSpace
    alphaBoolBlend {blendInTime=0.2, blendOption=Linear, blendOutTime=1}
    boneToModify {boneName=cape_mid_02_l}
    translation {z=5}
   }
   n62 = "Transform (Modify) Bone - Bone: cape_mid_01_l"(ComponentPose=@n61, bAlphaBoolEnabled=@n60, Translation="0,0,4", Rotation="0,0,0", Scale="1,1,1") {
    alphaInputType=Bool
    rotationMode=BMM_Additive
    rotationSpace=BCS_BoneSpace
    scaleMode=BMM_Additive
    scaleSpace=BCS_BoneSpace
    translationMode=BMM_Additive
    translationSpace=BCS_BoneSpace
    alphaBoolBlend {blendInTime=0.2, blendOption=Linear, blendOutTime=1}
    boneToModify {boneName=cape_mid_01_l}
    translation {z=4}
   }
   n63 = "Constraint - cape_01_l (T)"!(ComponentPose=@n62, bAlphaBoolEnabled:bool=false, ConstraintWeights_0=1) -> (Pose) {
    alphaInputType=Bool
    alphaBoolBlend {blendOption=Linear}
    boneToModify {boneName=cape_01_l}
    constraintSetup [{transformType=Translation, targetBone {boneName=cape_mid_01_l}}]
    constraintWeights [1]
   }
   n64 = "Transform (Modify) Bone - Bone: cape_02_l"(ComponentPose=@n63, bAlphaBoolEnabled=@n60, Translation="0,0,3", Rotation="5,-5,0", Scale="1,1,1") {
    alphaInputType=Bool
    rotationMode=BMM_Additive
    rotationSpace=BCS_BoneSpace
    scaleMode=BMM_Additive
    scaleSpace=BCS_BoneSpace
    translationMode=BMM_Additive
    translationSpace=BCS_BoneSpace
    alphaBoolBlend {blendInTime=0.2, blendOption=Linear, blendOutTime=1}
    boneToModify {boneName=cape_02_l}
    rotation {pitch=5, yaw=-5}
    translation {z=3}
   }
   n65 = "Transform (Modify) Bone - Bone: cape_01_l"!(ComponentPose=@n64, bAlphaBoolEnabled:bool=@n60, Translation="0,0,2", Rotation="0,0,0", Scale="1,1,1") -> (Pose) {
    alphaInputType=Bool
    rotationMode=BMM_Additive
    rotationSpace=BCS_BoneSpace
    scaleMode=BMM_Additive
    scaleSpace=BCS_BoneSpace
    translationMode=BMM_Additive
    translationSpace=BCS_BoneSpace
    alphaBoolBlend {blendInTime=0.2, blendOption=Linear, blendOutTime=1}
    boneToModify {boneName=cape_01_l}
    translation {z=2}
   }
   n66 = "Transform (Modify) Bone - Bone: cape_mid_01_r"(ComponentPose=@n65, bAlphaBoolEnabled=@n60, Translation="0,0,6", Rotation="0,0,0", Scale="1,1,1") {
    alphaInputType=Bool
    rotationMode=BMM_Additive
    rotationSpace=BCS_BoneSpace
    scaleMode=BMM_Additive
    scaleSpace=BCS_BoneSpace
    translationMode=BMM_Additive
    translationSpace=BCS_BoneSpace
    alphaBoolBlend {blendInTime=0.15, blendOption=Linear, blendOutTime=1}
    boneToModify {boneName=cape_mid_01_r}
    translation {z=6}
   }
   n67 = "Transform (Modify) Bone - Bone: cape_mid_02_r"(ComponentPose=@n66, bAlphaBoolEnabled=@n60, Translation="0,0,6", Rotation="0,0,0", Scale="1,1,1") {
    alphaInputType=Bool
    rotationMode=BMM_Additive
    rotationSpace=BCS_BoneSpace
    scaleMode=BMM_Additive
    scaleSpace=BCS_BoneSpace
    translationMode=BMM_Additive
    translationSpace=BCS_BoneSpace
    alphaBoolBlend {blendInTime=0.15, blendOption=Linear, blendOutTime=1}
    boneToModify {boneName=cape_mid_02_r}
    translation {z=6}
   }
   n68 = "Transform (Modify) Bone - Bone: cape_02_r"(ComponentPose=@n67, bAlphaBoolEnabled=@n60, Translation="0,3,15", Rotation="0,0,0", Scale="1,1,1") {
    alphaInputType=Bool
    rotationMode=BMM_Additive
    rotationSpace=BCS_BoneSpace
    scaleMode=BMM_Additive
    scaleSpace=BCS_BoneSpace
    translationMode=BMM_Additive
    translationSpace=BCS_BoneSpace
    alphaBoolBlend {blendInTime=0.15, blendOption=Linear, blendOutTime=1}
    boneToModify {boneName=cape_02_r}
    translation {y=3, z=15}
   }
   n69 = Get bIsTacticalSprinting()
   n70 = "Transform (Modify) Bone - Bone: dyn_coat_bk_c_02"!(ComponentPose=@n68, bAlphaBoolEnabled:bool=@n69, Translation="0,0,5", Rotation="15,0,0", Scale="1,1,1") -> (Pose) {
    alphaInputType=Bool
    rotationMode=BMM_Additive
    rotationSpace=BCS_BoneSpace
    scaleMode=BMM_Additive
    scaleSpace=BCS_BoneSpace
    translationMode=BMM_Additive
    translationSpace=BCS_BoneSpace
    alphaBoolBlend {blendInTime=0.2, blendOption=Linear, blendOutTime=1}
    boneToModify {boneName=dyn_coat_bk_c_02}
    rotation {pitch=15}
    translation {z=5}
   }
   n71 = "Transform (Modify) Bone - Bone: dyn_coat_bk_c_04"(ComponentPose=@n70, bAlphaBoolEnabled=@n69, Translation="0,0,0", Rotation="-10,0,0", Scale="1,1,1") {
    alphaInputType=Bool
    rotationMode=BMM_Additive
    rotationSpace=BCS_BoneSpace
    scaleMode=BMM_Additive
    scaleSpace=BCS_BoneSpace
    alphaBoolBlend {blendInTime=0.2, blendOption=Linear, blendOutTime=1}
    boneToModify {boneName=dyn_coat_bk_c_04}
    rotation {pitch=-10}
   }
   n72 = "Transform (Modify) Bone - Bone: dyn_coat_bk_l_02"!(ComponentPose=@n71, bAlphaBoolEnabled:bool=@n69, Translation="0,0,5", Rotation="15,0,0", Scale="1,1,1") -> (Pose) {
    alphaInputType=Bool
    rotationMode=BMM_Additive
    rotationSpace=BCS_BoneSpace
    scaleMode=BMM_Additive
    scaleSpace=BCS_BoneSpace
    translationMode=BMM_Additive
    translationSpace=BCS_BoneSpace
    alphaBoolBlend {blendInTime=0.2, blendOption=Linear, blendOutTime=1}
    boneToModify {boneName=dyn_coat_bk_l_02}
    rotation {pitch=15}
    translation {z=5}
   }
   n73 = "Transform (Modify) Bone - Bone: dyn_coat_bk_l_04"(ComponentPose=@n72, bAlphaBoolEnabled=@n69, Translation="0,0,0", Rotation="-10,0,0", Scale="1,1,1") {
    alphaInputType=Bool
    rotationMode=BMM_Additive
    rotationSpace=BCS_BoneSpace
    scaleMode=BMM_Additive
    scaleSpace=BCS_BoneSpace
    alphaBoolBlend {blendInTime=0.2, blendOption=Linear, blendOutTime=1}
    boneToModify {boneName=dyn_coat_bk_l_04}
    rotation {pitch=-10}
   }
   n74 = "Transform (Modify) Bone - Bone: dyn_coat_bk_r_02"!(ComponentPose=@n73, bAlphaBoolEnabled:bool=@n69, Translation="0,0,5", Rotation="15,0,0", Scale="1,1,1") -> (Pose) {
    alphaInputType=Bool
    rotationMode=BMM_Additive
    rotationSpace=BCS_BoneSpace
    scaleMode=BMM_Additive
    scaleSpace=BCS_BoneSpace
    translationMode=BMM_Additive
    translationSpace=BCS_BoneSpace
    alphaBoolBlend {blendInTime=0.2, blendOption=Linear, blendOutTime=1}
    boneToModify {boneName=dyn_coat_bk_r_02}
    rotation {pitch=15}
    translation {z=5}
   }
   n75 = "Transform (Modify) Bone - Bone: dyn_coat_bk_r_04"(ComponentPose=@n74, bAlphaBoolEnabled=@n69, Translation="0,0,0", Rotation="-10,0,0", Scale="1,1,1") {
    alphaInputType=Bool
    rotationMode=BMM_Additive
    rotationSpace=BCS_BoneSpace
    scaleMode=BMM_Additive
    scaleSpace=BCS_BoneSpace
    alphaBoolBlend {blendInTime=0.2, blendOption=Linear, blendOutTime=1}
    boneToModify {boneName=dyn_coat_bk_r_04}
    rotation {pitch=-10}
   }
   n76 = Get ADS Up Crouch Bool() -> (ADS Up Crouch Bool:double)
   n77 = "Transform (Modify) Bone - Bone: dyn_capeShoulder_r_01"!(ComponentPose=@n75, Alpha:float=@n76, Translation="7.929,-2.256,-0.284", Rotation="-0.009,20.381,-0.000", Scale="1,1,1") -> (Pose) {
    rotationMode=BMM_Additive
    rotationSpace=BCS_BoneSpace
    scaleMode=BMM_Additive
    scaleSpace=BCS_BoneSpace
    translationMode=BMM_Additive
    translationSpace=BCS_BoneSpace
    alphaBoolBlend {blendInTime=0.5, blendOption=Linear, blendOutTime=0.1}
    alphaScaleBiasClamp {bInterpResult=true, bMapRange=true}
    boneToModify {boneName=dyn_capeShoulder_r_01}
    rotation=<rot -0.009 20.381 0>
    translation=<7.929 -2.256 -0.284>
   }
   n78 = "Transform (Modify) Bone - Bone: dyn_capeShoulder_l_01"!(ComponentPose=@n77, Alpha:float=@n76, Translation="0.082,-1.690,-0.000", Rotation="0,0,0", Scale="1,1,1") -> (Pose) {
    rotationMode=BMM_Additive
    rotationSpace=BCS_BoneSpace
    scaleMode=BMM_Additive
    scaleSpace=BCS_BoneSpace
    translationMode=BMM_Additive
    translationSpace=BCS_BoneSpace
    alphaBoolBlend {blendInTime=1, blendOption=Linear, blendOutTime=1}
    alphaScaleBiasClamp {bInterpResult=true, bMapRange=true}
    boneToModify {boneName=dyn_capeShoulder_l_01}
    translation=<0.082 -1.69>
   }
   n79 = Get bIsPlayingMeleeAnim()
   n80 = "Transform (Modify) Bone - Bone: cape_l_MSTR"!(ComponentPose=@n78, bAlphaBoolEnabled:bool=@n79, Translation="10,0,0", Rotation="0,0,0", Scale="1,1,1") -> (Pose) {
    alphaInputType=Bool
    rotationMode=BMM_Additive
    rotationSpace=BCS_BoneSpace
    scaleMode=BMM_Additive
    scaleSpace=BCS_BoneSpace
    translationMode=BMM_Additive
    translationSpace=BCS_BoneSpace
    alphaBoolBlend {blendInTime=1, blendOption=Linear, blendOutTime=3}
    alphaScaleBiasClamp {bInterpResult=true, bMapRange=true}
    boneToModify {boneName=cape_l_MSTR}
    translation {x=10}
   }
   n81 = "Transform (Modify) Bone - Bone: cape_mid_l_MSTR"!(ComponentPose=@n80, bAlphaBoolEnabled:bool=@n79, Translation="0,0,10", Rotation="0,0,0", Scale="1,1,1") -> (Pose) {
    alphaInputType=Bool
    rotationMode=BMM_Additive
    rotationSpace=BCS_BoneSpace
    scaleMode=BMM_Additive
    scaleSpace=BCS_BoneSpace
    translationMode=BMM_Additive
    translationSpace=BCS_BoneSpace
    alphaBoolBlend {blendInTime=1, blendOption=Linear, blendOutTime=3}
    alphaScaleBiasClamp {bInterpResult=true, bMapRange=true}
    boneToModify {boneName=cape_mid_l_MSTR}
    translation {z=10}
   }
   n82 = "Transform (Modify) Bone - Bone: cape_mid_r_MSTR"!(ComponentPose=@n81, bAlphaBoolEnabled:bool=@n79, Translation="0,0,10", Rotation="0,0,0", Scale="1,1,1") -> (Pose) {
    alphaInputType=Bool
    rotationMode=BMM_Additive
    rotationSpace=BCS_BoneSpace
    scaleMode=BMM_Additive
    scaleSpace=BCS_BoneSpace
    translationMode=BMM_Additive
    translationSpace=BCS_BoneSpace
    alphaBoolBlend {blendInTime=1, blendOption=Linear, blendOutTime=3}
    alphaScaleBiasClamp {bInterpResult=true, bMapRange=true}
    boneToModify {boneName=cape_mid_r_MSTR}
    translation {z=10}
   }
   n83 = "Transform (Modify) Bone - Bone: cape_r_MSTR"!(ComponentPose=@n82, bAlphaBoolEnabled:bool=@n79, Translation="10,0,0", Rotation="0,0,0", Scale="1,1,1") -> (Pose) {
    alphaInputType=Bool
    rotationMode=BMM_Additive
    rotationSpace=BCS_BoneSpace
    scaleMode=BMM_Additive
    scaleSpace=BCS_BoneSpace
    translationMode=BMM_Additive
    translationSpace=BCS_BoneSpace
    alphaBoolBlend {blendInTime=1, blendOption=Linear, blendOutTime=3}
    alphaScaleBiasClamp {bInterpResult=true, bMapRange=true}
    boneToModify {boneName=cape_r_MSTR}
    translation {x=10}
   }
   n84 = Get crouch moving() -> (crouch moving:bool)
   n85 = "Transform (Modify) Bone - Bone: dyn_beltBox_01"!(ComponentPose=@n83, bAlphaBoolEnabled:bool=@n84, Translation="-0.5,-1,4", Rotation="-8,18,-9", Scale="1,1,1") -> (Pose) {
    alphaInputType=Bool
    rotationMode=BMM_Additive
    rotationSpace=BCS_BoneSpace
    scaleMode=BMM_Additive
    scaleSpace=BCS_BoneSpace
    translationMode=BMM_Additive
    translationSpace=BCS_BoneSpace
    alphaBoolBlend {blendInTime=0.2, blendOption=Linear, blendOutTime=0.1}
    alphaScaleBiasClamp {bInterpResult=true, bMapRange=true}
    boneToModify {boneName=dyn_beltBox_01}
    rotation=<rot -8 18 -9>
    translation=<-0.5 -1 4>
   }
   n86 = "Transform (Modify) Bone - Bone: dyn_beltBox_02"!(ComponentPose=@n85, bAlphaBoolEnabled:bool=@n84, Translation="-1,5,2.5", Rotation="12,20,18", Scale="1,1,1") -> (Pose) {
    alphaInputType=Bool
    rotationMode=BMM_Additive
    rotationSpace=BCS_BoneSpace
    scaleMode=BMM_Additive
    scaleSpace=BCS_BoneSpace
    translationMode=BMM_Additive
    translationSpace=BCS_BoneSpace
    alphaBoolBlend {blendInTime=0.35, blendOption=Linear, blendOutTime=0.1}
    alphaScaleBiasClamp {bInterpResult=true, bMapRange=true}
    boneToModify {boneName=dyn_beltBox_02}
    rotation=<rot 12 20 18>
    translation=<-1 5 2.5>
   }
   n87 = "Transform (Modify) Bone - Bone: dyn_shoulderPad_r_01"!(ComponentPose=@n86, Alpha:float=Get ADS Rifle Dn() -> (ADS Rifle Dn:double), Translation="0,-1.5,-1", Rotation="-8,10,-5", Scale="1,1,1") -> (Pose) {
    rotationMode=BMM_Additive
    scaleMode=BMM_Additive
    translationMode=BMM_Additive
    alphaBoolBlend {blendOption=Linear}
    boneToModify {boneName=dyn_shoulderPad_r_01}
    rotation=<rot -8 10 -5>
    translation {y=-1.5, z=-1}
   }
   n88 = "Transform (Modify) Bone - Bone: cape_01_l"(ComponentPose=@n87, bAlphaBoolEnabled=Get bIsInFrontEnd(), Translation="-3,0,0", Rotation="0,0,0", Scale="1,1,1") {
    alphaInputType=Bool
    rotationMode=BMM_Additive
    scaleMode=BMM_Additive
    translationMode=BMM_Additive
    alphaBoolBlend {blendOption=Linear}
    boneToModify {boneName=cape_01_l}
    translation {x=-3}
   }
   n89 = "Transform (Modify) Bone - Bone: dyn_shoulderPad_r_01"!(ComponentPose=@n88, bAlphaBoolEnabled:bool=false, Translation="-0.000,-1,0.000", Rotation="1.529,27.659,-5.887", Scale="1,1,1") -> (Pose) {
    alphaInputType=Bool
    rotationMode=BMM_Additive
    scaleMode=BMM_Additive
    translationMode=BMM_Additive
    alphaBoolBlend {blendOption=Linear}
    boneToModify {boneName=dyn_shoulderPad_r_01}
    rotation=<rot 1.529 27.659 -5.887>
    translation {y=-1}
   }
   n90 = "Transform (Modify) Bone - Bone: dyn_shoulderPad_l_01"(
    ComponentPose=@n89
    bAlphaBoolEnabled=Get isCustomEmotePlaying() -> (isCustomEmotePlaying:bool)
    Translation="0,0,0"
    Rotation="0,0,-3"
    Scale="1,1,1"
   ) {
    alphaInputType=Bool
    rotationMode=BMM_Additive
    scaleMode=BMM_Additive
    translationMode=BMM_Additive
    alphaBoolBlend {blendOption=Linear}
    boneToModify {boneName=dyn_shoulderPad_l_01}
    rotation {roll=-3}
   }
   n91 = Get in vehicle scale alpha() -> (in vehicle scale alpha:double)
   n92 = "Transform (Modify) Bone - Bone: cape_mid_04_l"!(
    ComponentPose="Transform (Modify) Bone - Bone: cape_04_r"!(
     ComponentPose="Transform (Modify) Bone - Bone: cape_04_l"(ComponentPose=@n90, Alpha=@n91, Translation="0,0,0", Rotation="0,0,0", Scale="0.1,1,1") {scaleMode=BMM_Additive, alphaBoolBlend {blendOption=Linear}, boneToModify {boneName=cape_04_l}, scale {x=0.1}}
     Alpha:float=@n91
     Translation="0,0,0"
     Rotation="0,0,0"
     Scale="0.1,1,1"
    ) -> (Pose) {scaleMode=BMM_Additive, alphaBoolBlend {blendOption=Linear}, boneToModify {boneName=cape_04_r}, scale {x=0.1}}
    Alpha:float=@n91
    Translation="0,0,0"
    Rotation="0,0,0"
    Scale="0.1,1,1"
   ) -> (Pose) {scaleMode=BMM_Additive, alphaBoolBlend {blendOption=Linear}, boneToModify {boneName=cape_mid_04_l}, scale {x=0.1}}
   n93 = "Transform (Modify) Bone - Bone: cape_mid_04_r"!(ComponentPose=@n92, Alpha:float=@n91, Translation="0,0,0", Rotation="0,0,0", Scale="0.1,1,1") -> (Pose) {scaleMode=BMM_Additive, alphaBoolBlend {blendOption=Linear}, boneToModify {boneName=cape_mid_04_r}, scale {x=0.1}}
   n94 = "Transform (Modify) Bone - Bone: dyn_coat_front_l_04"(ComponentPose=@n93, Alpha=@n91, Translation="0,0,0", Rotation="0,0,0", Scale="0.1,1,1") {scaleMode=BMM_Additive, alphaBoolBlend {blendOption=Linear}, boneToModify {boneName=dyn_coat_front_l_04}, scale {x=0.1}}
   n95 = "Transform (Modify) Bone - Bone: dyn_coat_l_04"(ComponentPose=@n94, Alpha=@n91, Translation="0,0,0", Rotation="0,0,0", Scale="0.1,1,1") {scaleMode=BMM_Additive, alphaBoolBlend {blendOption=Linear}, boneToModify {boneName=dyn_coat_l_04}, scale {x=0.1}}
   n96 = "Transform (Modify) Bone - Bone: dyn_coat_bk_l_04"!(ComponentPose=@n95, Alpha:float=@n91, Translation="0,0,0", Rotation="0,0,0", Scale="0.1,1,1") -> (Pose) {scaleMode=BMM_Additive, alphaBoolBlend {blendOption=Linear}, boneToModify {boneName=dyn_coat_bk_l_04}, scale {x=0.1}}
   n97 = "Transform (Modify) Bone - Bone: dyn_coat_bk_c_04"!(ComponentPose=@n96, Alpha:float=@n91, Translation="0,0,0", Rotation="0,0,0", Scale="0.1,1,1") -> (Pose) {scaleMode=BMM_Additive, alphaBoolBlend {blendOption=Linear}, boneToModify {boneName=dyn_coat_bk_c_04}, scale {x=0.1}}
   n98 = "Transform (Modify) Bone - Bone: dyn_coat_front_r_04"(ComponentPose=@n97, Alpha=@n91, Translation="0,0,0", Rotation="0,0,0", Scale="0.1,1,1") {scaleMode=BMM_Additive, alphaBoolBlend {blendOption=Linear}, boneToModify {boneName=dyn_coat_front_r_04}, scale {x=0.1}}
   n99 = "Transform (Modify) Bone - Bone: dyn_coat_r_04"(ComponentPose=@n98, Alpha=@n91, Translation="0,0,0", Rotation="0,0,0", Scale="0.1,1,1") {scaleMode=BMM_Additive, alphaBoolBlend {blendOption=Linear}, boneToModify {boneName=dyn_coat_r_04}, scale {x=0.1}}
   n100 = "Transform (Modify) Bone - Bone: dyn_coat_bk_r_04"!(ComponentPose=@n99, Alpha:float=@n91, Translation="0,0,0", Rotation="0,0,0", Scale="0.1,1,1") -> (Pose) {scaleMode=BMM_Additive, alphaBoolBlend {blendOption=Linear}, boneToModify {boneName=dyn_coat_bk_r_04}, scale {x=0.1}}
   n101 = Get isDarkSideEmotePlaying() -> (isDarkSideEmotePlaying:bool)
   n102 = "Transform (Modify) Bone - Bone: dyn_shoulderPad_r_01"(ComponentPose=@n100, bAlphaBoolEnabled=@n101, Translation="-1,0,2.5", Rotation="5,0,15", Scale="1,0.9,1") {
    alphaInputType=Bool
    rotationMode=BMM_Additive
    rotationSpace=BCS_BoneSpace
    scaleMode=BMM_Additive
    translationMode=BMM_Additive
    translationSpace=BCS_BoneSpace
    alphaBoolBlend {blendOption=Linear}
    boneToModify {boneName=dyn_shoulderPad_r_01}
    rotation {pitch=5, roll=15}
    scale {y=0.9}
    translation {x=-1, z=2.5}
   }
   n103 = "Transform (Modify) Bone - Bone: dyn_shoulderPad_l_01"(ComponentPose=@n102, bAlphaBoolEnabled=@n101, Translation="2,0,0", Rotation="-5,-10,-12", Scale="1,1,1") {
    alphaInputType=Bool
    rotationMode=BMM_Additive
    rotationSpace=BCS_BoneSpace
    scaleMode=BMM_Additive
    translationMode=BMM_Additive
    translationSpace=BCS_BoneSpace
    alphaBoolBlend {blendOption=Linear}
    boneToModify {boneName=dyn_shoulderPad_l_01}
    rotation=<rot -5 -10 -12>
    translation {x=2}
   }
   Output Pose(Result=Component To Local(ComponentPose=@n103)) {layerGroup=None}
   n104 = "Transform (Modify) Bone - Bone: dyn_ground_root"(
    ComponentPose=Local To Component(LocalPose=Copy Pose From Mesh() {bUseAttachedParent=true})
    Alpha=Get PlummetAlpha() -> (PlummetAlpha:double)
    Translation="0,0,-600"
    Rotation="0,0,0"
    Scale="1,1,1"
   ) {
    translationMode=BMM_Additive
    translationSpace=BCS_BoneSpace
    alphaBoolBlend {blendOption=Linear}
    boneToModify {boneName=dyn_ground_root}
    translation {z=-600}
   }
   n105 = Control Rig(Source=Component To Local(ComponentPose=@n104)) {
    alphaBoolBlend {blendOption=Linear}
    controlRigAssetReference {blueprintRigClass=&/M_MED_Realm_Deform_CtrlRig_twists.M_MED_Realm_Deform_CtrlRig_twists_C}
   }
   n106 = Control Rig!(
    Source=@n105
    Alpha=Get controlRigAlpha() -> (controlRigAlpha:double)
    KneeRot_fr_l_114deg:Vector
    KneeTrans_fr_l_114deg:Vector="0.258,-3.591,0.009"
    KneeScale_fr_l_114deg:Vector="1,1,1"
    KneeScale_fr_l_130deg:Vector="1,1,1"
    KneeRot_fr_l_130deg:Vector="0,0,-51.2"
    KneeTrans_fr_l_130deg:Vector="0.258,-3.591,0.009"
   ) -> (Pose) {
    alphaBoolBlend {blendOption=Linear}
    controlRigAssetReference {blueprintRigClass=&/M_MED_Realm_Deform_CtrlRig_sculpters.M_MED_Realm_Deform_CtrlRig_sculpters_C}
   }
   Save cached pose 'controlRigCache'(Pose=@n106) {cacheName=controlRigCache}
   n107 = RigidBody!(
    ComponentPose=Local To Component(LocalPose=Use cached pose 'standingPoseCache'())
    Alpha=Get in vehicle rigid alpha() -> (in vehicle rigid alpha:double)
    OverrideWorldGravity:Vector="0,0,-980"
    ExternalForce
   ) -> (Pose) {
    bClampLinearTranslationLimitToRefPose=true
    bOverrideWorldGravity=true
    overridePhysicsAsset=&/M_MED_Realm_Cape_Physics
    simulationSpace=BaseBoneSpace
    alphaBoolBlend {blendOption=Linear}
    baseBoneRef {boneName=pelvis}
    componentAppliedLinearAccClamp=<3000 3000 3000>
    componentLinearAccScale=<0.05 0.05 0.5>
    componentLinearVelScale=<0.05 0.05 0.5>
    overrideWorldGravity {z=-980}
    simSpaceSettings {maxAngularAcceleration=50, maxAngularVelocity=1, maxLinearAcceleration=300, maxLinearVelocity=150, velocityScaleZ=0, worldAlpha=1}
   }
   n108 = Component To Local(ComponentPose=@n107)
   n109 = RigidBody(
    ComponentPose=Local To Component(LocalPose=Use cached pose 'standingPoseCache'())
    Alpha=Get in vehicle rigid alpha()
    OverrideWorldGravity=Get Gravity Override Cape() -> (Gravity Override Cape:Vector)
   ) {
    bClampLinearTranslationLimitToRefPose=true
    bOverrideWorldGravity=true
    overridePhysicsAsset=&/M_MED_RealmBackBling_Cape_Physics2
    simulationSpace=BaseBoneSpace
    alphaBoolBlend {blendOption=Linear}
    baseBoneRef {boneName=spine_05}
    componentAppliedLinearAccClamp=<3000 3000 3000>
    componentLinearAccScale=<0.05 0.05 0.5>
    componentLinearVelScale=<0.05 0.05 0.5>
    overrideWorldGravity {z=-980}
    simSpaceSettings {
     maxAngularAcceleration=50
     maxAngularVelocity=1
     maxLinearAcceleration=300
     maxLinearVelocity=150
     velocityScaleZ=0
     externalLinearDragV=<0.2 0.2>
    }
   }
   n110 = Component To Local(ComponentPose=@n109)
   n111 = Get Is Wearing Backbling()
   n112 = RigidBody(ComponentPose=Local To Component(LocalPose=Use cached pose 'standingPoseCache'()), OverrideWorldGravity=Get Gravity Override Cape()) {
    bClampLinearTranslationLimitToRefPose=true
    bOverrideWorldGravity=true
    overridePhysicsAsset=&/M_MED_Realm_CapeFrontEnd_Physics
    simulationSpace=BaseBoneSpace
    alphaBoolBlend {blendOption=Linear}
    baseBoneRef {boneName=spine_05}
    componentAppliedLinearAccClamp=<3000 3000 3000>
    componentLinearAccScale=<0.05 0.05 0.5>
    componentLinearVelScale=<0.05 0.05 0.5>
    overrideWorldGravity {z=-980}
    simSpaceSettings {
     maxAngularAcceleration=50
     maxAngularVelocity=1
     maxLinearAcceleration=300
     maxLinearVelocity=150
     velocityScaleZ=0
     externalLinearDragV=<0.2 0.2>
    }
   }
   n113 = Blend Poses by bool(
    BlendPose_0=Use cached pose 'crouchPose'()
    BlendPose_1=Blend Poses by bool(
     BlendPose_0=Component To Local(ComponentPose=@n112)
     BlendPose_1=Blend Poses by bool(BlendPose_0=@n110, BlendPose_1=@n108, bActiveValue=@n111) {blendType=Linear}
     bActiveValue=Get bIsInFrontEnd()
    ) {blendType=Linear}
    BlendTime_0=0.3
    BlendTime_1=0.7
    bActiveValue=Get isCrouchStopped()
   ) {blendType=Linear}
   Save cached pose 'postSim'(Pose=Component To Local(ComponentPose=Local To Component(LocalPose=@n113))) {cacheName=postSim}
   n114 = RigidBody!(
    ComponentPose=Local To Component(LocalPose=Use cached pose 'skirtDefaultCache'())
    bAlphaBoolEnabled:bool
    OverrideWorldGravity=Get Gravity Override Cape()
    ExternalForce
   ) -> (Pose) {
    alphaInputType=Bool
    bClampLinearTranslationLimitToRefPose=true
    bOverrideWorldGravity=true
    overridePhysicsAsset=&/M_MED_Realm_SkirtStandingFrontEnd_Physics
    simulationSpace=BaseBoneSpace
    alphaBoolBlend {blendOption=Linear}
    baseBoneRef {boneName=root}
    componentAppliedLinearAccClamp=<100 100 100>
    componentLinearAccScale=<0.05 0.05 0.1>
    componentLinearVelScale=<0.05 0.05 0.1>
    overrideWorldGravity {z=-980}
   }
   n115 = Component To Local(ComponentPose=@n114)
   n116 = RigidBody!(
    ComponentPose=Local To Component(LocalPose=Use cached pose 'skirtDefaultCache'())
    Alpha:float=Get in vehicle rigid alpha()
    OverrideWorldGravity=Get Gravity Override Cape()
    ExternalForce
   ) -> (Pose) {
    bClampLinearTranslationLimitToRefPose=true
    bOverrideWorldGravity=true
    overridePhysicsAsset=&/M_MED_Realm_SkirtStanding_Physics
    simulationSpace=BaseBoneSpace
    alphaBoolBlend {blendOption=Linear}
    baseBoneRef {boneName=root}
    componentAppliedLinearAccClamp=<100 100 100>
    componentLinearAccScale=<0.05 0.05 0.1>
    componentLinearVelScale=<0.05 0.05 0.1>
    overrideWorldGravity {z=-980}
   }
   n117 = Component To Local(ComponentPose=@n116)
   n118 = Get bIsInFrontEnd()
   n119 = RigidBody!(
    ComponentPose=Local To Component(LocalPose=Use cached pose 'skirtDefaultCache'())
    bAlphaBoolEnabled:bool
    OverrideWorldGravity=Get Gravity Override Cape()
    ExternalForce
   ) -> (Pose) {
    alphaInputType=Bool
    bClampLinearTranslationLimitToRefPose=true
    bOverrideWorldGravity=true
    overridePhysicsAsset=&/M_MED_Realm_SkirtCrouchMoving_Physics
    simulationSpace=BaseBoneSpace
    alphaBoolBlend {blendOption=Linear}
    baseBoneRef {boneName=root}
    componentAppliedLinearAccClamp=<100 100 100>
    componentLinearAccScale=<0.05 0.05 0.1>
    componentLinearVelScale=<0.05 0.05 0.1>
    overrideWorldGravity {z=-980}
   }
   n120 = Blend Poses by bool(
    BlendPose_0=Component To Local(ComponentPose=@n119)
    BlendPose_1=Blend Poses by bool(
     BlendPose_0=Use cached pose 'skirtSlidingCache'()
     BlendPose_1=Blend Poses by bool(BlendPose_0=@n115, BlendPose_1=@n117, BlendTime_0=0.3, BlendTime_1=0.3, bActiveValue=@n118) {blendType=Linear}
     BlendTime_0=0.3
     BlendTime_1=0.3
     bActiveValue=Get bIsSliding()
    ) {blendType=Linear}
    BlendTime_0=0.3
    BlendTime_1=0.3
    bActiveValue=Get crouchMovingNotSliding() -> (crouchMovingNotSliding:bool)
   ) {blendType=Linear}
   Save cached pose 'standingPoseCache'(Pose=Component To Local(ComponentPose=Local To Component(LocalPose=@n120))) {cacheName=standingPoseCache}
   n121 = "Transform (Modify) Bone - Bone: dyn_coat_front_l_01"(
    ComponentPose=Local To Component(LocalPose=Use cached pose 'controlRigCache'())
    Alpha=1
    Translation="0,3,0"
    Rotation="20,-78.117,-85"
    Scale="1,1,1"
   ) {
    rotationMode=BMM_Additive
    rotationSpace=BCS_BoneSpace
    scaleMode=BMM_Additive
    scaleSpace=BCS_BoneSpace
    translationMode=BMM_Additive
    translationSpace=BCS_BoneSpace
    alphaBoolBlend {blendOption=Linear}
    boneToModify {boneName=dyn_coat_front_l_01}
    rotation=<rot 20 -78.117 -85>
    translation {y=3}
   }
   n122 = "Transform (Modify) Bone - Bone: dyn_coat_front_l_02"(ComponentPose=@n121, Alpha=1, Translation="0,0,0", Rotation="-60.832,-4.831,0.145", Scale="1,1,1") {
    rotationMode=BMM_Additive
    rotationSpace=BCS_BoneSpace
    scaleMode=BMM_Additive
    scaleSpace=BCS_BoneSpace
    translationMode=BMM_Additive
    translationSpace=BCS_BoneSpace
    alphaBoolBlend {blendOption=Linear}
    boneToModify {boneName=dyn_coat_front_l_02}
    rotation=<rot -60.832 -4.831 0.145>
   }
   n123 = "Transform (Modify) Bone - Bone: dyn_coat_front_l_03"(ComponentPose=@n122, Alpha=1, Translation="0,0,0", Rotation="0.761,-3.678,-0.538", Scale="1,1,1") {
    rotationMode=BMM_Additive
    rotationSpace=BCS_BoneSpace
    scaleMode=BMM_Additive
    scaleSpace=BCS_BoneSpace
    translationMode=BMM_Additive
    translationSpace=BCS_BoneSpace
    alphaBoolBlend {blendOption=Linear}
    boneToModify {boneName=dyn_coat_front_l_03}
    rotation=<rot 0.761 -3.678 -0.538>
   }
   n124 = "Transform (Modify) Bone - Bone: dyn_coat_front_l_05"(ComponentPose=@n123, Alpha=1, Translation="0,0,0", Rotation="46.711,-12.031,-8.207", Scale="1,1,1") {
    rotationMode=BMM_Additive
    rotationSpace=BCS_BoneSpace
    scaleMode=BMM_Additive
    scaleSpace=BCS_BoneSpace
    translationMode=BMM_Additive
    translationSpace=BCS_BoneSpace
    alphaBoolBlend {blendOption=Linear}
    boneToModify {boneName=dyn_coat_front_l_05}
    rotation=<rot 46.711 -12.031 -8.207>
   }
   n125 = "Transform (Modify) Bone - Bone: dyn_coat_front_l_06"(ComponentPose=@n124, Alpha=1, Translation="0,0,0", Rotation="24.981,-8.676,-6.168", Scale="1,1,1") {
    rotationMode=BMM_Additive
    rotationSpace=BCS_BoneSpace
    translationMode=BMM_Additive
    translationSpace=BCS_BoneSpace
    alphaBoolBlend {blendOption=Linear}
    boneToModify {boneName=dyn_coat_front_l_06}
    rotation=<rot 24.981 -8.676 -6.168>
   }
   n126 = "Transform (Modify) Bone - Bone: dyn_coat_l_01"(ComponentPose=@n125, Alpha=1, Translation="-4,0,0", Rotation="8,-12,-60", Scale="1,1,1") {
    rotationMode=BMM_Additive
    rotationSpace=BCS_BoneSpace
    scaleMode=BMM_Additive
    scaleSpace=BCS_BoneSpace
    translationMode=BMM_Additive
    translationSpace=BCS_BoneSpace
    alphaBoolBlend {blendOption=Linear}
    boneToModify {boneName=dyn_coat_l_01}
    rotation=<rot 8 -12 -60>
    translation {x=-4}
   }
   n127 = "Transform (Modify) Bone - Bone: dyn_coat_l_02"(ComponentPose=@n126, Alpha=1, Translation="0,0,0", Rotation="-11.103,-6.600,1.274", Scale="1,1,1") {
    rotationMode=BMM_Additive
    rotationSpace=BCS_BoneSpace
    scaleMode=BMM_Additive
    scaleSpace=BCS_BoneSpace
    translationMode=BMM_Additive
    translationSpace=BCS_BoneSpace
    alphaBoolBlend {blendOption=Linear}
    boneToModify {boneName=dyn_coat_l_02}
    rotation=<rot -11.103 -6.6 1.274>
   }
   n128 = "Transform (Modify) Bone - Bone: dyn_coat_l_04"(ComponentPose=@n127, Alpha=1, Translation="0,0,0", Rotation="18.060,0.026,-0.009", Scale="1,1,1") {
    rotationMode=BMM_Additive
    rotationSpace=BCS_BoneSpace
    scaleMode=BMM_Additive
    scaleSpace=BCS_BoneSpace
    translationMode=BMM_Additive
    translationSpace=BCS_BoneSpace
    alphaBoolBlend {blendOption=Linear}
    boneToModify {boneName=dyn_coat_l_04}
    rotation=<rot 18.06 0.026 -0.009>
   }
   n129 = "Transform (Modify) Bone - Bone: dyn_coat_l_05"(ComponentPose=@n128, Alpha=1, Translation="0,0,0", Rotation="67.291,0.075,2.919", Scale="1,1,1") {
    rotationMode=BMM_Additive
    rotationSpace=BCS_BoneSpace
    scaleMode=BMM_Additive
    scaleSpace=BCS_BoneSpace
    translationMode=BMM_Additive
    translationSpace=BCS_BoneSpace
    alphaBoolBlend {blendOption=Linear}
    boneToModify {boneName=dyn_coat_l_05}
    rotation=<rot 67.291 0.075 2.919>
   }
   n130 = "Transform (Modify) Bone - Bone: dyn_coat_bk_l_02"!(ComponentPose=@n129, Alpha:float=1, Translation="0,0,0", Rotation="0.001,-7.847,0.001", Scale="1,1,1") -> (Pose) {
    rotationMode=BMM_Additive
    rotationSpace=BCS_BoneSpace
    scaleMode=BMM_Additive
    scaleSpace=BCS_BoneSpace
    translationMode=BMM_Additive
    translationSpace=BCS_BoneSpace
    alphaBoolBlend {blendOption=Linear}
    boneToModify {boneName=dyn_coat_bk_l_02}
    rotation=<rot 0.001 -7.847 0.001>
   }
   n131 = "Transform (Modify) Bone - Bone: dyn_coat_bk_l_04"(ComponentPose=@n130, Alpha=1, Translation="0,0,0", Rotation="49.784,-0.006,-0.042", Scale="1,1,1") {
    rotationMode=BMM_Additive
    rotationSpace=BCS_BoneSpace
    scaleMode=BMM_Additive
    scaleSpace=BCS_BoneSpace
    translationMode=BMM_Additive
    translationSpace=BCS_BoneSpace
    alphaBoolBlend {blendOption=Linear}
    boneToModify {boneName=dyn_coat_bk_l_04}
    rotation=<rot 49.784 -0.006 -0.042>
   }
   n132 = "Transform (Modify) Bone - Bone: dyn_coat_bk_l_05"(ComponentPose=@n131, Alpha=1, Translation="0,0,0", Rotation="40.896,-0.398,0.903", Scale="1,1,1") {
    rotationMode=BMM_Additive
    rotationSpace=BCS_BoneSpace
    scaleMode=BMM_Additive
    scaleSpace=BCS_BoneSpace
    translationMode=BMM_Additive
    translationSpace=BCS_BoneSpace
    alphaBoolBlend {blendOption=Linear}
    boneToModify {boneName=dyn_coat_bk_l_05}
    rotation=<rot 40.896 -0.398 0.903>
   }
   n133 = "Transform (Modify) Bone - Bone: dyn_coat_bk_l_06"(ComponentPose=@n132, Alpha=1, Translation="0,0,0", Rotation="0,0,0", Scale="1,1,1") {
    rotationMode=BMM_Additive
    rotationSpace=BCS_BoneSpace
    scaleMode=BMM_Additive
    scaleSpace=BCS_BoneSpace
    translationMode=BMM_Additive
    translationSpace=BCS_BoneSpace
    alphaBoolBlend {blendOption=Linear}
    boneToModify {boneName=dyn_coat_bk_l_06}
   }
   n134 = "Transform (Modify) Bone - Bone: dyn_coat_bk_c_03"!(ComponentPose=@n133, Alpha:float=1, Translation="0,0,0", Rotation="14.531,0.009,0.011", Scale="1,1,1") -> (Pose) {
    rotationMode=BMM_Additive
    rotationSpace=BCS_BoneSpace
    scaleMode=BMM_Additive
    scaleSpace=BCS_BoneSpace
    translationMode=BMM_Additive
    translationSpace=BCS_BoneSpace
    alphaBoolBlend {blendOption=Linear}
    boneToModify {boneName=dyn_coat_bk_c_03}
    rotation=<rot 14.531 0.009 0.011>
   }
   n135 = "Transform (Modify) Bone - Bone: dyn_coat_bk_c_04"(ComponentPose=@n134, Alpha=1, Translation="0,0,0", Rotation="46.462,-0.012,-0.002", Scale="1,1,1") {
    rotationMode=BMM_Additive
    rotationSpace=BCS_BoneSpace
    scaleMode=BMM_Additive
    scaleSpace=BCS_BoneSpace
    translationMode=BMM_Additive
    translationSpace=BCS_BoneSpace
    alphaBoolBlend {blendOption=Linear}
    boneToModify {boneName=dyn_coat_bk_c_04}
    rotation=<rot 46.462 -0.012 -0.002>
   }
   n136 = "Transform (Modify) Bone - Bone: dyn_coat_bk_c_05"(ComponentPose=@n135, Alpha=1, Translation="0,0,0", Rotation="35.023,0.372,2.256", Scale="1,1,1") {
    rotationMode=BMM_Additive
    rotationSpace=BCS_BoneSpace
    scaleMode=BMM_Additive
    scaleSpace=BCS_BoneSpace
    translationMode=BMM_Additive
    translationSpace=BCS_BoneSpace
    alphaBoolBlend {blendOption=Linear}
    boneToModify {boneName=dyn_coat_bk_c_05}
    rotation=<rot 35.023 0.372 2.256>
   }
   n137 = "Transform (Modify) Bone - Bone: dyn_coat_bk_c_06"(ComponentPose=@n136, Alpha=1, Translation="0,0,0", Rotation="0,0,0", Scale="1,1,1") {
    rotationMode=BMM_Additive
    rotationSpace=BCS_BoneSpace
    scaleMode=BMM_Additive
    scaleSpace=BCS_BoneSpace
    translationMode=BMM_Additive
    translationSpace=BCS_BoneSpace
    alphaBoolBlend {blendOption=Linear}
    boneToModify {boneName=dyn_coat_bk_c_06}
   }
   n138 = "Transform (Modify) Bone - Bone: dyn_coat_front_r_01"(ComponentPose=@n137, Alpha=1, Translation="0,0,0", Rotation="12.650,-29.725,-18.944", Scale="1,1,1") {
    rotationMode=BMM_Additive
    rotationSpace=BCS_BoneSpace
    scaleMode=BMM_Additive
    scaleSpace=BCS_BoneSpace
    translationMode=BMM_Additive
    translationSpace=BCS_BoneSpace
    alphaBoolBlend {blendOption=Linear}
    boneToModify {boneName=dyn_coat_front_r_01}
    rotation=<rot 12.65 -29.725 -18.944>
   }
   n139 = "Transform (Modify) Bone - Bone: dyn_coat_front_r_02"(ComponentPose=@n138, Alpha=1, Translation="0,0,0", Rotation="0.740,-4.473,0", Scale="1,1,1") {
    rotationMode=BMM_Additive
    rotationSpace=BCS_BoneSpace
    scaleMode=BMM_Additive
    scaleSpace=BCS_BoneSpace
    translationMode=BMM_Additive
    translationSpace=BCS_BoneSpace
    alphaBoolBlend {blendOption=Linear}
    boneToModify {boneName=dyn_coat_front_r_02}
    rotation {pitch=0.74, yaw=-4.473}
   }
   n140 = "Transform (Modify) Bone - Bone: dyn_coat_front_r_03"(ComponentPose=@n139, Alpha=1, Translation="0,0,0", Rotation="-8.654,0.015,0.009", Scale="1,1,1") {
    rotationMode=BMM_Additive
    rotationSpace=BCS_BoneSpace
    scaleMode=BMM_Additive
    scaleSpace=BCS_BoneSpace
    translationMode=BMM_Additive
    translationSpace=BCS_BoneSpace
    alphaBoolBlend {blendOption=Linear}
    boneToModify {boneName=dyn_coat_front_r_03}
    rotation=<rot -8.654 0.015 0.009>
   }
   n141 = "Transform (Modify) Bone - Bone: dyn_coat_front_r_04"(ComponentPose=@n140, Alpha=1, Translation="0,0,0", Rotation="17.183,-13.904,-1.581", Scale="1,1,1") {
    rotationMode=BMM_Additive
    rotationSpace=BCS_BoneSpace
    scaleMode=BMM_Additive
    scaleSpace=BCS_BoneSpace
    translationMode=BMM_Additive
    translationSpace=BCS_BoneSpace
    alphaBoolBlend {blendOption=Linear}
    boneToModify {boneName=dyn_coat_front_r_04}
    rotation=<rot 17.183 -13.904 -1.581>
   }
   n142 = "Transform (Modify) Bone - Bone: dyn_coat_front_r_05"(ComponentPose=@n141, Alpha=1, Translation="0,0,0", Rotation="15.453,-0.561,-6.423", Scale="1,1,1") {
    rotationMode=BMM_Additive
    rotationSpace=BCS_BoneSpace
    scaleMode=BMM_Additive
    scaleSpace=BCS_BoneSpace
    translationMode=BMM_Additive
    translationSpace=BCS_BoneSpace
    alphaBoolBlend {blendOption=Linear}
    boneToModify {boneName=dyn_coat_front_r_05}
    rotation=<rot 15.453 -0.561 -6.423>
   }
   n143 = "Transform (Modify) Bone - Bone: dyn_coat_r_01"(ComponentPose=@n142, Alpha=1, Translation="0,0,0", Rotation="7.235,-4.132,-0.000", Scale="1,1,1") {
    rotationMode=BMM_Additive
    rotationSpace=BCS_BoneSpace
    scaleMode=BMM_Additive
    scaleSpace=BCS_BoneSpace
    translationMode=BMM_Additive
    translationSpace=BCS_BoneSpace
    alphaBoolBlend {blendOption=Linear}
    boneToModify {boneName=dyn_coat_r_01}
    rotation=<rot 7.235 -4.132 0>
   }
   n144 = "Transform (Modify) Bone - Bone: dyn_coat_r_02"(ComponentPose=@n143, Alpha=1, Translation="0,0,0", Rotation="5.238,-4.652,-0.232", Scale="1,1,1") {
    rotationMode=BMM_Additive
    rotationSpace=BCS_BoneSpace
    scaleMode=BMM_Additive
    scaleSpace=BCS_BoneSpace
    translationMode=BMM_Additive
    translationSpace=BCS_BoneSpace
    alphaBoolBlend {blendOption=Linear}
    boneToModify {boneName=dyn_coat_r_02}
    rotation=<rot 5.238 -4.652 -0.232>
   }
   n145 = "Transform (Modify) Bone - Bone: dyn_coat_r_03"(ComponentPose=@n144, Alpha=1, Translation="0,0,0", Rotation="7.773,-9.833,-2.686", Scale="1,1,1") {
    rotationMode=BMM_Additive
    rotationSpace=BCS_BoneSpace
    scaleMode=BMM_Additive
    scaleSpace=BCS_BoneSpace
    translationMode=BMM_Additive
    translationSpace=BCS_BoneSpace
    alphaBoolBlend {blendOption=Linear}
    boneToModify {boneName=dyn_coat_r_03}
    rotation=<rot 7.773 -9.833 -2.686>
   }
   n146 = "Transform (Modify) Bone - Bone: dyn_coat_r_04"(ComponentPose=@n145, Alpha=1, Translation="0,0,0", Rotation="16.472,-10.258,-2.764", Scale="1,1,1") {
    rotationMode=BMM_Additive
    rotationSpace=BCS_BoneSpace
    scaleMode=BMM_Additive
    scaleSpace=BCS_BoneSpace
    translationMode=BMM_Additive
    translationSpace=BCS_BoneSpace
    alphaBoolBlend {blendOption=Linear}
    boneToModify {boneName=dyn_coat_r_04}
    rotation=<rot 16.472 -10.258 -2.764>
   }
   n147 = "Transform (Modify) Bone - Bone: dyn_coat_r_05"(ComponentPose=@n146, Alpha=1, Translation="0,0,0", Rotation="16.472,-10.258,-2.764", Scale="1,1,1") {
    rotationMode=BMM_Additive
    rotationSpace=BCS_BoneSpace
    scaleMode=BMM_Additive
    scaleSpace=BCS_BoneSpace
    translationMode=BMM_Additive
    translationSpace=BCS_BoneSpace
    alphaBoolBlend {blendOption=Linear}
    boneToModify {boneName=dyn_coat_r_05}
    rotation=<rot 16.472 -10.258 -2.764>
   }
   n148 = "Transform (Modify) Bone - Bone: dyn_coat_bk_r_03"!(ComponentPose=@n147, Alpha:float=1, Translation="0,0,0", Rotation="66.850,-4.107,-0.014", Scale="1,1,1") -> (Pose) {
    rotationMode=BMM_Additive
    rotationSpace=BCS_BoneSpace
    scaleMode=BMM_Additive
    scaleSpace=BCS_BoneSpace
    translationMode=BMM_Additive
    translationSpace=BCS_BoneSpace
    alphaBoolBlend {blendOption=Linear}
    boneToModify {boneName=dyn_coat_bk_r_03}
    rotation=<rot 66.85 -4.107 -0.014>
   }
   n149 = "Transform (Modify) Bone - Bone: dyn_coat_bk_r_04"(ComponentPose=@n148, Alpha=1, Translation="0,0,0", Rotation="7.235,-4.132,-0.000", Scale="1,1,1") {
    rotationMode=BMM_Additive
    rotationSpace=BCS_BoneSpace
    scaleMode=BMM_Additive
    scaleSpace=BCS_BoneSpace
    translationMode=BMM_Additive
    translationSpace=BCS_BoneSpace
    alphaBoolBlend {blendOption=Linear}
    boneToModify {boneName=dyn_coat_bk_r_04}
    rotation=<rot 7.235 -4.132 0>
   }
   n150 = "Transform (Modify) Bone - Bone: dyn_coat_bk_r_05"(ComponentPose=@n149, Alpha=1, Translation="0,0,0", Rotation="0,0,0", Scale="1,1,1") {
    rotationMode=BMM_Additive
    rotationSpace=BCS_BoneSpace
    scaleMode=BMM_Additive
    scaleSpace=BCS_BoneSpace
    translationMode=BMM_Additive
    translationSpace=BCS_BoneSpace
    alphaBoolBlend {blendOption=Linear}
    boneToModify {boneName=dyn_coat_bk_r_05}
   }
   n151 = "Transform (Modify) Bone - Bone: dyn_coat_bk_r_06"(ComponentPose=@n150, Alpha=1, Translation="0,0,0", Rotation="0,0,0", Scale="1,1,1") {
    rotationMode=BMM_Additive
    rotationSpace=BCS_BoneSpace
    scaleMode=BMM_Additive
    scaleSpace=BCS_BoneSpace
    translationMode=BMM_Additive
    translationSpace=BCS_BoneSpace
    alphaBoolBlend {blendOption=Linear}
    boneToModify {boneName=dyn_coat_bk_r_06}
   }
   n152 = "Transform (Modify) Bone - Bone: dyn_lightSaber_01"!(ComponentPose=@n151, Alpha:float=1, Translation="0,0,0", Rotation="25,-45,-40", Scale="1,1,1") -> (Pose) {
    rotationMode=BMM_Additive
    rotationSpace=BCS_BoneSpace
    scaleMode=BMM_Additive
    scaleSpace=BCS_BoneSpace
    translationMode=BMM_Additive
    translationSpace=BCS_BoneSpace
    alphaBoolBlend {blendOption=Linear}
    boneToModify {boneName=dyn_lightSaber_01}
    rotation=<rot 25 -45 -40>
   }
   n153 = "Transform (Modify) Bone - Bone: dyn_lightSaber_02"!(ComponentPose=@n152, Alpha:float=1, Translation="0,0,0", Rotation="-27,0,0", Scale="1,1,1") -> (Pose) {
    rotationMode=BMM_Additive
    rotationSpace=BCS_BoneSpace
    scaleMode=BMM_Additive
    scaleSpace=BCS_BoneSpace
    translationMode=BMM_Additive
    translationSpace=BCS_BoneSpace
    alphaBoolBlend {blendOption=Linear}
    boneToModify {boneName=dyn_lightSaber_02}
    rotation {pitch=-27}
   }
   n154 = RigidBody!(ComponentPose=@n153, Alpha:float, OverrideWorldGravity=Get Gravity Override Cape(), ExternalForce) -> (Pose) {
    bClampLinearTranslationLimitToRefPose=true
    bOverrideWorldGravity=true
    overridePhysicsAsset=&/M_MED_Realm_SkirtCrouching_Physics
    simulationSpace=BaseBoneSpace
    alphaBoolBlend {blendOption=Linear}
    baseBoneRef {boneName=root}
    componentAppliedLinearAccClamp=<100 100 100>
    componentLinearAccScale=<0.1 0.1 0.1>
    componentLinearVelScale=<0.1 0.1 0.1>
    overrideWorldGravity {z=-980}
   }
   n155 = Anim Dynamics - Bone: dyn_coat_bk_r_06(ComponentPose=@n154, Alpha=1, ExternalForce="0,0,0") {
    angularDampingOverride=1
    angularSpringConstant=20000
    bAngularSpring=true
    bOverrideAngularDamping=true
    bOverrideLinearDamping=true
    bUseGravityOverride=true
    gravityScale=0.35
    linearDampingOverride=0.7
    simulationSpace=BoneRelative
    alphaBoolBlend {blendOption=Linear}
    boundBone {boneName=dyn_coat_bk_r_06}
    componentLinearAccScale=<1 1 1>
    componentLinearVelScale=<1 1 1>
    gravityOverride {z=-980}
    physicsBodyDefinitions [{
      boundBone {boneName=dyn_coat_bk_r_06}
      boxExtents=<15 30 30>
      constraintSetup {angularLimitsMax=<0 0 0>, angularLimitsMin=<0 0 0>, angularTarget=<0 0 0>}
      localJointOffset=<-10 0 0>
     }]
    relativeSpaceBone {boneName=root}
   }
   n156 = Anim Dynamics - Bone: dyn_coat_r_06(ComponentPose=@n155, Alpha=1, ExternalForce="0,0,0") {
    angularDampingOverride=1
    angularSpringConstant=20000
    bAngularSpring=true
    bOverrideAngularDamping=true
    bOverrideLinearDamping=true
    bUseGravityOverride=true
    gravityScale=0.35
    linearDampingOverride=0.7
    simulationSpace=BoneRelative
    alphaBoolBlend {blendOption=Linear}
    boundBone {boneName=dyn_coat_r_06}
    componentLinearAccScale=<1 1 1>
    componentLinearVelScale=<1 1 1>
    gravityOverride {z=-980}
    physicsBodyDefinitions [{
      boundBone {boneName=dyn_coat_r_06}
      boxExtents=<15 30 30>
      constraintSetup {angularLimitsMax=<0 0 0>, angularLimitsMin=<0 0 0>, angularTarget=<0 0 0>}
      localJointOffset=<-10 0 0>
     }]
    relativeSpaceBone {boneName=root}
   }
   n157 = Anim Dynamics - Bone: dyn_coat_front_r_06(ComponentPose=@n156, Alpha=1, ExternalForce="0,0,0") {
    angularDampingOverride=1
    angularSpringConstant=20000
    bAngularSpring=true
    bOverrideAngularDamping=true
    bOverrideLinearDamping=true
    bUseGravityOverride=true
    gravityScale=0.35
    linearDampingOverride=0.7
    simulationSpace=BoneRelative
    alphaBoolBlend {blendOption=Linear}
    boundBone {boneName=dyn_coat_front_r_06}
    componentLinearAccScale=<1 1 1>
    componentLinearVelScale=<1 1 1>
    gravityOverride {z=-980}
    physicsBodyDefinitions [{
      boundBone {boneName=dyn_coat_front_r_06}
      boxExtents=<15 30 30>
      constraintSetup {angularLimitsMax=<0 0 0>, angularLimitsMin=<0 0 0>, angularTarget=<0 0 0>}
      localJointOffset=<-10 0 0>
     }]
    relativeSpaceBone {boneName=root}
   }
   n158 = Anim Dynamics - Bone: dyn_coat_bk_c_06(ComponentPose=@n157, Alpha=1, ExternalForce="0,0,0") {
    angularDampingOverride=1
    angularSpringConstant=20000
    bAngularSpring=true
    bOverrideAngularDamping=true
    bOverrideLinearDamping=true
    bUseGravityOverride=true
    gravityScale=0.35
    linearDampingOverride=0.7
    simulationSpace=BoneRelative
    alphaBoolBlend {blendOption=Linear}
    boundBone {boneName=dyn_coat_bk_c_06}
    componentLinearAccScale=<1 1 1>
    componentLinearVelScale=<1 1 1>
    gravityOverride {z=-980}
    physicsBodyDefinitions [{
      boundBone {boneName=dyn_coat_bk_c_06}
      boxExtents=<15 30 30>
      constraintSetup {angularLimitsMax=<0 0 0>, angularLimitsMin=<0 0 0>, angularTarget=<0 0 0>}
      localJointOffset=<10 0 0>
     }]
    relativeSpaceBone {boneName=root}
   }
   n159 = Anim Dynamics - Bone: dyn_coat_bk_l_06(ComponentPose=@n158, Alpha=1, ExternalForce="0,0,0") {
    angularDampingOverride=1
    angularSpringConstant=20000
    bAngularSpring=true
    bOverrideAngularDamping=true
    bOverrideLinearDamping=true
    bUseGravityOverride=true
    gravityScale=0.35
    linearDampingOverride=0.7
    simulationSpace=BoneRelative
    alphaBoolBlend {blendOption=Linear}
    boundBone {boneName=dyn_coat_bk_l_06}
    componentLinearAccScale=<1 1 1>
    componentLinearVelScale=<1 1 1>
    gravityOverride {z=-980}
    physicsBodyDefinitions [{
      boundBone {boneName=dyn_coat_bk_l_06}
      boxExtents=<15 30 30>
      constraintSetup {angularLimitsMax=<0 0 0>, angularLimitsMin=<0 0 0>, angularTarget=<0 0 0>}
      localJointOffset=<10 0 0>
     }]
    relativeSpaceBone {boneName=root}
   }
   n160 = Anim Dynamics - Bone: dyn_coat_l_06(ComponentPose=@n159, Alpha=1, ExternalForce="0,0,0") {
    angularDampingOverride=1
    angularSpringConstant=20000
    bAngularSpring=true
    bOverrideAngularDamping=true
    bOverrideLinearDamping=true
    bUseGravityOverride=true
    gravityScale=0.35
    linearDampingOverride=0.7
    simulationSpace=BoneRelative
    alphaBoolBlend {blendOption=Linear}
    boundBone {boneName=dyn_coat_l_06}
    componentLinearAccScale=<1 1 1>
    componentLinearVelScale=<1 1 1>
    gravityOverride {z=-980}
    physicsBodyDefinitions [{
      boundBone {boneName=dyn_coat_l_06}
      boxExtents=<15 30 30>
      constraintSetup {angularLimitsMax=<0 0 0>, angularLimitsMin=<0 0 0>, angularTarget=<0 0 0>}
      localJointOffset=<10 0 0>
     }]
    relativeSpaceBone {boneName=root}
   }
   n161 = Anim Dynamics - Bone: dyn_coat_front_l_06(ComponentPose=@n160, Alpha=1, ExternalForce="0,0,0") {
    angularDampingOverride=1
    angularSpringConstant=20000
    bAngularSpring=true
    bOverrideAngularDamping=true
    bOverrideLinearDamping=true
    bUseGravityOverride=true
    gravityScale=0.35
    linearDampingOverride=0.7
    simulationSpace=BoneRelative
    alphaBoolBlend {blendOption=Linear}
    boundBone {boneName=dyn_coat_front_l_06}
    componentLinearAccScale=<1 1 1>
    componentLinearVelScale=<1 1 1>
    gravityOverride {z=-980}
    physicsBodyDefinitions [{
      boundBone {boneName=dyn_coat_front_l_06}
      boxExtents=<15 30 30>
      constraintSetup {angularLimitsMax=<0 0 0>, angularLimitsMin=<0 0 0>, angularTarget=<0 0 0>}
      localJointOffset=<10 0 0>
     }]
    relativeSpaceBone {boneName=root}
   }
   n162 = Get ADS Up Rifle() -> (ADS Up Rifle:double)
   n163 = "Transform (Modify) Bone - Bone: cape_01_l"!(ComponentPose=@n161, Alpha:float=@n162, Translation="0,0,0", Rotation="15,-80,0", Scale="1,1,1") -> (Pose) {
    rotationMode=BMM_Additive
    rotationSpace=BCS_BoneSpace
    scaleMode=BMM_Additive
    scaleSpace=BCS_BoneSpace
    translationMode=BMM_Additive
    translationSpace=BCS_BoneSpace
    alphaBoolBlend {blendInTime=1, blendOption=Linear, blendOutTime=1}
    boneToModify {boneName=cape_01_l}
    rotation {pitch=15, yaw=-80}
   }
   n164 = "Transform (Modify) Bone - Bone: cape_mid_01_l"!(ComponentPose=@n163, Alpha:float=@n162, Translation="0,0,0", Rotation="0,0,0", Scale="1,1,1") -> (Pose) {
    rotationMode=BMM_Additive
    rotationSpace=BCS_BoneSpace
    scaleMode=BMM_Additive
    scaleSpace=BCS_BoneSpace
    translationMode=BMM_Additive
    translationSpace=BCS_BoneSpace
    alphaBoolBlend {blendInTime=1, blendOption=Linear, blendOutTime=1}
    alphaScaleBiasClamp {bInterpResult=true, interpSpeedDecreasing=30}
    boneToModify {boneName=cape_mid_01_l}
   }
   n165 = "Transform (Modify) Bone - Bone: cape_mid_01_l"(ComponentPose=@n164, Alpha=@n162, Translation="0,0,0", Rotation="0,0,0", Scale="1,1,1") {
    rotationMode=BMM_Additive
    rotationSpace=BCS_BoneSpace
    scaleMode=BMM_Additive
    scaleSpace=BCS_BoneSpace
    translationMode=BMM_Additive
    translationSpace=BCS_BoneSpace
    alphaBoolBlend {blendInTime=1, blendOption=Linear, blendOutTime=1}
    alphaScaleBiasClamp {bInterpResult=true, interpSpeedDecreasing=30}
    boneToModify {boneName=cape_mid_01_l}
   }
   n166 = "Transform (Modify) Bone - Bone: cape_mid_01_l"(ComponentPose=@n165, Alpha=@n162, Translation="0,0,0", Rotation="0,0,0", Scale="1,1,1") {
    rotationMode=BMM_Additive
    rotationSpace=BCS_BoneSpace
    scaleMode=BMM_Additive
    scaleSpace=BCS_BoneSpace
    translationMode=BMM_Additive
    translationSpace=BCS_BoneSpace
    alphaBoolBlend {blendInTime=1, blendOption=Linear, blendOutTime=1}
    alphaScaleBiasClamp {bInterpResult=true, interpSpeedDecreasing=30}
    boneToModify {boneName=cape_mid_01_l}
   }
   n167 = RigidBody(ComponentPose=@n166, OverrideWorldGravity=Get Gravity Override Cape()) {
    bClampLinearTranslationLimitToRefPose=true
    bOverrideWorldGravity=true
    overridePhysicsAsset=&/M_MED_Realm_Cape_PhysicsCrouching
    simulationSpace=BaseBoneSpace
    alphaBoolBlend {blendOption=Linear}
    alphaScaleBiasClamp {bInterpResult=true, interpSpeedDecreasing=30}
    baseBoneRef {boneName=root}
    componentAppliedLinearAccClamp=<3000 3000 3000>
    componentLinearAccScale=<0.05 0.05 0.5>
    componentLinearVelScale=<0.05 0.05 0.5>
    overrideWorldGravity {z=-980}
    simSpaceSettings {
     maxAngularAcceleration=50
     maxAngularVelocity=1
     maxLinearAcceleration=300
     maxLinearVelocity=150
     velocityScaleZ=0
     externalLinearDragV=<0.2 0.2>
    }
   }
   Save cached pose 'crouchPose'(Pose=Component To Local(ComponentPose=@n167)) {cacheName=crouchPose}
   n168 = Get is not front end() -> (is not front end:double)
   n169 = "Transform (Modify) Bone - Bone: dyn_coat_front_l_01"(
    ComponentPose=Local To Component(LocalPose=Use cached pose 'controlRigCache'())
    Alpha=@n168
    Translation="0,0,0"
    Rotation="14.442,-9.634,-2.425"
    Scale="1,1,1"
   ) {
    rotationMode=BMM_Additive
    rotationSpace=BCS_BoneSpace
    translationMode=BMM_Additive
    translationSpace=BCS_BoneSpace
    alphaBoolBlend {blendOption=Linear}
    boneToModify {boneName=dyn_coat_front_l_01}
    rotation=<rot 14.4421 -9.634 -2.425>
   }
   n170 = "Transform (Modify) Bone - Bone: dyn_coat_front_l_02"(ComponentPose=@n169, Alpha=@n168, Translation="0,0,0", Rotation="-0.001,-4.831,0.005", Scale="1,1,1") {
    rotationMode=BMM_Additive
    rotationSpace=BCS_BoneSpace
    translationMode=BMM_Additive
    translationSpace=BCS_BoneSpace
    alphaBoolBlend {blendOption=Linear}
    boneToModify {boneName=dyn_coat_front_l_02}
    rotation=<rot -0.001 -4.831 0.005>
   }
   n171 = "Transform (Modify) Bone - Bone: dyn_coat_front_l_03"(ComponentPose=@n170, Alpha=@n168, Translation="0,0,0", Rotation="-4.621,-0.002,-0.000", Scale="1,1,1") {
    rotationMode=BMM_Additive
    rotationSpace=BCS_BoneSpace
    translationMode=BMM_Additive
    translationSpace=BCS_BoneSpace
    alphaBoolBlend {blendOption=Linear}
    boneToModify {boneName=dyn_coat_front_l_03}
    rotation {pitch=-4.621, yaw=-0.002}
   }
   n172 = "Transform (Modify) Bone - Bone: dyn_coat_l_01"(ComponentPose=@n171, Alpha=@n168, Translation="0,0,0", Rotation="0,0,0", Scale="1,1,1") {
    rotationMode=BMM_Additive
    rotationSpace=BCS_BoneSpace
    translationMode=BMM_Additive
    translationSpace=BCS_BoneSpace
    alphaBoolBlend {blendOption=Linear}
    boneToModify {boneName=dyn_coat_l_01}
   }
   n173 = "Transform (Modify) Bone - Bone: dyn_coat_l_02"(ComponentPose=@n172, Alpha=@n168, Translation="0,0,0", Rotation="0,0,0", Scale="1,1,1") {
    rotationMode=BMM_Additive
    rotationSpace=BCS_BoneSpace
    translationMode=BMM_Additive
    translationSpace=BCS_BoneSpace
    alphaBoolBlend {blendOption=Linear}
    boneToModify {boneName=dyn_coat_l_02}
   }
   n174 = "Transform (Modify) Bone - Bone: dyn_coat_l_03"(ComponentPose=@n173, Alpha=@n168, Translation="0,0,0", Rotation="0,0,0", Scale="1,1,1") {
    rotationMode=BMM_Additive
    rotationSpace=BCS_BoneSpace
    translationMode=BMM_Additive
    translationSpace=BCS_BoneSpace
    alphaBoolBlend {blendOption=Linear}
    boneToModify {boneName=dyn_coat_l_03}
   }
   n175 = Get is not front end()
   n176 = "Transform (Modify) Bone - Bone: dyn_coat_front_r_01"(ComponentPose=@n174, Alpha=@n175, Translation="0,0,0", Rotation="0.021,-7.765,-12.533", Scale="1,1,1") {
    rotationMode=BMM_Additive
    rotationSpace=BCS_BoneSpace
    translationMode=BMM_Additive
    translationSpace=BCS_BoneSpace
    alphaBoolBlend {blendOption=Linear}
    boneToModify {boneName=dyn_coat_front_r_01}
    rotation=<rot 0.021 -7.765 -12.533>
   }
   n177 = "Transform (Modify) Bone - Bone: dyn_coat_front_r_02"(ComponentPose=@n176, Alpha=@n175, Translation="0,0,0", Rotation="0.740,-4.473,-0.000", Scale="1,1,1") {
    rotationMode=BMM_Additive
    rotationSpace=BCS_BoneSpace
    translationMode=BMM_Additive
    translationSpace=BCS_BoneSpace
    alphaBoolBlend {blendOption=Linear}
    boneToModify {boneName=dyn_coat_front_r_02}
    rotation=<rot 0.74 -4.473 0>
   }
   n178 = "Transform (Modify) Bone - Bone: dyn_coat_front_r_03"(ComponentPose=@n177, Alpha=@n175, Translation="0,0,0", Rotation="0,0,0", Scale="1,1,1") {
    rotationMode=BMM_Additive
    rotationSpace=BCS_BoneSpace
    translationMode=BMM_Additive
    translationSpace=BCS_BoneSpace
    alphaBoolBlend {blendOption=Linear}
    boneToModify {boneName=dyn_coat_front_r_03}
   }
   n179 = "Transform (Modify) Bone - Bone: dyn_coat_r_01"(ComponentPose=@n178, Alpha=@n175, Translation="0,0,0", Rotation="2.000,-4.128,-0.000", Scale="1,1,1") {
    rotationMode=BMM_Additive
    rotationSpace=BCS_BoneSpace
    translationMode=BMM_Additive
    translationSpace=BCS_BoneSpace
    alphaBoolBlend {blendOption=Linear}
    boneToModify {boneName=dyn_coat_r_01}
    rotation=<rot 2 -4.128 0>
   }
   n180 = "Transform (Modify) Bone - Bone: dyn_coat_r_02"(ComponentPose=@n179, Alpha=@n175, Translation="0,0,0", Rotation="0,0,0", Scale="1,1,1") {
    rotationMode=BMM_Additive
    rotationSpace=BCS_BoneSpace
    translationMode=BMM_Additive
    translationSpace=BCS_BoneSpace
    alphaBoolBlend {blendOption=Linear}
    boneToModify {boneName=dyn_coat_r_02}
   }
   n181 = "Transform (Modify) Bone - Bone: dyn_coat_r_03"(ComponentPose=@n180, Alpha=@n175, Translation="0,0,0", Rotation="0,0,0", Scale="1,1,1") {
    rotationMode=BMM_Additive
    rotationSpace=BCS_BoneSpace
    translationMode=BMM_Additive
    translationSpace=BCS_BoneSpace
    alphaBoolBlend {blendOption=Linear}
    boneToModify {boneName=dyn_coat_r_03}
   }
   n182 = "Constraint - dyn_coat_bk_l_04 (RR)"!(ComponentPose=@n181, Alpha=0.8, ConstraintWeights_0=0.5, ConstraintWeights_1:float=0.5) -> (Pose) {
    alphaBoolBlend {blendOption=Linear}
    boneToModify {boneName=dyn_coat_bk_l_04}
    constraintSetup [{transformType=Rotation, targetBone {boneName=calf_l}}, {transformType=Rotation, targetBone {boneName=foot_l}}]
    constraintWeights [1,1]
   }
   n183 = "Constraint - dyn_coat_bk_r_04 (RR)"!(ComponentPose=@n182, Alpha=0.8, ConstraintWeights_0=0.5, ConstraintWeights_1:float=0.5) -> (Pose) {
    alphaBoolBlend {blendOption=Linear}
    boneToModify {boneName=dyn_coat_bk_r_04}
    constraintSetup [{transformType=Rotation, targetBone {boneName=calf_r}}, {transformType=Rotation, targetBone {boneName=foot_r}}]
    constraintWeights [1,1]
   }
   n184 = "Constraint - dyn_coat_bk_c_04 (RR)"!(ComponentPose=@n183, Alpha=1, ConstraintWeights_0=0.5, ConstraintWeights_1:float=0.5) -> (Pose) {
    alphaBoolBlend {blendOption=Linear}
    boneToModify {boneName=dyn_coat_bk_c_04}
    constraintSetup [{transformType=Rotation, targetBone {boneName=dyn_coat_bk_l_04}}, {transformType=Rotation, targetBone {boneName=dyn_coat_bk_r_04}}]
    constraintWeights [1,1]
   }
   n185 = "Constraint - dyn_coat_bk_l_05 (RR)"!(ComponentPose=@n184, Alpha=0.8, ConstraintWeights_0=0.5, ConstraintWeights_1:float=0.5) -> (Pose) {
    alphaBoolBlend {blendOption=Linear}
    boneToModify {boneName=dyn_coat_bk_l_05}
    constraintSetup [{transformType=Rotation, targetBone {boneName=calf_l}}, {transformType=Rotation, targetBone {boneName=foot_l}}]
    constraintWeights [1,1]
   }
   n186 = "Constraint - dyn_coat_bk_r_05 (RR)"!(ComponentPose=@n185, Alpha=0.8, ConstraintWeights_0=0.5, ConstraintWeights_1:float=0.5) -> (Pose) {
    alphaBoolBlend {blendOption=Linear}
    boneToModify {boneName=dyn_coat_bk_r_05}
    constraintSetup [{transformType=Rotation, targetBone {boneName=calf_r}}, {transformType=Rotation, targetBone {boneName=foot_r}}]
    constraintWeights [1,1]
   }
   n187 = "Constraint - dyn_coat_bk_c_05 (RR)"!(ComponentPose=@n186, Alpha=1, ConstraintWeights_0=0.5, ConstraintWeights_1:float=0.5) -> (Pose) {
    alphaBoolBlend {blendOption=Linear}
    boneToModify {boneName=dyn_coat_bk_c_05}
    constraintSetup [{transformType=Rotation, targetBone {boneName=dyn_coat_bk_l_04}}, {transformType=Rotation, targetBone {boneName=dyn_coat_bk_r_04}}]
    constraintWeights [1,1]
   }
   Save cached pose 'skirtDefaultCache'(Pose=Component To Local(ComponentPose=@n187)) {cacheName=skirtDefaultCache}
   n188 = "Constraint - dyn_coat_front_l_01 (P)"(ComponentPose=Local To Component(LocalPose=Use cached pose 'controlRigCache'()), Alpha=1, ConstraintWeights_0=1) {
    alphaBoolBlend {blendOption=Linear}
    boneToModify {boneName=dyn_coat_front_l_01}
    constraintSetup [{targetBone {boneName=thigh_l}}]
    constraintWeights [1]
   }
   n189 = "Constraint - dyn_coat_front_l_02 (P)"(ComponentPose=@n188, Alpha=1, ConstraintWeights_0=1) {
    alphaBoolBlend {blendOption=Linear}
    boneToModify {boneName=dyn_coat_front_l_02}
    constraintSetup [{targetBone {boneName=thigh_l}}]
    constraintWeights [1]
   }
   n190 = "Constraint - dyn_coat_front_l_04 (P)"(ComponentPose=@n189, Alpha=1, ConstraintWeights_0=1) {
    alphaBoolBlend {blendOption=Linear}
    boneToModify {boneName=dyn_coat_front_l_04}
    constraintSetup [{targetBone {boneName=calf_l}}]
    constraintWeights [1]
   }
   n191 = "Constraint - dyn_coat_l_01 (P)"(ComponentPose=@n190, Alpha=1, ConstraintWeights_0=1) {
    alphaBoolBlend {blendOption=Linear}
    boneToModify {boneName=dyn_coat_l_01}
    constraintSetup [{targetBone {boneName=thigh_l}}]
    constraintWeights [1]
   }
   n192 = "Constraint - dyn_coat_l_03 (R)"(ComponentPose=@n191, Alpha=1, ConstraintWeights_0=1) {
    alphaBoolBlend {blendOption=Linear}
    boneToModify {boneName=dyn_coat_l_03}
    constraintSetup [{transformType=Rotation, targetBone {boneName=thigh_l}}]
    constraintWeights [1]
   }
   n193 = "Constraint - dyn_coat_l_04 (R)"(ComponentPose=@n192, Alpha=1, ConstraintWeights_0=1) {
    alphaBoolBlend {blendOption=Linear}
    boneToModify {boneName=dyn_coat_l_04}
    constraintSetup [{transformType=Rotation, targetBone {boneName=calf_l}}]
    constraintWeights [1]
   }
   n194 = "Transform (Modify) Bone - Bone: dyn_coat_r_03"(ComponentPose=@n193, Alpha=0, Translation="0,0,-10", Rotation="0,0,0", Scale="1,1,1") {
    rotationMode=BMM_Additive
    translationMode=BMM_Additive
    alphaBoolBlend {blendOption=Linear}
    boneToModify {boneName=dyn_coat_r_03}
    translation {z=-10}
   }
   n195 = "Transform (Modify) Bone - Bone: dyn_coat_front_l_03"!(
    ComponentPose=@n194
    bAlphaBoolEnabled:bool=Get isSlidingMelee() -> (isSlidingMelee:bool)
    Translation="0,3,3"
    Rotation="0,0,0"
    Scale="1,1,1"
   ) -> (Pose) {
    alphaInputType=Bool
    rotationMode=BMM_Additive
    translationMode=BMM_Additive
    alphaBoolBlend {blendOption=Linear}
    boneToModify {boneName=dyn_coat_front_l_03}
    translation {y=3, z=3}
   }
   n196 = "Constraint - dyn_coat_bk_l_02 (P)"(ComponentPose=@n195, Alpha=1, ConstraintWeights_0=1) {
    alphaBoolBlend {blendOption=Linear}
    boneToModify {boneName=dyn_coat_bk_l_02}
    constraintSetup [{targetBone {boneName=thigh_l}}]
    constraintWeights [1]
   }
   n197 = "Constraint - dyn_coat_bk_l_04 (RP)"!(ComponentPose=@n196, Alpha=1, ConstraintWeights_0=1, ConstraintWeights_1:float=0) -> (Pose) {
    alphaBoolBlend {blendOption=Linear}
    boneToModify {boneName=dyn_coat_bk_l_04}
    constraintSetup [{transformType=Rotation, targetBone {boneName=calf_l}}, {targetBone {boneName=thigh_l}}]
    constraintWeights [1,1]
   }
   n198 = "Constraint - dyn_coat_bk_c_02 (PP)"!(ComponentPose=@n197, Alpha=1, ConstraintWeights_0=0.5, ConstraintWeights_1:float=0.5) -> (Pose) {
    alphaBoolBlend {blendOption=Linear}
    boneToModify {boneName=dyn_coat_bk_c_02}
    constraintSetup [{targetBone {boneName=thigh_l}}, {targetBone {boneName=thigh_r}}]
    constraintWeights [1,1]
   }
   n199 = "Constraint - dyn_coat_bk_c_04 (RR)"(ComponentPose=@n198, Alpha=1, ConstraintWeights_0=0.5, ConstraintWeights_1=0.5) {
    alphaBoolBlend {blendOption=Linear}
    boneToModify {boneName=dyn_coat_bk_c_04}
    constraintSetup [{transformType=Rotation, targetBone {boneName=calf_l}}, {transformType=Rotation, targetBone {boneName=calf_r}}]
    constraintWeights [1,1]
   }
   n200 = "Constraint - dyn_coat_front_r_01 (R)"(ComponentPose=@n199, Alpha=1, ConstraintWeights_0=1) {
    alphaBoolBlend {blendOption=Linear}
    boneToModify {boneName=dyn_coat_front_r_01}
    constraintSetup [{transformType=Rotation, targetBone {boneName=thigh_twist_01_r}}]
    constraintWeights [1]
   }
   n201 = "Constraint - dyn_coat_front_r_02 (R)"(ComponentPose=@n200, Alpha=1, ConstraintWeights_0=1) {
    alphaBoolBlend {blendOption=Linear}
    boneToModify {boneName=dyn_coat_front_r_02}
    constraintSetup [{transformType=Rotation, targetBone {boneName=thigh_twist_01_r}}]
    constraintWeights [1]
   }
   n202 = "Constraint - dyn_coat_front_r_04 (P)"(ComponentPose=@n201, Alpha=1, ConstraintWeights_0=1) {
    alphaBoolBlend {blendOption=Linear}
    boneToModify {boneName=dyn_coat_front_r_04}
    constraintSetup [{targetBone {boneName=calf_twist_01_r}}]
    constraintWeights [1]
   }
   n203 = "Constraint - dyn_coat_bk_r_01 (P)"(ComponentPose=@n202, Alpha=1, ConstraintWeights_0=1) {
    alphaBoolBlend {blendOption=Linear}
    boneToModify {boneName=dyn_coat_bk_r_01}
    constraintSetup [{targetBone {boneName=thigh_r}}]
    constraintWeights [1]
   }
   n204 = "Constraint - dyn_coat_bk_r_04 (PP)"!(ComponentPose=@n203, Alpha=1, ConstraintWeights_0=1, ConstraintWeights_1:float=0) -> (Pose) {
    alphaBoolBlend {blendOption=Linear}
    boneToModify {boneName=dyn_coat_bk_r_04}
    constraintSetup [{targetBone {boneName=calf_r}}, {targetBone {boneName=thigh_r}}]
    constraintWeights [1,1]
   }
   n205 = "Constraint - dyn_coat_r_01 (P)"(ComponentPose=@n204, Alpha=1, ConstraintWeights_0=1) {
    alphaBoolBlend {blendOption=Linear}
    boneToModify {boneName=dyn_coat_r_01}
    constraintSetup [{targetBone {boneName=thigh_r}}]
    constraintWeights [1]
   }
   n206 = "Constraint - dyn_coat_r_04 (P)"(ComponentPose=@n205, Alpha=1, ConstraintWeights_0=1) {
    alphaBoolBlend {blendOption=Linear}
    boneToModify {boneName=dyn_coat_r_04}
    constraintSetup [{targetBone {boneName=calf_twist_01_r}}]
    constraintWeights [1]
   }
   n207 = "Transform (Modify) Bone - Bone: dyn_coat_front_l_03"!(ComponentPose=@n206, Alpha:float=1, Translation="0,0,0", Rotation="0,0,0", Scale="1,1,1") -> (Pose) {rotationMode=BMM_Additive, translationMode=BMM_Additive, alphaBoolBlend {blendOption=Linear}, boneToModify {boneName=dyn_coat_front_l_03}}
   n208 = "Transform (Modify) Bone - Bone: dyn_coat_front_r_03"(ComponentPose=@n207, Alpha=1, Translation="0,0,0", Rotation="0,0,0", Scale="1,1,1") {rotationMode=BMM_Additive, translationMode=BMM_Additive, alphaBoolBlend {blendOption=Linear}, boneToModify {boneName=dyn_coat_front_r_03}}
   n209 = "Transform (Modify) Bone - Bone: dyn_coat_front_r_03"!(ComponentPose=@n208, bAlphaBoolEnabled:bool=Get isSlidingMelee(), Translation="0,0,0", Rotation="-20,0,0", Scale="1,1,1") -> (Pose) {
    alphaInputType=Bool
    rotationMode=BMM_Additive
    translationMode=BMM_Additive
    alphaBoolBlend {blendOption=Linear}
    boneToModify {boneName=dyn_coat_front_r_03}
    rotation {pitch=-20}
   }
   n210 = Get isSlidingAltWeapon() -> (isSlidingAltWeapon:bool)
   n211 = "Transform (Modify) Bone - Bone: dyn_coat_front_r_02"!(ComponentPose=@n209, bAlphaBoolEnabled:bool=@n210, Translation="0,0,0", Rotation="0,15,35", Scale="1,1,1") -> (Pose) {
    alphaInputType=Bool
    rotationMode=BMM_Additive
    translationMode=BMM_Additive
    alphaBoolBlend {blendOption=Linear}
    boneToModify {boneName=dyn_coat_front_r_02}
    rotation {roll=35, yaw=15}
   }
   n212 = "Transform (Modify) Bone - Bone: dyn_coat_front_r_04"!(ComponentPose=@n211, bAlphaBoolEnabled:bool=@n210, Translation="5,5,5", Rotation="0,0,0", Scale="1,1,1") -> (Pose) {
    alphaInputType=Bool
    rotationMode=BMM_Additive
    translationMode=BMM_Additive
    alphaBoolBlend {blendOption=Linear}
    boneToModify {boneName=dyn_coat_front_r_04}
    translation=<5 5 5>
   }
   Save cached pose 'skirtSlidingCache'(Pose=Component To Local(ComponentPose=@n212)) {cacheName=skirtSlidingCache}
   Get Gravity Override Cape()
   Get isCustomEmotePlaying()
   Get bIsInVehicle() -> (bIsInVehicle:bool)
   Get ADS Up Rifle()
  ]}
 EventGraph {nodes [
   Event BlueprintUpdateAnimation()
   > n1 = Sequence!(execute) -> (
    then_0
    then_1
    then_2:Exec
    then_3:Exec
    then_4:Exec
    then_5:Exec
    then_6:Exec
    then_7:Exec
    then_8:Exec
    then_9:Exec
    then_10:Exec
    then_11:Exec
    then_12:Exec
    then_13:Exec
    then_14:Exec
    then_15:Exec
    then_16:Exec
    then_17:Exec
    then_18:Exec
   )
   > Set arm L Rot(arm L Rot:double=GetAlphaFromReferenceTransformDelta(
     SkeletalMeshComponent=Get Body Mesh() -> (Body Mesh:Skeletal Mesh Component Object Reference)
     BoneName=upperarm_l
     BaseName=spine_05
     DeltaComponentType=EulerY
     InputRange=(X=0,Y=100)
     OutputRange=(X=0,Y=100)
     TwistAxis="1,0,0"
    ).RawOutput) -> (Output_Get:double)
   n2 = Gravity Override State Function DEPRECATED(execute=@n1.then_1, CharacterPartAnimInstance=Self-Reference(), GameTimeInSeconds=0, GravityOverrideParams=Get Gravity Override Params())
   > Set Gravity Override Cape(Gravity Override Cape:Vector=@n2) -> (Output_Get:Vector)
   n3 = Branch(execute=@n1.then_10, Condition=Get bIsSliding())
   > Set isNotSliding(isNotSliding:bool=false) -> (Output_Get:bool)
   Set isNotSliding(execute=@n3.else, isNotSliding=true)
   n4 = Cast To FortPlayerAnimInstance(execute=@n1.then_11, Object=GetAnimInstance(self=GetOwnerSkeletalMeshComponent()))
   > Set Alt_weapon(Alt_weapon:bool=false) -> (Output_Get:bool)
   > Set isRifle(isRifle:bool=true) -> (Output_Get:bool)
   > n5 = Switch on EFortWeaponCoreAnimation!(execute, Selection:EFortWeaponCoreAnimation Enum=Get WeaponCoreAnim(self=@n4)) -> (
    Melee:Exec
    Pistol:Exec
    Shotgun:Exec
    PaperBlueprint:Exec
    Rifle:Exec
    MeleeOneHand:Exec
    MachinePistol:Exec
    RocketLauncher:Exec
    GrenadeLauncher:Exec
    GoingCommando:Exec
    AssaultRifle:Exec
    TacticalShotgun:Exec
    SniperRifle:Exec
    TrapPlacement:Exec
    ShoulderLauncher:Exec
    AbilityDecoTool:Exec
    Crossbow:Exec
    C4:Exec
    RemoteControl:Exec
    DualWield:Exec
    AR_BullPup:Exec
    AR_ForwardGrip:Exec
    MedPackPaddles:Exec
    SMG_P90:Exec
    AR_DrumGun:Exec
    Consumable_Small:Exec
    Consumable_Large:Exec
    Balloon:Exec
    MountedTurret:Exec
    CreativeTool:Exec
    ExplosiveBow:Exec
    AshtonIndigo:Exec
    AshtonChicago:Exec
    MeleeDualWield:Exec
    Unarmed:Exec
   )
   n6 = Branch(execute=[@n5.AssaultRifle, @n5.Rifle])
   > Set isRifle(isRifle=true)
   Set isRifle(execute=@n6.else, isRifle=false)
   Branch(execute=[
     @n5.Crossbow
     @n5.DualWield
     @n5.GoingCommando
     @n5.GrenadeLauncher
     @n5.MachinePistol
     @n5.PaperBlueprint
     @n5.Pistol
     @n5.RocketLauncher
     @n5.Shotgun
     @n5.ShoulderLauncher
     @n5.SniperRifle
     @n5.TacticalShotgun
     @n5.TrapPlacement
    ])
   > Set Alt_weapon(Alt_weapon=true)
   > Set isRifle(isRifle=false)
   > n7 = Branch(Condition=Get bIsSliding())
   > Set isSlidingAltWeapon(isSlidingAltWeapon:bool=true) -> (Output_Get:bool)
   Set isSlidingAltWeapon(execute=@n7.else, isSlidingAltWeapon=false)
   n8 = Branch(execute=[@n5.Melee, @n5.MeleeOneHand])
   > Set isMelee(isMelee:bool=true) -> (Output_Get:bool)
   > n9 = Branch(Condition=Get bIsSliding())
   > Set isSlidingMelee(isSlidingMelee:bool=true) -> (Output_Get:bool)
   Set isSlidingMelee(execute=@n9.else, isSlidingMelee=false)
   Set isMelee(execute=@n8.else, isMelee=false)
   > Set isSlidingMelee(isSlidingMelee=false)
   n10 = Set head Rot Dn(
    execute=@n1.then_12
    head Rot Dn:double=GetAlphaFromReferenceTransformDelta(SkeletalMeshComponent=Get Body Mesh(), BoneName=head, BaseName=spine_05, InputRange=(X=0,Y=15), TwistAxis="1,0,0")
   ) -> (Output_Get:double)
   > n11 = Branch(Condition=AND Boolean(A=Get bIsTargeting(), B=Get isRifle() -> (isRifle:bool)))
   > Set ADS Rifle Dn(ADS Rifle Dn:double=@n10) -> (Output_Get:double)
   Set ADS Rifle Dn(execute=@n11.else, ADS Rifle Dn=0)
   n12 = Branch(execute=@n1.then_13, Condition=Get bIsInFrontEnd())
   > Set is not front end(is not front end:double=0.1) -> (Output_Get:double)
   Set is not front end(execute=@n12.else, is not front end=1)
   n13 = GetAnimInstance(self=GetOwnerSkeletalMeshComponent())
   Is Valid(exec=@n1.then_14, InputObject=@n13)
   > n14 = Set isCustomEmotePlaying(isCustomEmotePlaying:bool="Equal (Object)"(
     A=GetCurrentActiveMontage!(self:Anim Instance Object Reference=@n13) -> (ReturnValue:Anim Montage Object Reference)
     B=/BRCosmetics/Animation/Game/MainPlayer/Menu/BR/CustomSkins/_s21/RealmM/Realm_Male_Idle_Montage
    )) -> (Output_Get:bool)
   > n15 = Branch(Condition=@n14)
   > Set controlRigAlpha(controlRigAlpha:double=0.5) -> (Output_Get:double)
   Set controlRigAlpha(execute=@n15.else, controlRigAlpha=1)
   n16 = Sequence!(execute=@n1.then_15) -> (then_0, then_1, then_2)
   > n17 = Branch(Condition=OR Boolean!(
     A=Get bIsOstrichVehicleDriver() -> (bIsOstrichVehicleDriver:bool)
     B=Get bIsOstrichVehicleGunner() -> (bIsOstrichVehicleGunner:bool)
     C:bool=Get bIsFerretVehicleDriver() -> (bIsFerretVehicleDriver:bool)
     D:bool=Get bIsValetDriver() -> (bIsValetDriver:bool)
     E:bool=Get bIsGolfCartVehicleDriver() -> (bIsGolfCartVehicleDriver:bool)
     F:bool=Get bIsGolfCartVehicleFrontPassenger() -> (bIsGolfCartVehicleFrontPassenger:bool)
     G:bool=Get bIsOctopusVehicleDriver() -> (bIsOctopusVehicleDriver:bool)
     H:bool=Get bIsAntelopeVehicleDriver() -> (bIsAntelopeVehicleDriver:bool)
     I:bool=Get bIsMeatballDriver() -> (bIsMeatballDriver:bool)
    ) -> (ReturnValue))
   > Set in vehicle rigid alpha(in vehicle rigid alpha:double=0) -> (Output_Get:double)
   > Set in vehicle scale alpha(in vehicle scale alpha:double=1) -> (Output_Get:double)
   Set in vehicle rigid alpha(execute=@n17.else, in vehicle rigid alpha=1)
   > Set in vehicle scale alpha(in vehicle scale alpha=0)
   n18 = Cast To FortPlayerPawn(execute=@n16.then_1, Object=GetOwningActor())
   > Set As Fort Player Pawn(As Fort Player Pawn:Fort Player Pawn Object Reference=@n18) -> (Output_Get:Fort Player Pawn Object Reference)
   > n19 = Get!(execute:Exec) -> (then:Exec, else:Exec, As Fort Player Pawn:Fort Player Pawn Object Reference)
   n20 = Get Dynamic Material Instance(execute=@n16.then_2, SkeletalMeshComponent=GetPartSkeletalMeshComponent(), ElementIndex=1)
   n21 = HasMatchingGameplayTag!(self:Gameplay Tag Asset Interface Interface=@n19, TagToCheck:Gameplay Tag Structure=(TagName="Item.Weapon.Galileo.Lobster.Realm")) -> (ReturnValue:bool)
   > n22 = Branch(Condition=@n21)
   > n23 = Branch(Condition=Get bIsSkyDiving())
   SetScalarParameterValue(execute=[@n45.then, @n23.then, @n46.else, @n22.else], self=@n20, ParameterName=Hide Element 02, Value=0)
   SetScalarParameterValue(execute=[@n23.else, @n46.then], self=@n20, ParameterName=Hide Element 02, Value=1)
   Get!(execute=@n1.then_16) -> (then, else, Anim_Inst:Anim Instance Object Reference)
   n24 = GetCurveValue(self=Get Anim_Inst() -> (Anim_Inst:Anim Instance Object Reference), CurveName=PhysicsGroundOffset)
   > n25 = Branch(Condition="float > float"(A=@n24))
   > Set PlummetAlpha(PlummetAlpha:double=@n24) -> (Output_Get:double)
   > Set shouldReset(shouldReset:bool=true) -> (Output_Get:bool)
   Set PlummetAlpha(execute=@n25.else, PlummetAlpha=0)
   Branch(
    execute=@n1.then_17
    Condition=AND Boolean!(A=Get shouldReset() -> (shouldReset:bool), B="Equal (Float)"(A=@n24), C:bool=Get bIsInFrontEnd()) -> (ReturnValue)
   )
   > ResetDynamics(InTeleportType=TeleportPhysics)
   > ResetDynamics(InTeleportType=ResetPhysics)
   > Set shouldReset(shouldReset=false)
   n26 = GetAnimInstance(self=GetOwnerSkeletalMeshComponent())
   Is Valid(exec=@n1.then_18, InputObject=@n26)
   > n27 = Branch(Condition="Equal (Object)"(A=GetCurrentActiveMontage(self=@n26), B=/BRCosmetics/Animation/Game/MainPlayer/Emotes/Galileo1/Emote_Galileo1_CMM_M))
   > Set isDarkSideEmotePlaying(isDarkSideEmotePlaying:bool=true) -> (Output_Get:bool)
   Set isDarkSideEmotePlaying(execute=@n27.else, isDarkSideEmotePlaying=false)
   n28 = Get bIsParachuteOpened()
   n29 = Branch(execute=@n1.then_2, Condition=AND Boolean!(A=Get bIsSkyDiving(), B=NOT Boolean(A=@n28)) -> (ReturnValue))
   > Set skydiveNoParachute(skydiveNoParachute:bool=true) -> (Output_Get:bool)
   Set skydiveNoParachute(execute=@n29.else, skydiveNoParachute=false)
   Set Is Wearing Backbling(execute=@n1.then_3, Is Wearing Backbling=CheckForBackbling())
   Set thigh L Rot(
    execute=@n1.then_4
    thigh L Rot:double=GetAlphaFromReferenceTransformDelta(SkeletalMeshComponent=Get Body Mesh(), BoneName=thigh_l, BaseName=pelvis, InputRange=(X=-45,Y=-15), OutputRange=(X=1,Y=0), TwistAxis="1,0,0")
   ) -> (Output_Get:double)
   Set Body Mesh(execute=@n1.then_5, Body Mesh:Skeletal Mesh Component Object Reference=GetPartSkeletalMeshComponent(self=Self-Reference())) -> (Output_Get:Skeletal Mesh Component Object Reference)
   n30 = Branch(execute=@n1.then_6, Condition=Get bIsTargeting())
   n31 = Self-Reference()
   n32 = BreakTransform(InTransform=GetDeltaTransformFromRefPose(self=GetOwnerSkeletalMeshComponent(self=@n31), BoneName=neck_01, BaseName=spine_05))
   > n33 = Set ADS Up Rifle(ADS Up Rifle:double=Remap Float(Input=BreakRotator(InRot=@n32.Rotation).Pitch, OldRange=(X=0,Y=40), NewRange=(X=0,Y=1), Clamp=true) -> (Output=~)) -> (Output_Get:double)
   > n34 = Branch(Condition=Get bIsCrouching())
   > Set ADS Up Crouch Bool(ADS Up Crouch Bool:double=@n33) -> (Output_Get:double)
   Set ADS Up Crouch Bool(execute=@n34.else, ADS Up Crouch Bool=0)
   Set ADS Up Rifle(execute=@n30.else, ADS Up Rifle=0)
   n35 = Branch(execute=@n1.then_7, Condition=Get bIsTacticalSprinting())
   > Set Gravity Override Params(Gravity Override Params:Gravity Override Params Struct Structure) -> (Output_Get:Gravity Override Params Struct Structure)
   Set Gravity Override Params(execute=@n35.else)
   n36 = PlayerMovement(execute=@n1.then_8, AnimInstance=Self-Reference(), XDotProductCoefficient=1, YDotProductCoefficient=1, ZDotProductCoefficient=1)
   > n37 = Branch(Condition=OR Boolean!(A="Not Equal (Float)"(A=@n36), B="Not Equal (Float)"(A=@n36.YMovement), C="Not Equal (Float)"(A=@n36.ZMovement)) -> (ReturnValue))
   > Set charMoving(charMoving:bool=true) -> (Output_Get:bool)
   Set charMoving(execute=@n37.else, charMoving=false)
   n38 = Get charMoving() -> (charMoving:bool)
   n39 = Branch(execute=@n1.then_9, Condition=AND Boolean(A=Get bIsCrouching(), B=NOT Boolean(A=@n38)))
   > Set isCrouchStopped(isCrouchStopped:bool=true) -> (Output_Get:bool)
   > Set crouch moving(crouch moving:bool=false) -> (Output_Get:bool)
   Set isCrouchStopped(execute=@n39.else, isCrouchStopped=false)
   > n40 = Branch(Condition=Get bIsCrouching())
   > Set crouch moving(crouch moving=true)
   > n41 = Branch(Condition=Get bIsSliding())
   > Set crouchMovingNotSliding(crouchMovingNotSliding:bool=false) -> (Output_Get:bool)
   Set crouchMovingNotSliding(execute=@n41.else, crouchMovingNotSliding=true)
   Set crouch moving(execute=@n40.else, crouch moving=false)
   Event BlueprintInitializeAnimation()
   n42 = GetOwnerSkeletalMeshComponent()
   > Is Valid(InputObject=@n42)
   > Set Skel_Mesh(Skel_Mesh=@n42)
   > n43 = Cast To FortPlayerPawn(Object=GetOwningActor())
   > n44 = Set Pawn(Pawn:Fort Player Pawn Object Reference=@n43) -> (Output_Get:Fort Player Pawn Object Reference)
   > Set Anim_Inst(Anim_Inst:Anim Instance Object Reference=GetAnimInstance(self=Get Mesh(self=@n44))) -> (Output_Get:Anim Instance Object Reference)
   n45 = Event OnBeginSkydiving()
   Event OnEndSkydiving()
   > n46 = Branch(Condition=@n21)
   n47 = Branch(Condition=Get bIsInVehicle())
   > Set drivingVehicle(drivingVehicle:bool=true) -> (Output_Get:bool)
   > n48 = Branch(Condition=Get bIsInCannon() -> (bIsInCannon:bool))
   > Set drivingVehicle(drivingVehicle=false)
   Set drivingVehicle(execute=@n48.else, drivingVehicle=false)
   Set drivingVehicle(execute=@n47.else, drivingVehicle=false)
   AND Boolean(A=false, B=NOT Boolean(A=false))
   Event OnWeaponChanged()
   Get bIsSlopeSliding() -> (bIsSlopeSliding:bool)
   Set Z Movement(Z Movement:double=@n36.YMovement) -> (Output_Get:double)
  ]}
]
variables [
 ADS Rifle Dn {type=number}
 ADS Up and Crouch {type=boolean}
 ADS Up Crouch Bool {type=number}
 ADS Up Rifle {type=number}
 Alt_weapon {type=boolean}
 Anim_Inst {type=AnimInstance, default=null}
 arm L Rot {type=number}
 As Fort Player Pawn {type=FortPlayerPawn}
 Body Mesh {type=SkeletalMeshComponent}
 charMoving {type=boolean}
 controlRigAlpha {type=number}
 crouch moving {type=boolean}
 crouchMovingNotSliding {type=boolean}
 drivingVehicle {type=boolean}
 Gravity Override Cape {type=Vector}
 Gravity Override Params {
  type=GravityOverrideParamsStruct
  default {
   onGround_Moving {gravityOverride=<0 -150 5000>, windAmplitude=<0 -500 500>, windFrequency=<0 -1000 5000>}
   onGround_Standing {gravityOverride=<0 -500 -980>, windFrequency=<0 -1000 5000>}
   skydive_DivingDown {gravityOverride=<-500 0 0>}
   skydive_FloatingDown {gravityOverride=<0 -400 600>}
   skydive_Parachuting {gravityOverride=<0 -600 -980>}
  }
 }
 head Rot Dn {type=number}
 in vehicle rigid alpha {type=number}
 in vehicle scale alpha {type=number}
 is not front end {type=number}
 Is Wearing Backbling {type=boolean}
 isCrouchStopped {type=boolean}
 isCustomEmotePlaying {type=boolean}
 isDarkSide {type=boolean}
 isDarkSideEmotePlaying {type=boolean}
 isMelee {type=boolean}
 isNotSliding {type=boolean}
 isRifle {type=boolean}
 isSlidingAltWeapon {type=boolean}
 isSlidingMelee {type=boolean}
 NewVar {type=boolean}
 Pawn {type=FortPlayerPawn}
 PlummetAlpha {type=number}
 shouldReset {type=boolean}
 Skel_Mesh {type=SkeletalMeshComponent}
 skydiveNoParachute {type=boolean}
 thigh L Rot {type=number}
 Z Movement {type=number}
]
