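# Animation Blueprint for the Realm male body's face accessory, damping head and neck motion inherited from the attached parent mesh.
# It copies the parent pose, layer-blends the neck toward an unconnected input, then adds bone-space rotations to head and neck_01.
# Updates derive these rotations from negated, scaled owner-mesh ref-pose deltas, adding a head pitch offset when targeting with certain weapons.
# A GuineaPig emote flag is set but unused here.
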AnimBlueprint /BRCosmetics/Characters/Player/Male/Medium/Bodies/M_MED_Realm/Meshes/Parts/M_MED_Realm_FaceAcc_AnimBP
target_skeleton=&/M_MED_Realm_FaceAcc_Skeleton
graphs [
 AnimGraph {nodes [
   n1 = "Transform (Modify) Bone - Bone: head"(
    ComponentPose=Local To Component(LocalPose=Use cached pose 'headPose'())
    Alpha=1
    Translation="0,0,0"
    Rotation=Get headRotator() -> (headRotator:Rotator)
    Scale="1,1,1"
   ) {
    rotationMode=BMM_Additive
    rotationSpace=BCS_BoneSpace
    alphaBoolBlend {blendOption=Linear}
    alphaScaleBiasClamp {bInterpResult=true}
    boneToModify {boneName=head}
   }
   n2 = "Transform (Modify) Bone - Bone: neck_01"(ComponentPose=@n1, Alpha=1, Translation="0,0,0", Rotation=Get neck02 Rotator() -> (neck02 Rotator:Rotator), Scale="1,1,1") {
    rotationMode=BMM_Additive
    rotationSpace=BCS_BoneSpace
    alphaBoolBlend {blendOption=Linear}
    alphaScaleBiasClamp {bInterpResult=true}
    boneToModify {boneName=neck_01}
   }
   Output Pose(Result=Component To Local(ComponentPose=@n2)) {layerGroup=None}
   Save cached pose 'headJointCache'(Pose=Copy Pose From Mesh() {bCopyCurves=true, bUseAttachedParent=true}) {cacheName=headJointCache}
   n3 = Layered blend per bone!(
    BasePose=Use cached pose 'headJointCache'()
    BlendPoses_0
    BlendPoses_1:Pose Link Structure=Use cached pose 'headJointCache'()
    BlendWeights_0=0.6
    BlendWeights_1:float=1
   ) -> (Pose) {
    blendPoses [{linkId=-1, sourceLinkId=-1}, {linkId=-1, sourceLinkId=-1}]
    blendWeights [1,1]
    layerSetup [{branchFilters [{blendDepth=1, boneName=neck_01}, {blendDepth=-1, boneName=head}]}, {branchFilters []}]
   }
   Save cached pose 'headPose'(Pose=@n3) {cacheName=headPose}
  ]}
 EventGraph {nodes [
   Event BlueprintUpdateAnimation()
   > n1 = Sequence!(execute) -> (then_0, then_1, then_2:Exec, then_3:Exec, then_4:Exec)
   n2 = GetOwnerSkeletalMeshComponent()
   n3 = BreakRotator(InRot=BreakTransform(InTransform=GetDeltaTransformFromRefPose(self=@n2, BoneName=head, BaseName=neck_02)).Rotation)
   n4 = Get multiplyValHead() -> (multiplyValHead:double)
   n5 = float * float!(A=@n3.Yaw, B=@n4, C:double=-1) -> (ReturnValue) {is=title K2Node_PromotableOperator}
   n6 = float * float(A=@n3.Pitch, B=@n4, C=-1)
   n7 = Get Rot Head Val() -> (Rot Head Val:double)
   > Set headRotator(headRotator:Rotator=MakeRotator(Roll=float * float(A=@n3, B=@n4, C=-1), Pitch=float + float(A=@n6, B=@n7), Yaw=float + float(A=@n5))) -> (Output_Get:Rotator)
   n8 = GetOwnerSkeletalMeshComponent()
   n9 = BreakRotator(InRot=BreakTransform(InTransform=GetDeltaTransformFromRefPose(self=@n8, BoneName=neck_02, BaseName=neck_01)).Rotation)
   n10 = Get multiplyValNeck() -> (multiplyValNeck:double)
   Set neck02 Rotator(
    execute=@n1.then_1
    neck02 Rotator:Rotator=MakeRotator(Roll=float * float(A=@n9, B=@n10, C=-1), Pitch=float * float(A=@n9.Pitch, B=@n10, C=-1), Yaw=float * float(A=@n9.Yaw, B=@n10, C=-1))
   ) -> (Output_Get:Rotator)
   n11 = Cast To FortPlayerAnimInstance(execute=@n1.then_2, Object=GetAnimInstance(self=GetOwnerSkeletalMeshComponent()))
   > Set Alt_weapon(Alt_weapon:bool=false) -> (Output_Get:bool)
   > n12 = Switch on EFortWeaponCoreAnimation!(execute, Selection:EFortWeaponCoreAnimation Enum=Get WeaponCoreAnim(self=@n11)) -> (
    Melee:Exec
    Pistol:Exec
    Shotgun:Exec
    PaperBlueprint:Exec
    Rifle:Exec
    MeleeOneHand:Exec
    MachinePistol:Exec
    RocketLauncher:Exec
    GrenadeLauncher:Exec
    GoingCommando:Exec
    AssaultRifle:Exec
    TacticalShotgun:Exec
    SniperRifle:Exec
    TrapPlacement:Exec
    ShoulderLauncher:Exec
    AbilityDecoTool:Exec
    Crossbow:Exec
    C4:Exec
    RemoteControl:Exec
    DualWield:Exec
    AR_BullPup:Exec
    AR_ForwardGrip:Exec
    MedPackPaddles:Exec
    SMG_P90:Exec
    AR_DrumGun:Exec
    Consumable_Small:Exec
    Consumable_Large:Exec
    Balloon:Exec
    MountedTurret:Exec
    CreativeTool:Exec
    ExplosiveBow:Exec
    AshtonIndigo:Exec
    AshtonChicago:Exec
    MeleeDualWield:Exec
    Unarmed:Exec
   )
   n13 = Get bIsTargeting()
   n14 = Branch(execute=[@n12.AssaultRifle, @n12.Pistol, @n12.Rifle, @n12.Shotgun], Condition=@n13)
   > Set Rot Head Val(Rot Head Val:double=-7) -> (Output_Get:double)
   Set Rot Head Val(execute=@n14.else, Rot Head Val=0)
   n15 = Branch(execute=@n12.DualWield, Condition=@n13)
   > Set Rot Head Val(Rot Head Val=-11)
   Set Rot Head Val(execute=@n15.else, Rot Head Val=0)
   Branch(execute=@n1.then_3, Condition=Get Alt_weapon() -> (Alt_weapon:bool))
   n16 = GetAnimInstance(self=GetOwnerSkeletalMeshComponent())
   Is Valid(exec=@n1.then_4, InputObject=@n16)
   > Set isGuineaPigEmote(isGuineaPigEmote:bool=NOT Boolean(A=Montage_IsPlaying(self=@n16, Montage=/BRCosmetics/Animation/Game/MainPlayer/Emotes/GuineaPig/Emote_GuineaPig_CMM_FE_M))) -> (Output_Get:bool)
   Event BlueprintInitializeAnimation()
   > Set Body_Mesh(Body_Mesh=GetOwnerSkeletalMeshComponent())
   > n17 = Cast To PlayerPawn_Athena(Object=TryGetPawnOwner())
   > n18 = Set Pawn Owner(Pawn Owner:Player Pawn Athena Object Reference=@n17) -> (Output_Get:Player Pawn Athena Object Reference)
   > Set Head_Mesh(Head_Mesh:Skeletal Mesh Component Object Reference=GetSkeletalMeshForPartType(self=@n18)) -> (Output_Get:Skeletal Mesh Component Object Reference)
   Set Alt_weapon(Alt_weapon=true)
  ]}
]
variables [
 ADS Up Rifle {type=number}
 Alt_weapon {type=boolean}
 Body_Mesh {type=SkeletalMeshComponent}
 Head_Mesh {type=SkeletalMeshComponent}
 headRotator {type=Rotator}
 isGuineaPigEmote {type=boolean}
 multiplyValHead {type=number, default=0.3}
 multiplyValNeck {type=number}
 neck02 Rotator {type=Rotator}
 Out Rotation {type=Rotator}
 Pawn Owner {type=PlayerPawn_Athena_C}
 Rot Head Val {type=number}
 Target {type=SkinnedMeshComponent, default=null}
]
